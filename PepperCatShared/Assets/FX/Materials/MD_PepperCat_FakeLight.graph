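# Deferred decal material for PepperCat FX that, per its name, fakes a light as an unlit, translucent glow on surfaces.
# It builds a spherical falloff from the world-space distance between each pixel and the actor position, shaped by Scale and Expo parameters.
# The inverted falloff is multiplied by decal lifetime opacity so the glow can fade out, while Decal Color supplies the tint.
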Material /PepperCatShared/Assets/FX/Materials/MD_PepperCat_FakeLight
graph {
 properties {
  bAutomaticallySetUsageInEditor=false
  blendMode=BLEND_Translucent
  materialDomain=MD_DeferredDecal
  shadingModel=MSM_Unlit
  usedShadingModels=Unlit
 }
 nodes [
  DecalColor()
  n1 = WorldPosition()
  n2 = ActorPositionWS()
  n3 = Saturate(None=Multiply(A=Distance(A=@n1.XYZ, B=@n2), B=ScalarParameter() {defaultValue=0.002, parameterName=Scale}))
  n4 = ScalarParameter() {defaultValue=0.8, parameterName=Expo}
  Multiply(A=DecalLifetimeOpacity().Opacity, B=OneMinus(None=Power(Base=@n3, Exp=@n4)))
 ]
}
