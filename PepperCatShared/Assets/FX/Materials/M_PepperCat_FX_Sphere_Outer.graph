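# Translucent, unlit, two-sided effect material for Niagara mesh particles, by its name the outer layer of a PepperCat sphere effect, tinted by particle color.
# Its alpha combines a contrasted Fresnel band, depth fade, bottom falloff and particle alpha with dynamic-parameter cell-texture erosion.
# Rotating, scrolling world-space Voronoi noise breaks up that alpha, and squared near-camera fade applies except on mobile.
# Named reroutes hold the noise and a normal-based up/down mask.
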Material /PepperCatShared/Assets/FX/Materials/M_PepperCat_FX_Sphere_Outer
graph {
 properties {
  bUsedWithNanite=true
  bUsedWithNiagaraMeshParticles=true
  blendMode=BLEND_Translucent
  pixelDepthOffsetMode=PDOM_Legacy
  shadingModel=MSM_Unlit
  twoSided=true
  usedShadingModels=Unlit
 }
 nodes [
  n1 = ParticleColor()
  Add(A=NamedRerouteUsage(), B=@n1.RGB)
  MaterialFunctionCall(Emissive=@n1.RGB) {materialFunction=&MF_FX_EmissiveAdjustments}
  n2 = MaterialFunctionCall(Distance=ScalarParameter() {defaultValue=800, parameterName=FadeDistance}) {materialFunction=&Near_Camera_Fade}
  n3 = MaterialFunctionCall(
   In=OneMinus(None=MaterialFunctionCall() {materialFunction=&LinearGradient}.VGradient)
   Contrast=ScalarParameter() {defaultValue=0.05, parameterName=BottomFalloff}
  ) {materialFunction=&CheapContrast}
  n4 = Fresnel() {exponent=1}
  n5 = Multiply(
   A=MaterialFunctionCall(In=Fresnel() {exponent=1}, Contrast=Constant() {r=0.1}) {materialFunction=&CheapContrast}.Result
   B=MaterialFunctionCall(In=OneMinus(None=@n4), Contrast=Constant() {r=1}) {materialFunction=&CheapContrast}.Result
  )
  n6 = ScalarParameter() {defaultValue=1, parameterName=FresnelBrightness}
  n7 = DepthFade(
   Opacity=Clamp(None=Multiply(A=@n5, B=@n6))
   FadeDistance=DynamicParameter() {defaultValue=<rgba 1 1 1 100>, paramNames [RefractionIndex, Erode, ErodeHardness, DepthFadeAmount]}.DepthFadeAmount
  )
  n8 = TextureSample(UVs=TextureCoordinate() {uTiling=3, vTiling=2}) {texture=/Game/Global/Textures/Default/Noise/T_SoftCells_Negative}
  n9 = DynamicParameter() {defaultValue=<rgba 1 1 1 100>, paramNames [RefractionIndex, Erode, ErodeHardness, DepthFadeAmount]}
  n10 = Add(A=Multiply(A=Add(A=@n8.G, B=@n8.G), B=@n9.ErodeHardness) {constB=2}, B=Multiply(A=@n9.ErodeHardness, B=@n9.Erode))
  n11 = WorldPosition()
  n12 = Multiply(A=AppendVector(A=Rotator(Coordinate=@n11.XY), B=@n11.Z), B=ScalarParameter() {defaultValue=0.0075, parameterName=Detail Scale})
  n13 = Multiply(A=Time(), B=VectorParameter() {parameterName=3D Noise Scroll, defaultValue=<rgba 2 1.5 -4 0>}.RGB)
  n14 = Power(
   Base=TextureSample(UVs=Add(A=@n12, B=@n13)) {samplerType=SAMPLERTYPE_LinearColor, texture=/Game/Effects/Explosion_Update/Textures/VT_Voronoi_Seamless}.RGB
   Exp=ScalarParameter() {defaultValue=9, parameterName=Pow}
  ) {constExponent=7}
  n15 = Multiply(A=Multiply(A=@n1.A, B=Multiply(A=@n3.Result, B=@n7)), B=Clamp(None=Multiply(A=@n10, B=@n14)))
  n16 = Multiply(A=Multiply(A=@n2.Result, B=@n2.Result), B=@n15)
  Multiply(A=Saturate(None=ShadingPathSwitch(Default=@n16, Deferred=@n16, Mobile=@n15) {desc=Migrated from FeatureLevelSwitch}), B=NamedRerouteUsage())
  NamedRerouteDeclaration(None=Saturate(None=@n14)) {name=3DNoise, nodeColor=<rgba 1 0.0234 0 1>}
  n17 = DotProduct(A=Constant3Vector() {constant=<rgba 0 0 1 0>}, B=VertexNormalWS())
  n18 = Multiply(A=OneMinus(None=Abs(None=@n17))) {constB=15}
  NamedRerouteDeclaration(None=Saturate(None=Power(Base=@n18) {constExponent=5})) {name=UPDOWN Mask, nodeColor=<rgba 0 0.4609 1 1>}
 ]
}
