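# Unlit translucent FX material for Niagara mesh particles that draws an airglow effect with drifting noise.
# Particle color RGB feeds a shared emissive adjustment function.
# Particle alpha scales a mask built from panning cloud noise shaped by vertical UV gradients, plus a sharper gradient band.
# An exponent-controlled UV gradient, offset by a per-particle dynamic parameter, further multiplies that mask.
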Material /PepperCatShared/Assets/FX/Materials/M_Peppercat_Airglow
graph {
 properties {
  bUsedWithNiagaraMeshParticles=true
  blendMode=BLEND_Translucent
  pixelDepthOffsetMode=PDOM_Legacy
  shadingModel=MSM_Unlit
  usedShadingModels=Unlit
 }
 nodes [
  MaterialFunctionCall(Emissive=ParticleColor().RGB) {materialFunction=&MF_FX_EmissiveAdjustments}
  n1 = Constant() {r=1}
  n2 = Constant()
  n3 = Panner(Coordinate=Multiply(A=TextureCoordinate(), B=AppendVector(A=@n1, B=@n2)), Speed=AppendVector(A=Constant() {r=0.02}, B=Constant() {r=0.1}))
  n4 = Multiply(A=Power(Base=TextureSample(UVs=@n3) {texture=&AtmosphericCloudNoise01}.R))
  n5 = Power(Base=ComponentMask(None=TextureCoordinate()) {g=true})
  n6 = ComponentMask(None=TextureCoordinate()) {g=true}
  n7 = Add(A=Multiply(A=Clamp(None=@n4), B=Multiply(A=@n5)), B=Multiply(A=Power(Base=@n6) {constExponent=8}) {constB=0.5})
  n8 = ComponentMask(None=TextureCoordinate()) {g=true}
  n9 = ScalarParameter() {defaultValue=2, parameterName=Exp}
  n10 = Add(
   A=Multiply(A=Power(Base=@n8, Exp=@n9), B=ScalarParameter() {defaultValue=1, parameterName=B})
   B=DynamicParameter() {defaultValue=<rgba 0 1 1 1>, paramNames [Mask_0to1,Param2,Param3,Param4]}.Mask_0to1
  )
  Multiply(A=ParticleColor().A, B=Multiply(A=Clamp(None=@n7), B=Clamp(None=@n10)))
  VectorParameter() {parameterName=Color, defaultValue=<rgba 0.5711 0 0.7991 1>}
 ]
}
