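# Masked material for Niagara mesh particles, giving the unstable Peppercat cube a glowing, cracked energy look.
# SwapMaterials blends two attribute sets on a black base: one with a glass-crack normal, one without.
# Emissive chains tinted by particle color draw on distorted cloud noise, sphere-masked cracks and edge glow, bump-offset inner glow and Fresnel-driven big-crack noise.
# A noise-texture erosion mask driven by particle alpha provides dissolve.
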Material /PepperCatShared/Assets/FX/Materials/M_Peppercat_Cube_Unstable
graph {
 properties {
  bAutomaticallySetUsageInEditor=false
  bUseMaterialAttributes=true
  bUsedWithNiagaraMeshParticles=true
  blendMode=BLEND_Masked
  usedShadingModels=Default Lit
 }
 nodes [
  n1 = Constant3Vector()
  n2 = NamedRerouteUsage()
  n3 = SetMaterialAttributes!(MaterialAttributes, Base Color=@n1, Emissive Color=NamedRerouteUsage(), Normal=NamedRerouteUsage(), Opacity Mask=@n2) {attributeSetTypes [
    #1
    #2
    #3
    #4
   ]}
  n4 = SetMaterialAttributes(Base Color=@n1, Emissive Color=NamedRerouteUsage(), Opacity Mask=@n2) {attributeSetTypes [
    #1
    #2
    #3
    #4
   ]}
  BlendMaterialAttributes(A=@n3, B=@n4, Alpha=LinearInterpolate(Alpha=ScalarParameter() {parameterName=SwapMaterials}))
  Constant() {r=-0.1}
  n5 = ScalarParameter() {defaultValue=1, group=Erosion_Simple, parameterName=Erosion UV Scale U, sortPriority=2}
  n6 = ScalarParameter() {defaultValue=1, group=Erosion_Simple, parameterName=Erosion UV Scale V, sortPriority=3}
  n7 = TextureSampleParameter2D(UVs=Multiply(A=TextureCoordinate(), B=AppendVector(A=@n5, B=@n6))) {
   group=Erosion_Simple
   parameterName=ErosionTextureMask
   samplerType=SAMPLERTYPE_LinearColor
   sortPriority=1
   texture=/Game/Global/Textures/Default/Noise/Noise32x32
  }
  n8 = Subtract(A=Add(A=@n7.R), B=LinearInterpolate(Alpha=ParticleColor().A) {constA=1.2, constB=0})
  NamedRerouteDeclaration(None=Clamp(None=Floor(None=@n8))) {name=Simple Erosion, nodeColor=<rgba 0 1 0.9766 1>}
  n9 = VectorParameter() {parameterName=BigCrack_Color_B, defaultValue=<rgba 0.9547 0.6945 1 1>}
  n10 = VectorParameter() {parameterName=BigCrack_Color_A, defaultValue=<rgba 0.7166 0 1 1>}
  n11 = Panner(Coordinate=Multiply(A=TextureCoordinate() {coordinateIndex=1}, B=ScalarParameter() {defaultValue=2, parameterName=BigCrack_Noise_Tile})) {speedX=0.05, speedY=0.05}
  n12 = Multiply(A=TextureCoordinate() {coordinateIndex=1}, B=ScalarParameter() {defaultValue=2, parameterName=BigCrack_Noise_Distortion_Tile})
  n13 = AppendVector(A=Constant() {r=0.1}, B=Constant() {r=0.1})
  n14 = Multiply(
   A=TextureSample(UVs=Panner(Coordinate=@n12, Speed=@n13)) {texture=&AtmosphericCloudNoise01}.R
   B=ScalarParameter() {defaultValue=0.05, parameterName=BigCrack_Noise_Distortion_Intensity}
  )
  n15 = Power(Base=TextureSample(UVs=Add(A=@n11, B=@n14)) {texture=&AtmosphericCloudNoise01}.R) {constExponent=0.2}
  NamedRerouteDeclaration(None=Multiply(
    A=LinearInterpolate(A=@n9.RGB, B=@n10.RGB, Alpha=@n15)
    B=MaterialFunctionCall() {materialFunction=&MF_Fresnel_FX}.Fresnel
   )) {name=RR_BigCrack_Noise, nodeColor=<rgba 0.125 0 1 1>}
  n16 = NamedRerouteUsage()
  n17 = Add(A=NamedRerouteUsage(), B=NamedRerouteUsage())
  NamedRerouteDeclaration(None=Multiply(A=Add(A=@n16, B=@n17), B=ParticleColor().RGB)) {name=RR_BigCrack_Emissive, nodeColor=<rgba 0.3125 0 1 1>}
  n18 = Multiply(A=TextureCoordinate(), B=ScalarParameter() {defaultValue=3, parameterName=Crack_Tile})
  n19 = SphereMask(A=TextureCoordinate() {coordinateIndex=1}, B=Constant() {r=0.5}) {attenuationRadius=0.7, hardnessPercent=0}
  n20 = Multiply(
   A=Power(Base=@n19, Exp=ScalarParameter() {defaultValue=4, parameterName=Crack_Normal_Power})
   B=ScalarParameter() {defaultValue=2, parameterName=Crack_Normal_Mult}
  )
  NamedRerouteDeclaration(None=MaterialFunctionCall(
    Normal=TextureSample(UVs=@n18) {samplerType=SAMPLERTYPE_Normal, texture=/Game/Effects/Fort_Effects/Textures/Tile/T_Glass_Cracks_N}.RGB
    Flatness=Clamp(None=@n20)
   ) {materialFunction=&FlattenNormal}.Result) {name=RR_Normal, nodeColor=<rgba 0 1 0.1328 1>}
  n21 = Add(A=NamedRerouteUsage(), B=NamedRerouteUsage())
  n22 = Add(A=NamedRerouteUsage(), B=NamedRerouteUsage())
  NamedRerouteDeclaration(None=Multiply(A=Add(A=@n21, B=@n22), B=ParticleColor().RGB)) {name=RR_Emissive, nodeColor=<rgba 1 0 0.1875 1>}
  n23 = OneMinus(None=SphereMask(A=TextureCoordinate() {coordinateIndex=1}, B=Constant() {r=0.5}) {attenuationRadius=0.7, hardnessPercent=0})
  n24 = ScalarParameter() {defaultValue=6, parameterName=EdgeGlow_SphereMask_Power}
  NamedRerouteDeclaration(None=LinearInterpolate(
    A=Constant3Vector()
    B=VectorParameter() {parameterName=Base_Color_B, defaultValue=<rgba 0.9572 0.6481 1 1>}.RGB
    Alpha=Power(Base=@n23, Exp=@n24)
   )) {name=RR_EdgeGlow, nodeColor=<rgba 0 0.5078 1 1>}
  n25 = TextureSample(UVs=Multiply(A=TextureCoordinate(), B=ScalarParameter() {defaultValue=3, parameterName=Crack_Tile})) {texture=/Game/Effects/Fort_Effects/Textures/Tile/T_Glass_Cracks_M}
  n26 = ScalarParameter() {defaultValue=1.5, parameterName=Crack_Power}
  n27 = Clamp(None=Multiply(A=Power(Base=@n25.G, Exp=@n26), B=ScalarParameter() {defaultValue=1, parameterName=Crack_Alpha_Mult}))
  n28 = SphereMask(A=TextureCoordinate() {coordinateIndex=1}, B=Constant() {r=0.5}) {attenuationRadius=0.7, hardnessPercent=0}
  n29 = Clamp(None=Power(Base=OneMinus(None=@n28), Exp=ScalarParameter() {defaultValue=4, parameterName=Crack_Alpha_SphereMask_Power}))
  NamedRerouteDeclaration(None=Multiply(A=Multiply(A=@n27, B=@n29), B=VectorParameter() {parameterName=Crack_Color, defaultValue=<rgba 19.772 17.4922 20 1>}.RGB)) {name=RR_Crack, nodeColor=<rgba 0.8281 1 0 1>}
  n30 = TextureCoordinate() {coordinateIndex=1}
  n31 = ScalarParameter() {defaultValue=1, parameterName=Base_Noise_Tile}
  n32 = TextureCoordinate() {coordinateIndex=1}
  n33 = ScalarParameter() {defaultValue=1, parameterName=Base_Noise_Distortion_Tile}
  n34 = Constant() {r=0.1}
  n35 = Constant() {r=0.1}
  n36 = TextureSample(UVs=Panner(Coordinate=Multiply(A=@n32, B=@n33), Speed=AppendVector(A=@n34, B=@n35))) {texture=&AtmosphericCloudNoise01}
  n37 = ScalarParameter() {defaultValue=0.05, parameterName=Base_Noise_Distortion_Intensity}
  n38 = TextureSample(UVs=Add(A=Multiply(A=@n30, B=@n31), B=Multiply(A=@n36.R, B=@n37))) {texture=&AtmosphericCloudNoise01}
  NamedRerouteDeclaration(None=LinearInterpolate(
    A=VectorParameter() {parameterName=Base_Noise_Color_A, defaultValue=<rgba 0.7166 0 1 1>}.RGB
    B=VectorParameter() {parameterName=Base_Noise_Color_B, defaultValue=<rgba 0 0 0 1>}.RGB
    Alpha=Power(Base=@n38.R) {constExponent=0.2}
   )) {name=RR_Noise, nodeColor=<rgba 0.8516 0 1 1>}
  n39 = BumpOffset(Coordinate=TextureCoordinate() {coordinateIndex=1}, Height=Constant() {r=2}) {heightRatio=-0.1, referencePlane=0}
  NamedRerouteDeclaration(None=LinearInterpolate(
    A=Constant3Vector()
    B=Constant3Vector() {constant=<rgba 0.7166 0 1 1>}
    Alpha=TextureSample(UVs=@n39) {samplerType=SAMPLERTYPE_LinearColor, texture=/Game/ArtTools/Tech_Art/TexturePainter/Brushes/T_New_Soft_Flair_Alpha}.R
   )) {name=RR_InnerGlow, nodeColor=<rgba 0 1 0.5313 1>}
  OneMinus(None=@n19)
 ]
}
