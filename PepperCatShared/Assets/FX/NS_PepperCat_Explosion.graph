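# Niagara system for the PepperCat explosion effect, layering flashes, glow, shockwave, dark and radial spheres, ground smoke, sparks, lightning, runes, light rays, distortion, decal and cube debris.
# User parameters expose explosion delay, scale, hue and saturation, a direction array, and flags for terrain or wall hits and weapon triggering.
# Front-end flags with their own offset and scale suggest menu preview support.
# A GPU mesh cube-destruction emitter is disabled.
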NiagaraSystem /PepperCatShared/Assets/FX/NS_PepperCat_Explosion
dependencies=null
emitters {
 AirStreaks_LW {rendererClasses [], renderers []}
 Broken_Cube {rendererClasses [], renderers []}
 DarkSphere_LW {rendererClasses [], renderers []}
 Decal {rendererClasses [], renderers []}
 Destruction_Cube {rendererClasses [], renderers []}
 Flash {rendererClasses [], renderers []}
 Flash2 {rendererClasses [], renderers []}
 LW_Ground_Shockwave {rendererClasses [], renderers []}
 LW_Intro_Glow {rendererClasses [], renderers []}
 LW_Smoke_Ground {rendererClasses [], renderers []}
 Light {rendererClasses [], renderers []}
 LightRays {rendererClasses [], renderers []}
 OuterRadialSphere001 {rendererClasses [], renderers []}
 Runes {rendererClasses [], renderers []}
 Sparks {rendererClasses [], renderers []}
 SplitLightning {rendererClasses [], renderers []}
 SplitLightning001 {rendererClasses [], renderers []}
 Triggered_Distortion {rendererClasses [], renderers []}
 Triggered_Distortion001 {rendererClasses [], renderers []}
 Wave {rendererClasses [], renderers []}
}
summary {
 systemName=NS_PepperCat_Explosion
 emitters [
  {bEnabled=false, emitterName=Destruction_Cube001, simTarget=GPUComputeSim, rendererClasses [&NiagaraMeshRendererProperties]}
  {emitterName=Sparks, rendererClasses []}
  {emitterName=LW_Ground_Shockwave, rendererClasses []}
  {emitterName=LW_Intro_Glow, rendererClasses []}
  {emitterName=DarkSphere_LW, rendererClasses []}
  {emitterName=LW_Smoke_Ground, rendererClasses []}
  {emitterName=Light, rendererClasses []}
  {emitterName=Decal, rendererClasses []}
  {emitterName=AirStreaks_LW, rendererClasses []}
  {emitterName=Flash, rendererClasses []}
  {emitterName=Flash2, rendererClasses []}
  {emitterName=OuterRadialSphere001, rendererClasses []}
  {emitterName=SplitLightning, rendererClasses []}
  {emitterName=SplitLightning001, rendererClasses []}
  {emitterName=Wave, rendererClasses []}
  {emitterName=LightRays, rendererClasses []}
  {emitterName=Broken_Cube, rendererClasses []}
  {emitterName=Triggered_Distortion, rendererClasses []}
  {emitterName=Runes, rendererClasses []}
  {emitterName=Destruction_Cube, rendererClasses []}
  {emitterName=Triggered_Distortion001, rendererClasses []}
 ]
 userVariables [
  User.bIsFrontEnd {type=NiagaraBool}
  User.bIsFrontEndPreview {type=NiagaraBool}
  User.DirArray {
   type=NiagaraDataInterfaceArrayFloat3
   description=""
   defaultValue {
    struct=NiagaraExt_VariableValue_DataInterface
    value {
     dataInterface=&self:NiagaraDataInterfaceArrayFloat3_0
     dataInterfaceClass=&NiagaraDataInterfaceArrayFloat3
    }
   }
  }
  User.Explosion_Delay {type=NiagaraFloat, defaultValue=0, description=""}
  User.FEOffset {type=NiagaraFloat, defaultValue=10, description=""}
  User.FEScale {type=NiagaraFloat, defaultValue=0.6, description=""}
  User.FX_Scale_Global {type=NiagaraFloat, defaultValue=1, description=""}
  User.FXScale {type=NiagaraFloat, defaultValue=1, description=""}
  User.HitTerrainOrWall {type=NiagaraBool, defaultValue=true}
  User.HueShift {type=NiagaraFloat, defaultValue=0.275, description=""}
  User.IsFrontEnd {type=NiagaraBool, defaultValue=true}
  User.IsTriggeredbyWeapon {type=NiagaraBool, defaultValue=true}
  User.Saturation {type=NiagaraFloat, defaultValue=0.5, description=""}
 ]
}
