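# Niagara system for PepperCat's secondary muzzle flash, a short purple burst with smoke.
# Glow_HIGH and Lit_CINE drive sprite, shockwave and mesh glows through curve-animated attributes, with Lit_CINE adding a component renderer alongside light attributes.
# A probabilistic smoke puff erodes as it drifts, while Minimal003 pins flickering sprites to skeletal mesh sockets, oriented by sampled rotation.
# Several mesh, light, ribbon and spark emitters are present but disabled.
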NiagaraSystem /PepperCatShared/Assets/FX/NS_PepperCat_MuzzleFlash_Secondary
dependencies {
 usedDataInterfaces [
  &NiagaraDataInterfaceCurve
  &NiagaraDataInterfaceVectorCurve
  &NiagaraDataInterfaceMeshRendererInfo
  &NiagaraDataInterfaceColorCurve
  &NiagaraDataInterfaceArrayFloat3
  &NiagaraDataInterfaceSkeletalMesh
 ]
 usedDynamicInputs []
 usedModules [
  &self:SystemSpawnScript.SystemScriptSource.SystemScriptGraph.NiagaraNodeAssignment_2.SetVariables_#1
  &SystemState
  &self:SystemSpawnScript.SystemScriptSource.SystemScriptGraph.NiagaraNodeAssignment_0.SetVariables_#2
  &self:SystemSpawnScript.SystemScriptSource.SystemScriptGraph.NiagaraNodeAssignment_1.SetVariables_#3
 ]
}
emitters {
 CoreFlash001 {rendererClasses [], renderers []}
 FN_Weapons_Smoke_BETA {
  emitterUpdateScript [
   &EmitterState()
   &SpawnBurst_Instantaneous(Spawn Count=1, Spawn Probability=0.7, Use Spawn Probability=true)
  ]
  particleSpawnScript [
   /Niagara/Modules/Spawn/Initialization/InitializeParticle(
    Color=<rgba 0.1779 0.005 1 1>
    Lifetime=&UniformRangedFloat(Minimum=0.3, Maximum=0.4)
    Module.Write Color=true
    Module.Write Lifetime=true
    Module.Write Mass=true
    Module.Write Material Random=true
    Module.Write Position=false
    Module.Write Scale=false
    Module.Write SpriteRotation=true
    Module.Write SpriteSize=true
    Position=&SimulationPosition()
    Sprite Rotation=&NormalizedAngleToDegrees(NormalizedAngle=&UniformRangedFloat(Minimum=-15, Maximum=15))
    Sprite Size=&Vector2DFromFloat(Value=&UniformRangedFloat(Minimum=45, Maximum=60))
   )
   /SetVariables(Particles.InitSpriteRotRate=&UniformRangedFloat(Minimum=-3, Maximum=3)) {isSetParameters=true}
   &AddVelocityInCone(
    Cone Angle=15
    Module.Use Velocity Falloff On Cone Axis=false
    Velocity Strength=&UniformRangedFloat(Minimum=150, Maximum=400)
   )
   &InheritVelocity(Inherited Velocity Speed Limit=2000, Speed Limit=true, UseSpeedThreshold=false)
  ]
  particleUpdateScript [
   &ParticleState()
   &SpriteRotationRate(Rotation Rate=&Multiply_Float(
     A=$Particles.InitSpriteRotRate
     B=&FloatFromCurve(FloatCurve=DataInterface() {propertyValues {Curve {keys [<at -0.0046 1.0113 RCIM_Cubic>,<at 1.0011 0.09 RCIM_Cubic>]}}})
    ))
   &Drag(Drag=2, Use Linear Drag=true, Use Rotational Drag=true)
   &SolveForcesAndVelocity(Clamp Velocity=false, Limit Acceleration=false)
   &ScaleSpriteSize(Scale Factor=&Vector2DFromFloat(Value=&FloatFromCurve(FloatCurve=DataInterface() {propertyValues {Curve {keys [
          {arriveTangent=0.4478, interpMode=RCIM_Cubic, leaveTangent=0.4478, tangentMode=RCTM_User, time=0.0013, value=0.0367}
          {arriveTangent=0.9127, interpMode=RCIM_Cubic, leaveTangent=0.9127, tangentMode=RCTM_User, time=0.5367, value=0.8831}
          {arriveTangent=3.2289, interpMode=RCIM_Cubic, leaveTangent=3.2289, tangentMode=RCTM_User, time=1.0008, value=1.6042}
         ]}}})))
   /Game/Effects/Niagara/Modules/Materials/MaterialSmoke(
    Erode:NiagaraFloat=&FloatFromCurve(FloatCurve=DataInterface() {propertyValues {
       bOptimizeLUT=false
       bUseLUT=false
       Curve {keys [
         {arriveTangent=-0.0652, interpMode=RCIM_Cubic, leaveTangent=-0.0652, tangentMode=RCTM_User, time=0.11, value=0.97}
         {arriveTangent=-5.7797, interpMode=RCIM_Cubic, leaveTangent=-5.7797, tangentMode=RCTM_User, time=0.39, value=0.66}
         {arriveTangent=-0.342, interpMode=RCIM_Cubic, leaveTangent=-0.342, tangentMode=RCTM_User, time=0.9954, value=-0.74}
        ]}
      }})
    Shading:NiagaraFloat=&FloatFromCurve(FloatCurve=DataInterface() {propertyValues {
       bOptimizeLUT=false
       bUseLUT=false
       Curve {keys [
         <at 0.07 0.5 RCIM_Cubic>
         {arriveTangent=1.2205, interpMode=RCIM_Cubic, leaveTangent=1.2205, time=0.17, value=1.02}
         {interpMode=RCIM_Cubic, tangentMode=RCTM_User, time=0.3895, value=0.89}
         <at 0.8 0 RCIM_Cubic>
        ]}
      }})
   )
   &CameraOffset(Camera Offset Amount=20)
   &DynamicMaterialParameters(
    Index 0 Param 1=&Add_Float(
     A=&RandomRangeFloat(Minimum=0, Maximum=1)
     B=&FloatFromCurve(FloatCurve=DataInterface() {propertyValues {Curve {keys [<at 0 1>,<at 1 0>]}}})
    )
    Index 0 Param 2=&FloatFromCurve(FloatCurve=DataInterface() {propertyValues {Curve {keys [
         {arriveTangent=0.8564, interpMode=RCIM_Cubic, leaveTangent=0.8564, tangentMode=RCTM_User, time=0.0046, value=-0.9753}
         {arriveTangent=4.2876, interpMode=RCIM_Cubic, leaveTangent=5.5268, tangentMode=RCTM_User, time=1, value=1}
        ]}}})
    Index 0 Param 4=&FloatFromCurve(FloatCurve=DataInterface() {propertyValues {Curve {keys [{arriveTangent=-3.2497, interpMode=RCIM_Cubic, leaveTangent=-3.2497, tangentMode=RCTM_User, time=0, value=1}, <at 1 0 RCIM_Cubic>]}}})
    Module.Param0WriteEnabled=true
    Module.Param1WriteEnabled=true
    Module.Param2WriteEnabled=true
    Module.Param3WriteEnabled=true
   )
   &ScaleColor(
    Scale Alpha=&FloatFromCurve(
     FloatCurve=DataInterface() {propertyValues {Curve {keys [<at -0.003 -0.006>,<at 0.1806 0.5257>,<at 0.4581 0.2965>,<at 1 0>]}}}
     Scale Curve=0.6
    )
    Scale RGB=&VectorFromFloat(Value=&FloatFromCurve(FloatCurve=DataInterface() {propertyValues {Curve {keys [{interpMode=RCIM_Cubic, tangentMode=RCTM_Break, time=0, value=1}, {interpMode=RCIM_Cubic, tangentMode=RCTM_Break, time=1, value=0}]}}}))
    ScaleA=true
    ScaleRGB=true
   )
  ]
  rendererClasses [&NiagaraSpriteRendererProperties]
  renderers [{rendererClass=&NiagaraSpriteRendererProperties}]
 }
 FrontMuzzle002 {
  bEnabled=false
  emitterUpdateScript [
   &EmitterState()
   &SpawnBurst_Instantaneous(Use Spawn Probability=false, UseLoopCountLimit=false)
  ]
  particleSpawnScript [
   /Niagara/Modules/Spawn/Initialization/V2/InitializeParticle(
    Color=<rgba 4.2995 0 6 1>
    Lifetime=0.2
    Mesh Scale=&Multiply_VectorByFloat(Vector=<6 2.5 2.5>, Float=0.25)
    Mesh Scale Mode=::NewEnumerator4
    Position Offset=$User.OriginOffset
    UsePositionOffset=true
    Write Material Random=true
   )
   &InitialMeshOrientation(
    Mesh Orientation Mode=::NewEnumerator2
    Rotation=&RandomRangeVector(Minimum=<0 0 0>, Maximum=<1 0 0>)
    Use Rotation Vector=true
   )
  ]
  particleUpdateScript [
   &ParticleState()
   &ScaleMeshSize(Scale Factor=&VectorFromCurve(VectorCurve=DataInterface() {propertyValues {
       XCurve {keys [
         {arriveTangent=5.8858, interpMode=RCIM_Cubic, leaveTangent=5.8858, tangentMode=RCTM_User, time=0, value=0}
         {interpMode=RCIM_Cubic, tangentMode=RCTM_User, time=0.5, value=1}
         <at 1 0.5>
        ]}
       YCurve {keys [
         {arriveTangent=5.0355, interpMode=RCIM_Cubic, leaveTangent=5.0355, tangentMode=RCTM_User, time=0, value=0.1}
         <at 0.5 1 RCIM_Cubic>
         <at 1 0.1>
        ]}
       ZCurve {keys [
         {arriveTangent=5.0355, interpMode=RCIM_Cubic, leaveTangent=5.0355, tangentMode=RCTM_User, time=0, value=0.1}
         <at 0.5 1 RCIM_Cubic>
         <at 1 0.1>
        ]}
      }}))
   &ScaleColor(
    Scale Alpha=&FloatFromCurve(FloatCurve=DataInterface() {propertyValues {Curve {keys [
         {arriveTangent=9.0725, interpMode=RCIM_Cubic, leaveTangent=9.0725, tangentMode=RCTM_User, time=0, value=0}
         {arriveTangent=-2.5249, interpMode=RCIM_Cubic, leaveTangent=-2.5249, tangentMode=RCTM_User, time=0.1408, value=0.9847}
         {arriveTangent=-0.0405, interpMode=RCIM_Cubic, leaveTangent=-0.0405, tangentMode=RCTM_User, time=1, value=0}
        ]}}})
    ScaleA=true
    ScaleRGB=true
   )
  ]
  rendererClasses [&NiagaraMeshRendererProperties]
  renderers [{rendererClass=&NiagaraMeshRendererProperties, rendererIndex=0}]
 }
 FrontMuzzle003 {
  bEnabled=false
  emitterUpdateScript [
   &EmitterState()
   &SpawnBurst_Instantaneous(Use Spawn Probability=false, UseLoopCountLimit=false)
   &ShapeLocation(MaskSpawnGroup=false, Override Seed=false, Shape Primitive=::NewEnumerator5, UseOwnerScale=true)
  ]
  particleSpawnScript [
   &InitializeParticle(
    Color=<rgba 0 9.8552 10 1>
    Lifetime=0.2
    Mesh Scale Max=<2 6 6>
    Mesh Scale Min=<2 5.5 5.5>
    Mesh Scale Mode=::NewEnumerator2
    Position Offset=<10 0 0>
    UsePositionOffset=true
    Write Material Random=true
   )
   &InitialMeshOrientation(
    Mesh Orientation Mode=::NewEnumerator2
    Rotation=&RandomRangeVector(Minimum=<0 0 0>, Maximum=<1 0 0>)
    Use Rotation Vector=true
   )
   &AddVelocity(Cone Angle=60, Module.UseConeFalloff=false, Module.UseVelDistribution=false, Override Seed=false, Velocity Mode=::NewEnumerator2)
   &ApplyInitialForces()
  ]
  particleUpdateScript [
   &ParticleState()
   &ScaleMeshSize(Scale Factor=&VectorFromCurve(VectorCurve=DataInterface() {propertyValues {
       XCurve {keys [{arriveTangent=1.6669, interpMode=RCIM_Cubic, leaveTangent=1.6669, tangentMode=RCTM_User, time=0, value=0}, <at 0.75 1>]}
       YCurve {keys [{arriveTangent=4.9773, interpMode=RCIM_Cubic, leaveTangent=4.9773, tangentMode=RCTM_User, time=0, value=0}, <at 0.5 1 RCIM_Cubic>]}
       ZCurve {keys [
         {arriveTangent=3.5762, interpMode=RCIM_Cubic, leaveTangent=3.5762, tangentMode=RCTM_User, time=0, value=0}
         {arriveTangent=-0.0405, interpMode=RCIM_Cubic, leaveTangent=-0.0405, tangentMode=RCTM_User, time=0.5, value=1.05}
         <at 1 1>
        ]}
      }}))
   &ScaleColor(
    Scale Alpha=&FloatFromCurve(FloatCurve=DataInterface() {propertyValues {Curve {keys [
         {arriveTangent=9.0725, interpMode=RCIM_Cubic, leaveTangent=9.0725, tangentMode=RCTM_User, time=0, value=0}
         {arriveTangent=-2.5249, interpMode=RCIM_Cubic, leaveTangent=-2.5249, tangentMode=RCTM_User, time=0.1408, value=0.9847}
         {arriveTangent=-0.0405, interpMode=RCIM_Cubic, leaveTangent=-0.0405, tangentMode=RCTM_User, time=1, value=0}
        ]}}})
    ScaleA=true
    ScaleRGB=true
   )
  ]
  rendererClasses [&NiagaraMeshRendererProperties]
  renderers [{rendererClass=&NiagaraMeshRendererProperties, rendererIndex=0}]
 }
 Glow_HIGH {
  emitterUpdateScript [
   &EmitterState()
   &SpawnBurst_Instantaneous(Use Spawn Probability=false, UseLoopCountLimit=false)
  ]
  particleSpawnScript [
   &InitializeParticle(
    Color=<rgba 0.1264 0 0.3 1>
    Lifetime=0.1
    Mass Mode=::NewEnumerator3
    Mesh Scale Mode=::NewEnumerator1
    Mesh Uniform Scale Max=0.45
    Mesh Uniform Scale Min=0.4
    Position Offset=<3 0 0>
    Sprite Size Mode=::NewEnumerator3
    Uniform Sprite Size=90
    UsePositionOffset=true
    Write Material Random=true
   )
   &ShapeLocation(MaskSpawnGroup=false, Non Uniform Scale=<0.5 1 1>, Override Seed=false, Sphere Radius=0, UseOwnerScale=true)
   &InitialMeshOrientation(Orientation Axis=<0 0 1>, Orientation Vector=<0 0 1>, Rotation=<0 0.25 0>, Use Orientation Vector=true, Use Rotation Vector=true)
   &InitialMeshOrientation(
    Orientation Axis=<0 0 0>
    Orientation Vector=<0 0 0>
    Rotation=&RandomRangeVector(Minimum=<0 0 0>, Maximum=<1 0 0>)
    Use Orientation Vector=false
    Use Rotation Vector=true
   ) {module=InitialMeshOrientation001}
   /SetVariables(Particles.LightPositionNew:Vector3f=&Add_Vector(
     A=Linked() {linkedVariable {name=Particles.Position, type=NiagaraPosition}}
     B=/Niagara/DynamicInputs/Transforms/TransformPosition(Position=<25 0 0>, Source Space=::Local, Destination Space=::Simulation)
    )) {isSetParameters=true, module=SetVariables_#4}
  ]
  particleUpdateScript [
   &ParticleState()
   &ScaleColor(
    Scale Alpha=&FloatFromCurve(
     FloatCurve=DataInterface() {propertyValues {Curve {
        defaultValue=1
        keys [
         {tangentMode=RCTM_User, time=0, value=1}
         {arriveTangent=-0.0405, interpMode=RCIM_Cubic, leaveTangent=-0.0405, tangentMode=RCTM_User, time=0.6, value=1}
         {arriveTangent=-2.5463, interpMode=RCIM_Cubic, leaveTangent=-2.5463, tangentMode=RCTM_User, time=1, value=0}
        ]
       }}}
     Scale Curve=0.8
    )
    Scale RGB=&MakeVectorFromLinearColorRGB(Color=&ColorFromCurve(ColorCurve=DataInterface() {propertyValues {
        AlphaCurve {keys [<at 0 1 RCIM_Cubic>]}
        BlueCurve {keys [<at 0.3341 20 RCIM_Cubic>,<at 0.5698 8>,<at 0.729 1>]}
        GreenCurve {keys [<at 0.3341 20 RCIM_Cubic>,<at 0.5698 8>,<at 0.729 1>]}
        RedCurve {keys [<at 0.3341 20 RCIM_Cubic>,<at 0.5698 8>,<at 0.729 1>]}
       }}))
    ScaleA=true
    ScaleRGB=true
   ) {module=ScaleColor001}
   &ScaleSpriteSize(
    Scale Sprite Size Mode=::NewEnumerator0
    Uniform Scale Factor=&FloatFromCurve(FloatCurve=DataInterface() {propertyValues {Curve {keys [
         {arriveTangent=2.278, interpMode=RCIM_Cubic, leaveTangent=2.278, tangentMode=RCTM_User, time=0, value=0}
         {interpMode=RCIM_Cubic, tangentMode=RCTM_User, time=1, value=1}
        ]}}})
   )
   /SetVariables(Particles.FlatGlowColor:LinearColor=&MultiplyLinearColor(Linear Color=$Particles.Color, Scale Factor=<1 1 1 0.2>)) {isSetParameters=true, module=SetVariables_#5}
   /SetVariables(Particles.FlatGlowSize:Vector2f=<100 4>) {isSetParameters=true, module=SetVariables_#6}
   /SetVariables(Particles.ShockColor:LinearColor=&MultiplyLinearColor(
     Linear Color=&ColorFromCurve(ColorCurve=DataInterface() {propertyValues {
        AlphaCurve {keys [<at 0 1 RCIM_Cubic>]}
        BlueCurve {keys [<at 0 20 RCIM_Cubic>,<at 0.6348 8>]}
        GreenCurve {keys [<at 0 20 RCIM_Cubic>,<at 0.6348 8>]}
        RedCurve {keys [<at 0 20 RCIM_Cubic>,<at 0.6348 8>]}
       }})
     Scale Factor=&MakeFloat4FromVectorAndFloat(
      "Vector (RGB)"=&MakeVectorFromLinearColorRGB(Color=<rgba 0.2115 0 0.6042 1>)
      "Float (Alpha)"=&FloatFromCurve(
       FloatCurve=DataInterface() {propertyValues {Curve {
          defaultValue=1
          keys [
           {tangentMode=RCTM_User, time=0, value=1}
           {arriveTangent=-0.0405, interpMode=RCIM_Cubic, leaveTangent=-0.0405, tangentMode=RCTM_User, time=0.7009, value=1}
           {arriveTangent=-8.0089, interpMode=RCIM_Cubic, leaveTangent=-8.0089, tangentMode=RCTM_User, time=1, value=0}
          ]
         }}}
       Scale Curve=0.5
      )
     )
    )) {isSetParameters=true, module=SetVariables_#7}
   &ScaleMeshSize(Scale Factor=&VectorFromFloat(Value=&FloatFromCurve(FloatCurve=DataInterface() {propertyValues {Curve {keys [
          {arriveTangent=1.8207, interpMode=RCIM_Cubic, leaveTangent=1.8207, tangentMode=RCTM_User, time=0, value=0.4}
          {interpMode=RCIM_Cubic, tangentMode=RCTM_User, time=1, value=1.2}
         ]}}})))
   /SetVariables(Particles.WaveDynMat:Vector4f=&MakeVector4(
     X=&FloatFromCurve(
      FloatCurve=DataInterface() {propertyValues {Curve {keys [{arriveTangent=3.726, interpMode=RCIM_Cubic, leaveTangent=3.726, tangentMode=RCTM_User, time=0, value=-0.3}, <at 1 1>]}}}
      Scale Curve=0.6
     )
     Y=&FloatFromCurve(FloatCurve=DataInterface() {propertyValues {Curve {keys [{arriveTangent=-0.5818, interpMode=RCIM_Cubic, leaveTangent=-0.5818, tangentMode=RCTM_User, time=0, value=1}, <at 1 0.8>]}}})
     Z=&FloatFromCurve(FloatCurve=DataInterface() {propertyValues {Curve {keys [<at 0 0>,<at 0.15 1>,<at 1 1.2>]}}}, Scale Curve=10)
     W=&FloatFromCurve(FloatCurve=DataInterface() {propertyValues {Curve {keys [<at 0 0>,<at 1 1>]}}}, Scale Curve=0)
    )) {isSetParameters=true, module=SetVariables_#8}
   &CameraOffset(Camera Offset Amount=8)
   &FN_Light_Attributes(
    Color=<rgba 0.6347 0 1 1>
    Create Color Attribute=true
    Create Exponent Attribute=true
    Create Position Attribute=true
    Enable Particle Light Enable=true
    Exponent=4
    Light Enabled=true
    Scale Color=3
   )
   &FN_Light_Attributes(
    Color=<rgba 0.6347 0 1 1>
    Create Color Attribute=true
    Create Radius Attribute=true
    Enable Particle Light Enable=true
    Light Enabled=true
    Radius=200
   ) {module=FN_Light_Attributes001}
  ]
  rendererClasses [&NiagaraSpriteRendererProperties, &NiagaraMeshRendererProperties]
  renderers [
   {rendererClass=&NiagaraSpriteRendererProperties}
   {rendererClass=&NiagaraSpriteRendererProperties, rendererIndex=1}
   {rendererClass=&NiagaraMeshRendererProperties, rendererIndex=2}
  ]
 }
 Glow_LOW {rendererClasses [], renderers []}
 Light {
  bEnabled=false
  emitterUpdateScript [
   &EmitterState()
   &SpawnBurst_Instantaneous(Spawn Count=1, Use Spawn Probability=false)
  ]
  particleSpawnScript [
   &InitializeParticle(
    Color=<rgba 4.2995 0 6 1>
    Lifetime=0.25
    Position Offset=<20 0 0>
    Position Offset Coordinate Space=::Local
    Sprite Size Mode=::NewEnumerator3
    Uniform Sprite Size=150
    UsePositionOffset=true
    Write Material Random=true
   )
   &DynamicMaterialParameters(
    Index 0 Param 1=0.2
    Index 0 Param 2=1
    Module.Param0WriteEnabled=true
    Module.Param1WriteEnabled=true
    Module.Param2WriteEnabled=true
    Module.Param3WriteEnabled=true
   )
  ]
  particleUpdateScript [
   &ParticleState()
   &ScaleSpriteSize(
    Scale Sprite Size Mode=::NewEnumerator0
    Uniform Scale Factor=&FloatFromCurve(FloatCurve=DataInterface() {propertyValues {Curve {keys [
         {arriveTangent=3.7039, interpMode=RCIM_Cubic, leaveTangent=3.7039, tangentMode=RCTM_User, time=0, value=0.33}
         <at 0.2 1>
         {arriveTangent=-0.0625, interpMode=RCIM_Cubic, leaveTangent=-0.0625, time=0.5, value=0.9847}
         {arriveTangent=-0.0405, interpMode=RCIM_Cubic, leaveTangent=-0.0405, tangentMode=RCTM_User, time=1, value=0.33}
        ]}}})
   )
   &ScaleColor(
    Scale Alpha=&FloatFromCurve(FloatCurve=DataInterface() {propertyValues {Curve {keys [
         {tangentMode=RCTM_User, time=0, value=0}
         <at 0.1 1>
         {arriveTangent=-0.0405, interpMode=RCIM_Cubic, leaveTangent=-0.0405, tangentMode=RCTM_User, time=0.5, value=1}
         {interpMode=RCIM_Cubic, leaveTangent=-8.0089, time=1, value=0}
        ]}}})
    Scale RGB=&VectorFromFloat(Value=&FloatFromCurve(FloatCurve=DataInterface() {propertyValues {Curve {keys [<at 0 1>,<at 1 0>]}}}))
    ScaleA=true
    ScaleRGB=true
   )
   &FN_Light_v2(
    Light Exponent=2
    Light Radius=100
    LightEnabled=true
    Scale Color:NiagaraFloat=&FloatFromCurve(FloatCurve=DataInterface() {propertyValues {Curve {
        defaultValue=1
        keys [
         {arriveTangent=9.0725, interpMode=RCIM_Cubic, leaveTangent=9.0725, tangentMode=RCTM_User, time=0, value=0}
         {arriveTangent=-2.5249, interpMode=RCIM_Cubic, leaveTangent=-2.5249, tangentMode=RCTM_User, time=0.1408, value=0.9847}
         {arriveTangent=-0.0405, interpMode=RCIM_Cubic, leaveTangent=-0.0405, tangentMode=RCTM_User, time=1, value=0}
        ]
       }}})
   )
  ]
  rendererClasses [&NiagaraLightRendererProperties]
  renderers [{rendererClass=&NiagaraLightRendererProperties, rendererIndex=0}]
 }
 Lightning {rendererClasses [], renderers []}
 Lit {rendererClasses [], renderers []}
 Lit_CINE {
  emitterUpdateScript [
   &EmitterState()
   &SpawnBurst_Instantaneous(Use Spawn Probability=false, UseLoopCountLimit=false)
  ]
  particleSpawnScript [
   &InitializeParticle(
    Color=<rgba 0.1264 0 0.3 1>
    Lifetime=0.1
    Mass Mode=::NewEnumerator3
    Mesh Scale Mode=::NewEnumerator1
    Mesh Uniform Scale Max=0.45
    Mesh Uniform Scale Min=0.4
    Position Offset=<3 0 0>
    Sprite Size Mode=::NewEnumerator3
    Uniform Sprite Size=90
    UsePositionOffset=true
    Write Material Random=true
   )
   &ShapeLocation(MaskSpawnGroup=false, Non Uniform Scale=<0.5 1 1>, Override Seed=false, Sphere Radius=0, UseOwnerScale=true)
   &InitialMeshOrientation(Orientation Axis=<0 0 1>, Orientation Vector=<0 0 1>, Rotation=<0 0.25 0>, Use Orientation Vector=true, Use Rotation Vector=true)
   &InitialMeshOrientation(
    Orientation Axis=<0 0 0>
    Orientation Vector=<0 0 0>
    Rotation=&RandomRangeVector(Minimum=<0 0 0>, Maximum=<1 0 0>)
    Use Orientation Vector=false
    Use Rotation Vector=true
   ) {module=InitialMeshOrientation001}
   /SetVariables(Particles.LightPositionNew:Vector3f=&Add_Vector(
     A=Linked() {linkedVariable {name=Particles.Position, type=NiagaraPosition}}
     B=&TransformPosition(Position=<25 0 0>, Source Space=::Local, Destination Space=::Simulation)
    )) {isSetParameters=true, module=SetVariables_#4}
  ]
  particleUpdateScript [
   &ParticleState()
   &ScaleColor(
    Scale Alpha=&FloatFromCurve(
     FloatCurve=DataInterface() {propertyValues {Curve {
        defaultValue=1
        keys [
         {tangentMode=RCTM_User, time=0, value=1}
         {arriveTangent=-0.0405, interpMode=RCIM_Cubic, leaveTangent=-0.0405, tangentMode=RCTM_User, time=0.6, value=1}
         {arriveTangent=-2.5463, interpMode=RCIM_Cubic, leaveTangent=-2.5463, tangentMode=RCTM_User, time=1, value=0}
        ]
       }}}
     Scale Curve=0.8
    )
    Scale RGB=&MakeVectorFromLinearColorRGB(Color=&ColorFromCurve(ColorCurve=DataInterface() {propertyValues {
        AlphaCurve {keys [<at 0 1 RCIM_Cubic>]}
        BlueCurve {keys [<at 0.3341 20 RCIM_Cubic>,<at 0.5698 8>,<at 0.729 1>]}
        GreenCurve {keys [<at 0.3341 20 RCIM_Cubic>,<at 0.5698 8>,<at 0.729 1>]}
        RedCurve {keys [<at 0.3341 20 RCIM_Cubic>,<at 0.5698 8>,<at 0.729 1>]}
       }}))
    ScaleA=true
    ScaleRGB=true
   ) {module=ScaleColor001}
   &ScaleSpriteSize(
    Scale Sprite Size Mode=::NewEnumerator0
    Uniform Scale Factor=&FloatFromCurve(FloatCurve=DataInterface() {propertyValues {Curve {keys [
         {arriveTangent=2.278, interpMode=RCIM_Cubic, leaveTangent=2.278, tangentMode=RCTM_User, time=0, value=0}
         {interpMode=RCIM_Cubic, tangentMode=RCTM_User, time=1, value=1}
        ]}}})
   )
   /SetVariables(Particles.FlatGlowColor:LinearColor=&MultiplyLinearColor(Linear Color=$Particles.Color, Scale Factor=<1 1 1 0.2>)) {isSetParameters=true, module=SetVariables_#5}
   /SetVariables(Particles.FlatGlowSize:Vector2f=<100 4>) {isSetParameters=true, module=SetVariables_#6}
   /SetVariables(Particles.ShockColor:LinearColor=&MultiplyLinearColor(
     Linear Color=&ColorFromCurve(ColorCurve=DataInterface() {propertyValues {
        AlphaCurve {keys [<at 0 1 RCIM_Cubic>]}
        BlueCurve {keys [<at 0 20 RCIM_Cubic>,<at 0.6348 8>]}
        GreenCurve {keys [<at 0 20 RCIM_Cubic>,<at 0.6348 8>]}
        RedCurve {keys [<at 0 20 RCIM_Cubic>,<at 0.6348 8>]}
       }})
     Scale Factor=&MakeFloat4FromVectorAndFloat(
      "Vector (RGB)"=&MakeVectorFromLinearColorRGB(Color=<rgba 0.2115 0 0.6042 1>)
      "Float (Alpha)"=&FloatFromCurve(
       FloatCurve=DataInterface() {propertyValues {Curve {
          defaultValue=1
          keys [
           {tangentMode=RCTM_User, time=0, value=1}
           {arriveTangent=-0.0405, interpMode=RCIM_Cubic, leaveTangent=-0.0405, tangentMode=RCTM_User, time=0.7009, value=1}
           {arriveTangent=-8.0089, interpMode=RCIM_Cubic, leaveTangent=-8.0089, tangentMode=RCTM_User, time=1, value=0}
          ]
         }}}
       Scale Curve=0.5
      )
     )
    )) {isSetParameters=true, module=SetVariables_#7}
   &ScaleMeshSize(Scale Factor=&VectorFromFloat(Value=&FloatFromCurve(FloatCurve=DataInterface() {propertyValues {Curve {keys [
          {arriveTangent=1.8207, interpMode=RCIM_Cubic, leaveTangent=1.8207, tangentMode=RCTM_User, time=0, value=0.4}
          {interpMode=RCIM_Cubic, tangentMode=RCTM_User, time=1, value=1.2}
         ]}}})))
   /SetVariables(Particles.WaveDynMat:Vector4f=&MakeVector4(
     X=&FloatFromCurve(
      FloatCurve=DataInterface() {propertyValues {Curve {keys [{arriveTangent=3.726, interpMode=RCIM_Cubic, leaveTangent=3.726, tangentMode=RCTM_User, time=0, value=-0.3}, <at 1 1>]}}}
      Scale Curve=0.6
     )
     Y=&FloatFromCurve(FloatCurve=DataInterface() {propertyValues {Curve {keys [{arriveTangent=-0.5818, interpMode=RCIM_Cubic, leaveTangent=-0.5818, tangentMode=RCTM_User, time=0, value=1}, <at 1 0.8>]}}})
     Z=&FloatFromCurve(FloatCurve=DataInterface() {propertyValues {Curve {keys [<at 0 0>,<at 0.15 1>,<at 1 1.2>]}}}, Scale Curve=10)
     W=&FloatFromCurve(FloatCurve=DataInterface() {propertyValues {Curve {keys [<at 0 0>,<at 1 1>]}}}, Scale Curve=0)
    )) {isSetParameters=true, module=SetVariables_#8}
   &CameraOffset(Camera Offset Amount=8)
   &FN_Light_Attributes(
    Color=<rgba 0.6347 0 1 1>
    Create Color Attribute=true
    Create Exponent Attribute=true
    Create Position Attribute=true
    Enable Particle Light Enable=true
    Exponent=4
    Light Enabled=true
    Scale Color=3
   )
   &FN_Light_Attributes(
    Color=<rgba 0.6347 0 1 1>
    Create Color Attribute=true
    Create Radius Attribute=true
    Enable Particle Light Enable=true
    Light Enabled=true
    Radius=200
   ) {module=FN_Light_Attributes001}
  ]
  rendererClasses [
   &NiagaraSpriteRendererProperties
   &NiagaraMeshRendererProperties
   &NiagaraComponentRendererProperties
  ]
  renderers [
   {rendererClass=&NiagaraSpriteRendererProperties}
   {rendererClass=&NiagaraSpriteRendererProperties, rendererIndex=1}
   {rendererClass=&NiagaraMeshRendererProperties, rendererIndex=2}
   {rendererClass=&NiagaraComponentRendererProperties, rendererIndex=3}
  ]
 }
 Minimal {
  bEnabled=false
  emitterUpdateScript [
   &EmitterState(Inactive Response=::NewEnumerator1, Life Cycle Mode=::NewEnumerator1, Loop Behavior=::NewEnumerator1, Loop Duration=0.5, UseLoopDelay=false)
   &SpawnRate(SpawnRate=4, Use Spawn Probability=false) {enabled=false}
   &SpawnBurst_Instantaneous(Spawn Count=4, Use Spawn Probability=false, UseLoopCountLimit=false)
  ]
  particleSpawnScript [
   &InitializeParticle(
    Color=<rgba 4.8786 2.4 6 1>
    Lifetime=0.5
    Ribbon Width=2
    Ribbon Width Mode=::NewEnumerator1
    Sprite Size Mode=::NewEnumerator3
    Uniform Sprite Size=5
    UsePositionOffset=false
    Write Material Random=true
   )
   &PartitionParticles(
    Particles Per Partition=2
    Partition Mode:ENiagara_PartitionMode_v2=::NewEnumerator1
    Write Partition Index to Ribbon Link Order:ENiagara_AttributeSamplingApplyOutput=::NewEnumerator0
    Write Partition to Ribbon ID=::NewEnumerator0
   )
   &PartitionParticles(Number Of Partitions=4, Particles Per Partition=1, Partition Mode:ENiagara_PartitionMode_v2=::NewEnumerator1) {enabled=false, module=PartitionParticles001}
  ]
  particleUpdateScript [
   &ParticleState()
   &SkeletalMeshLocation(
    Bone / Socket Index=&ReturnExecIndex()
    MaskSpawnGroup=false
    Mesh Sampling Type=::NewEnumerator5
    Sampled Position Offset=&MakeVector(
     X=&Lerp_Float(
      A=0
      B=30
      Alpha=&FloatFromCurve(FloatCurve=DataInterface() {propertyValues {Curve {defaultValue=1, keys [{tangentMode=RCTM_User, time=0, value=0}, {tangentMode=RCTM_User, time=1, value=1}]}}})
     )
     Y=0
     Z=0
    )
    Skeletal Mesh=$System.Skeletal Mesh
    Socket Sampling Mode=::NewEnumerator5
    Use Sampled Mesh Attribute as Ribbon ID=false
   )
   &CameraOffset(Camera Offset Amount=5)
   &ScaleRibbonWidth()
   /SetVariables(Particles.RibbonLinkOrder:NiagaraFloat=$Particles.PartitionParticles.Normalized Index In Partition) {enabled=false, isSetParameters=true, module=SetVariables_#9}
   /SetVariables(Particles.NewRibbonOrder:NiagaraFloat=$Particles.PartitionParticles.Normalized Index In Partition) {isSetParameters=true, module=SetVariables_#10}
   &DynamicMaterialParameters(Module.Index0Float0Write=true, Module.Index0Float1Write=true, Module.Index0Float2Write=true, Module.Index0Float3Write=true)
   &ScaleColor(
    Scale RGB=&VectorFromFloat(Value=&FloatFromCurve(
      FloatCurve=DataInterface() {propertyValues {Curve {
         defaultValue=1
         keys [
          {arriveTangent=9.0725, interpMode=RCIM_Cubic, leaveTangent=9.0725, tangentMode=RCTM_User, time=0, value=0}
          {arriveTangent=-2.5249, interpMode=RCIM_Cubic, leaveTangent=-2.5249, tangentMode=RCTM_User, time=0.1408, value=0.9847}
          {arriveTangent=-0.0405, interpMode=RCIM_Cubic, leaveTangent=-0.0405, tangentMode=RCTM_User, time=1, value=0}
         ]
        }}}
      Scale Curve=10
     ))
    ScaleA=true
    ScaleRGB=true
   )
  ]
  rendererClasses [&NiagaraSpriteRendererProperties, &NiagaraRibbonRendererProperties]
  renderers [
   {rendererClass=&NiagaraSpriteRendererProperties}
   {rendererClass=&NiagaraRibbonRendererProperties, rendererIndex=1}
   {rendererClass=&NiagaraRibbonRendererProperties, rendererIndex=2}
  ]
 }
 Minimal001 {
  bEnabled=false
  emitterUpdateScript [
   &EmitterState(
    Inactive Response=::NewEnumerator1
    Life Cycle Mode=::NewEnumerator1
    Loop Behavior=::NewEnumerator2
    Loop Count=3
    Loop Duration=0.2
    UseLoopDelay=false
   )
   &SpawnRate(SpawnRate=20, Use Spawn Probability=false) {enabled=false}
   &SpawnBurst_Instantaneous(Spawn Count=2, Use Spawn Probability=false, UseLoopCountLimit=false)
  ]
  particleSpawnScript [
   &InitializeParticle(
    Color=<rgba 4.8786 2.4 6 1>
    Lifetime=0.5
    Ribbon Width=2
    Ribbon Width Mode=::NewEnumerator1
    Sprite Size Mode=::NewEnumerator3
    Uniform Sprite Size=5
    UsePositionOffset=false
    Write Material Random=true
   )
   &PartitionParticles(
    Particles Per Partition=2
    Partition Count Mode:ENiagara_InfiniteLoopDuration=::NewEnumerator1
    Partition Mode:ENiagara_PartitionMode_v2=::NewEnumerator1
    Write Partition Index to Ribbon Link Order:ENiagara_AttributeSamplingApplyOutput=::NewEnumerator0
    Write Partition to Ribbon ID=::NewEnumerator0
   )
   &PartitionParticles(
    Number Of Partitions=4
    Particles Per Partition=1
    Partition Mode:ENiagara_PartitionMode_v2=::NewEnumerator1
    Write Partition Index to Ribbon Link Order:ENiagara_AttributeSamplingApplyOutput=::NewEnumerator0
    Write Partition to Ribbon ID=::NewEnumerator0
   ) {enabled=false, module=PartitionParticles001}
  ]
  particleUpdateScript [
   &ParticleState()
   &SkeletalMeshLocation(
    Bone / Socket Index=&ReturnExecIndex()
    MaskSpawnGroup=false
    Mesh Sampling Type=::NewEnumerator5
    Sampled Position Offset=&Add_Vector(
     A=&MakeVector(
      X=&Lerp_Float(
       A=0
       B=0
       Alpha=&FloatFromCurve(FloatCurve=DataInterface() {propertyValues {Curve {defaultValue=1, keys [{tangentMode=RCTM_User, time=0, value=0}, {tangentMode=RCTM_User, time=1, value=1}]}}})
      )
      Y=0
      Z=0
     )
     B=&RandomRangeVector(Minimum=<0 0 0>, Maximum=<30 0 0>)
    )
    Skeletal Mesh=$System.Skeletal Mesh
    Socket Sampling Mode=::NewEnumerator5
    Use Sampled Mesh Attribute as Ribbon ID=false
   )
   &CameraOffset(Camera Offset Amount=5)
   &ScaleRibbonWidth()
   /SetVariables(Particles.RibbonLinkOrder:NiagaraFloat=$Particles.PartitionParticles.Normalized Index In Partition) {enabled=false, isSetParameters=true, module=SetVariables_#9}
   /SetVariables(Particles.NewRibbonOrder:NiagaraFloat=$Particles.PartitionParticles.Normalized Index In Partition) {isSetParameters=true, module=SetVariables_#10}
   &DynamicMaterialParameters(Module.Index0Float0Write=true, Module.Index0Float1Write=true, Module.Index0Float2Write=true, Module.Index0Float3Write=true)
   &ScaleColor(
    Scale RGB=&VectorFromFloat(Value=&FloatFromCurve(
      FloatCurve=DataInterface() {propertyValues {Curve {
         defaultValue=1
         keys [
          {arriveTangent=9.0725, interpMode=RCIM_Cubic, leaveTangent=9.0725, tangentMode=RCTM_User, time=0, value=0}
          {arriveTangent=-2.5249, interpMode=RCIM_Cubic, leaveTangent=-2.5249, tangentMode=RCTM_User, time=0.1408, value=0.9847}
          {arriveTangent=-0.0405, interpMode=RCIM_Cubic, leaveTangent=-0.0405, tangentMode=RCTM_User, time=1, value=0}
         ]
        }}}
      Scale Curve=10
     ))
    ScaleA=true
    ScaleRGB=true
   )
  ]
  rendererClasses [&NiagaraSpriteRendererProperties, &NiagaraRibbonRendererProperties]
  renderers [
   {rendererClass=&NiagaraSpriteRendererProperties}
   {rendererClass=&NiagaraRibbonRendererProperties, rendererIndex=1}
   {rendererClass=&NiagaraRibbonRendererProperties, rendererIndex=2}
  ]
 }
 Minimal003 {
  emitterUpdateScript [
   &EmitterState(Life Cycle Mode=::NewEnumerator1, Loop Behavior=::NewEnumerator1, Loop Duration=0.5, UseLoopDelay=true)
   &SpawnBurst_Instantaneous(Spawn Count=4, Use Spawn Probability=false, UseLoopCountLimit=false)
  ]
  particleSpawnScript [
   &InitializeParticle(
    Color=<rgba 1.4294 0 2 1>
    Lifetime=0.5
    Sprite Size=<40 10>
    Sprite Size Mode=::NewEnumerator4
    UsePositionOffset=false
    Write Material Random=true
   )
   &PartitionParticles(Number Of Partitions=4)
   /SetVariables(Particles.PartitionBool:NiagaraBool=&SetBoolByIntComparison(A=$Particles.PartitionParticles.Partition, B=1, Comparison Type=::NewEnumerator4)) {isSetParameters=true}
   /SetVariables(Particles.RandomBone:NiagaraInt32=&MakeCustomIntFromBool(Bool=$Particles.PartitionBool, True INT=0, False INT=2)) {isSetParameters=true}
   &CameraOffset(Camera Offset Amount=10)
  ]
  particleUpdateScript [
   &ParticleState()
   &SkeletalMeshLocation(
    Bone / Socket Index=$Particles.PartitionParticles.Partition
    MaskSpawnGroup=false
    Mesh Sampling Type=::NewEnumerator5
    Sampled Position Offset=&SelectVectorFromArray(
     Direct Array Index=$Particles.PartitionParticles.Partition
     Vector Selection Array=DataInterface() {propertyValues {FloatData [<15 1.5 -2>, <15 1.5 2>, <15 -1.5 -2>, <15 -1.5 2>]}}
    )
    Skeletal Mesh=$System.Skeletal Mesh2
    Socket Sampling Mode=::NewEnumerator5
   )
   &SpriteFacingAndAlignment(
    Alignment Coordinate Space=::World
    Facing Coordinate Space=::World
    Sprite Alignment=&RotateVectorByQuaternion(Vector To Rotate=<0 0 1>, Quaternion=$Particles.SkeletalMeshLocation.SampledRotation)
    Sprite Facing=&RotateVectorByQuaternion(Vector To Rotate=<0 1 0>, Quaternion=$Particles.SkeletalMeshLocation.SampledRotation)
    WriteAlignment=true
    WriteFacing=true
   )
   &ScaleColor(
    Scale Alpha=&FloatFromCurve(
     FloatCurve=DataInterface() {propertyValues {Curve {
        defaultValue=0
        keys [
         {interpMode=RCIM_Cubic, tangentMode=RCTM_User, time=0, value=0}
         {arriveTangent=0.1167, interpMode=RCIM_Cubic, leaveTangent=0.1167, tangentMode=RCTM_User, time=0.2, value=1}
         {interpMode=RCIM_Cubic, tangentMode=RCTM_User, time=0.8, value=1}
         {tangentMode=RCTM_User, time=1, value=0}
        ]
       }}}
     Scale Curve=&Waveform("Waveform [1]"=::NewEnumerator12, "[1] Frequency"=10, "[1] Amplitude Min/Max"=<0 1>, "[1] Bias"=0.5)
    )
    ScaleA=true
    ScaleRGB=true
   )
  ]
  rendererClasses [&NiagaraSpriteRendererProperties]
  renderers [{rendererClass=&NiagaraSpriteRendererProperties}]
 }
 MuzzleFlashStreak001 {rendererClasses [], renderers []}
 NE_LoopingZaps001 {rendererClasses [], renderers []}
 NE_Sprites {
  bEnabled=false
  emitterUpdateScript [
   &EmitterState()
   &SpawnBurst_Instantaneous(Spawn Count=5, Spawn Time=$User.ExplosionDelay, Use Spawn Probability=false, UseLoopCountLimit=false)
   &SpawnBurst_Instantaneous(Spawn Count=5, Spawn Group=1, Spawn Time=$User.ExplosionDelay, Use Spawn Probability=false, UseLoopCountLimit=false) {module=SpawnBurst_Instantaneous001}
  ]
  particleSpawnScript [
   &InitializeParticle(
    Lifetime=&RandomRangeFloat(Minimum=0.5, Maximum=0.8)
    Sprite Size=&Multiply_Vector2D_ByFloat(Float=&RandomRangeFloat(Minimum=2, Maximum=5), Vector2D=<1 3>)
    Sprite Size Mode=::NewEnumerator4
    UsePositionOffset=false
    Write Material Random=true
   )
   /SetVariables(Particles.VisibilityTag=&SpawnGroup()) {isSetParameters=true}
   &ShapeLocation(
    MaskSpawnGroup=false
    Override Seed=true
    Random Seed:NiagaraInt32=&SpawnGroup()
    Sphere Radius=2
    UseOwnerScale=true
   )
   &AddVelocity(
    Cone Angle=25
    Module.UseConeFalloff=false
    Module.UseVelDistribution=false
    Override Seed=false
    Velocity Mode=::NewEnumerator2
    Velocity Speed=&RandomRangeFloat(Minimum=250, Maximum=500)
   )
  ]
  particleUpdateScript [
   &ParticleState()
   &InheritSourceMovement(
    Applied Position Scale=&VectorFromFloat(Value=&Multiply_Float(
      A=1
      B=&FloatFromCurve(FloatCurve=DataInterface() {propertyValues {Curve {keys [
           {arriveTangent=-1.9407, interpMode=RCIM_Cubic, leaveTangent=-1.9407, tangentMode=RCTM_User, time=0, value=1}
           {interpMode=RCIM_Cubic, tangentMode=RCTM_User, time=1, value=0.5}
          ]}}})
     ))
    ApplyForce=false
    ApplyPosition=true
    ApplyPositionFalloff=false
   )
   &Drag(Drag=&RandomRangeFloat(Minimum=3, Maximum=5), Use Linear Drag=true, Use Rotational Drag=true)
   &ScaleColor(
    Scale Alpha=&FloatFromCurve(FloatCurve=DataInterface() {propertyValues {Curve {keys [
         {arriveTangent=6.4019, interpMode=RCIM_Cubic, leaveTangent=6.4019, tangentMode=RCTM_User, time=0, value=0}
         <at 0.1 0.7 RCIM_Cubic>
         <at 1 0 RCIM_Cubic>
        ]}}})
    ScaleA=true
    ScaleRGB=true
   )
   &SolveForcesAndVelocity(Clamp Velocity=false, Limit Acceleration=false)
   &DynamicMaterialParameters(
    Index 0 Param 1=2
    Index 0 Param 3=1
    Index 3 Param 1=&Lerp_Float(
     A=0.2
     B=1
     Alpha=&FloatFromCurve(FloatCurve=DataInterface() {propertyValues {Curve {keys [<at 0 0>,<at 1 1>]}}})
    )
    Module.Index0Float0Write=true
    Module.Index0Float1Write=true
    Module.Index0Float2Write=true
    Module.Index0Float3Write=true
    Module.Index3Float0Write=true
    Module.Index3Float1Write=true
    Module.Index3Float2Write=true
    Module.Index3Float3Write=true
    Write Parameter Index 3=::NewEnumerator1
   )
   /SetVariables(Particles.SparkGlowColor:LinearColor=&MultiplyLinearColor(
     Linear Color=<rgba 0.1 0.1 0.1 0.05>
     Scale Factor=&MakeVector4_FromLinearColor(LinearColor=$Particles.Color)
    )) {isSetParameters=true}
   /SetVariables(Particles.SparkGlowSize:Vector2f=&Multiply_Vector2D(A=$Particles.SpriteSize, B=<5 2>)) {isSetParameters=true}
   /SetVariables(Particles.SparkGlowDynMat:Vector4f=<2 20 90 1>) {isSetParameters=true}
   &FN_Sprite_Attributes(
    Color=&MakeLinearColorFromVectorAndFloat(
     "Vector (RGB)"=&MakeVectorFromLinearColorRGB(Color=&Multiply_LinearColorByFloat(
       Linear Color=&LerpLinearColors(
        StartColor=&ColorFromCurve(
         ColorCurve=DataInterface() {propertyValues {
           AlphaCurve {keys [<at 0 1 RCIM_Cubic>]}
           BlueCurve {keys [<at 0.3173 1 RCIM_Cubic>,<at 0.3548 1>,<at 0.6928 1>,<at 0.7245 1>]}
           GreenCurve {keys [<at 0.3173 0.0008 RCIM_Cubic>,<at 0.3548 0>,<at 0.6928 0>,<at 0.7245 0.4>]}
           RedCurve {keys [<at 0.3173 0.2843 RCIM_Cubic>,<at 0.3548 0.7166>,<at 0.6928 0.7166>,<at 0.7245 0.8131>]}
          }}
         CurveIndex=&RandomRangeFloat(Minimum=0, Maximum=1)
        )
        EndColor=&ColorFromCurve(
         ColorCurve=DataInterface() {propertyValues {
           AlphaCurve {keys [<at 0 1 RCIM_Cubic>]}
           BlueCurve {keys [<at 0.3173 1 RCIM_Cubic>,<at 0.3548 0.837>,<at 0.678 0.837>,<at 0.7245 1>]}
           GreenCurve {keys [<at 0.3173 0.0008 RCIM_Cubic>,<at 0.3548 0.25>,<at 0.678 0.25>,<at 0.7245 0.4>]}
           RedCurve {keys [<at 0.3173 0.2843 RCIM_Cubic>,<at 0.3548 1>,<at 0.678 1>,<at 0.7245 0.8131>]}
          }}
         CurveIndex=&RandomRangeFloat(Minimum=0, Maximum=1)
        )
        LerpFactor=&RandomRangeFloat(Minimum=0, Maximum=1)
       )
       Float=&Lerp_Float(
        A=1
        B=20
        Alpha=&FloatFromCurve(FloatCurve=DataInterface() {propertyValues {Curve {keys [
             {arriveTangent=-0.4493, interpMode=RCIM_Cubic, leaveTangent=-0.4493, tangentMode=RCTM_User, time=0, value=1}
             {interpMode=RCIM_Cubic, tangentMode=RCTM_User, time=0.7, value=0}
            ]}}})
       )
      ))
     "Float (Alpha)"=&FloatFromCurve(FloatCurve=DataInterface() {propertyValues {Curve {keys [
          {interpMode=RCIM_Cubic, tangentMode=RCTM_User, time=0, value=1}
          {arriveTangent=-2.0246, interpMode=RCIM_Cubic, leaveTangent=-2.0246, tangentMode=RCTM_User, time=1, value=0}
         ]}}})
    )
    Create Color Attribute=true
    Create SpriteSize Attribute=true
    SpriteSize=&Multiply_Vector2D_ByFloat(
     Float=&Multiply_Float(
      A=&RandomRangeFloat(Minimum=0.5, Maximum=1.5)
      B=&FloatFromCurve(FloatCurve=DataInterface() {propertyValues {Curve {keys [
           {interpMode=RCIM_Cubic, tangentMode=RCTM_User, time=0.2, value=1}
           {arriveTangent=-2.4619, interpMode=RCIM_Cubic, leaveTangent=-2.4619, tangentMode=RCTM_User, time=1, value=0}
          ]}}})
     )
     Vector2D=<5 5>
    )
    Write to Module Namespace=true
   )
  ]
  rendererClasses [&NiagaraSpriteRendererProperties]
  renderers [
   {rendererClass=&NiagaraSpriteRendererProperties}
   {rendererClass=&NiagaraSpriteRendererProperties, rendererIndex=1}
   {rendererClass=&NiagaraSpriteRendererProperties, rendererIndex=2}
  ]
 }
 ShellBurst {rendererClasses [], renderers []}
 Shockwave {rendererClasses [], renderers []}
 Spark {
  simTarget=GPUComputeSim
  emitterUpdateScript [
   &EmitterState()
   &SpawnBurst_Instantaneous(Spawn Count=4, Use Spawn Probability=false, UseLoopCountLimit=false)
  ]
  particleSpawnScript [
   &InitializeParticle(
    Color=<rgba 2.135 0.0601 12 1>
    Lifetime Max=0.3
    Lifetime Min=0.15
    Lifetime Mode=::NewEnumerator1
    Sprite Size Mode=::NewEnumerator3
    Uniform Sprite Size=5
    UsePositionOffset=false
    Write Material Random=true
   )
   &AddVelocityInCone(
    Cone Angle=25
    Module.Use Velocity Falloff On Cone Axis=false
    Velocity Strength=&RandomRangeFloat(Minimum=450, Maximum=1200)
   )
  ]
  particleUpdateScript [
   &ParticleState()
   &SolveForcesAndVelocity(Clamp Velocity=false, Limit Acceleration=false)
   &ScaleSpriteSizeBySpeed(Max Scale Factor=<0.4 2.5>, Min Scale Factor=<0 1>, Sample Scale Factor By Curve=true, Velocity Threshold=500)
   &ScaleSpriteSize(Uniform Curve Sprite Scale=DataInterface() {propertyValues {Curve {
       defaultValue=1
       keys [{interpMode=RCIM_Cubic, tangentMode=RCTM_User, time=0, value=1}, {interpMode=RCIM_Cubic, tangentMode=RCTM_User, time=1, value=0}]
      }}})
   &ScaleColor(
    Scale RGB=&MakeVectorFromLinearColorRGB(Color=&ColorFromCurve(ColorCurve=DataInterface() {propertyValues {
        AlphaCurve {keys [<at 0 1 RCIM_Cubic>,<at 1 1 RCIM_Cubic>]}
        BlueCurve {keys [<at 0.4478 3.1109>,<at 1 0.3497 RCIM_Cubic>]}
        GreenCurve {keys [<at 0.4478 6.9398>,<at 1 0.7314 RCIM_Cubic>]}
        RedCurve {keys [<at 0.4478 12>,<at 1 1 RCIM_Cubic>]}
       }}))
    ScaleA=true
    ScaleRGB=false
   )
  ]
  rendererClasses [&NiagaraSpriteRendererProperties]
  renderers [{rendererClass=&NiagaraSpriteRendererProperties}]
 }
 SparksLong001 {
  simTarget=GPUComputeSim
  emitterUpdateScript [
   &EmitterState()
   &SpawnBurst_Instantaneous(
    Spawn Count=&RandomRangeInt(Minimum=15, Maximum=21)
    Use Spawn Probability=false
    UseLoopCountLimit=false
   )
   &SpawnBurst_Instantaneous(
    Spawn Count=&RandomRangeInt(Minimum=15, Maximum=21)
    Spawn Time=0.05
    Use Spawn Probability=false
    UseLoopCountLimit=false
   ) {enabled=false, module=SpawnBurst_Instantaneous001}
  ]
  particleSpawnScript [
   &InitializeParticle(
    Color=<rgba 0.7117 0.02 4 1>
    Lifetime Max=0.24
    Lifetime Min=0.14
    Lifetime Mode=::NewEnumerator1
    Position Offset=<-5 0 0>
    Position Offset Coordinate Space=::Local
    Sprite Size Max=<3 1.5>
    Sprite Size Min=<2 1.2>
    Sprite Size Mode=::NewEnumerator2
    UsePositionOffset=true
    Write Material Random=true
   ) {module=InitializeParticle001}
   &ShapeLocation(
    MaskSpawnGroup=false
    Non Uniform Scale=<2 1 1>
    Override Seed=false
    Rotation Mode=::NewEnumerator2
    Sphere Radius=5
    UseOwnerScale=true
    Yaw / Pitch / Roll=<-0.25 0 0>
   )
   &AddVelocityInCone(
    Module.Use Velocity Falloff On Cone Axis=false
    Velocity Strength=&RandomRangeFloat(Minimum=500, Maximum=1000)
   )
   &AddVelocity(Velocity=&RandomRangeVector(Minimum=<150 0 0>, Maximum=<200 0 0>)) {module=AddVelocity001}
  ]
  particleUpdateScript [
   &ParticleState()
   &Drag(Drag=4, Use Linear Drag=true, Use Rotational Drag=true)
   &SolveForcesAndVelocity(Clamp Velocity=false, Limit Acceleration=false)
   &ScaleColor(
    Scale Alpha=&FloatFromCurve(FloatCurve=DataInterface() {propertyValues {Curve {keys [<at 0 1 RCIM_Cubic>,<at 1 0 RCIM_Cubic>]}}})
    ScaleA=true
    ScaleRGB=true
   )
   &ScaleSpriteSizeBySpeed(Max Scale Factor=<0.5 25>, Velocity Threshold=700)
  ]
  rendererClasses [&NiagaraSpriteRendererProperties]
  renderers [{rendererClass=&NiagaraSpriteRendererProperties}]
 }
 Sparks_Fast {
  bEnabled=false
  emitterUpdateScript [
   &EmitterState(Life Cycle Mode=::NewEnumerator1, Loop Behavior=::NewEnumerator1, Loop Delay=0.1325, Loop Duration=0.15, UseLoopDelay=false)
   &SpawnBurst_Instantaneous(Spawn Count=45, Use Spawn Probability=false)
  ]
  particleSpawnScript [
   &InitializeParticle(
    Color Maximum=<rgba 3.2524 1.6 4 1>
    Color Minimum=<rgba 2.8663 0 4 1>
    Color Mode=::NewEnumerator2
    Lifetime Max=0.25
    Lifetime Min=0.05
    Lifetime Mode=::NewEnumerator1
    Sprite Size Mode=::NewEnumerator1
    Uniform Sprite Size Max=5
    Uniform Sprite Size Min=3
    UsePositionOffset=false
    Write Material Random=true
   ) {module=InitializeParticle001}
   &AddVelocityInCone(
    Cone Angle=30
    Module.Use Velocity Falloff On Cone Axis=false
    Velocity Strength=&UniformRangedFloat(Minimum=300, Maximum=4000)
   )
   &ShapeLocation(
    Disc Coverage=0.1
    MaskSpawnGroup=false
    Override Seed=false
    Ring Radius=10
    Rotation Mode=::NewEnumerator1
    Shape Primitive=::NewEnumerator4
    UseOwnerScale=true
   )
  ]
  particleUpdateScript [
   &ParticleState()
   &ScaleColor(
    Scale Alpha=&FloatFromCurve(FloatCurve=DataInterface() {propertyValues {Curve {keys [
         {tangentMode=RCTM_User, time=0, value=1}
         {arriveTangent=-0.0405, interpMode=RCIM_Cubic, leaveTangent=-0.0405, tangentMode=RCTM_User, time=0.7009, value=1}
         {arriveTangent=-8.0089, interpMode=RCIM_Cubic, leaveTangent=-8.0089, tangentMode=RCTM_User, time=1, value=0}
        ]}}})
    ScaleA=true
    ScaleRGB=true
   )
   &Drag(Drag=&RandomRangeFloat(Minimum=4, Maximum=12), Use Linear Drag=true, Use Rotational Drag=false)
   &GravityForce(Gravity=&Lerp_Vector(A=<0 0 -980>, B=<0 0 -250>, Alpha=$Particles.MaterialRandom))
   &SolveForcesAndVelocity(Clamp Velocity=false, Limit Acceleration=false)
   &ScaleSpriteSize(Scale Factor=&Vector2DFromFloat(Value=&FloatFromCurve(FloatCurve=DataInterface() {propertyValues {Curve {keys [
          {tangentMode=RCTM_User, time=0, value=1}
          {arriveTangent=-0.0405, interpMode=RCIM_Cubic, leaveTangent=-0.0405, tangentMode=RCTM_User, time=0.7009, value=1}
          {arriveTangent=-8.0089, interpMode=RCIM_Cubic, leaveTangent=-8.0089, tangentMode=RCTM_User, time=1, value=0}
         ]}}})))
   &DynamicMaterialParameters(
    Index 0 Param 1=1
    Index 0 Param 2=1
    Index 0 Param 3=&FloatFromCurve(FloatCurve=DataInterface() {propertyValues {Curve {keys [{interpMode=RCIM_Cubic, tangentMode=RCTM_User, time=0, value=1}, {interpMode=RCIM_Cubic, tangentMode=RCTM_User, time=1, value=0}]}}})
    Index 0 Param 4=1
    Module.Param0WriteEnabled=true
    Module.Param1WriteEnabled=true
    Module.Param2WriteEnabled=true
    Module.Param3WriteEnabled=true
   ) {module=DynamicMaterialParameters001}
   &ScaleSpriteSizeBySpeed(Max Scale Factor=<0.8 6>, Min Scale Factor=<1 3>, Velocity Threshold=2000)
  ]
  rendererClasses [&NiagaraSpriteRendererProperties]
  renderers [{rendererClass=&NiagaraSpriteRendererProperties}]
 }
}
summary {
 systemName=NS_PepperCat_MuzzleFlash_Secondary
 emitters [
  {bEnabled=false, emitterName=Sparks_Fast, rendererClasses [&NiagaraSpriteRendererProperties]}
  {bEnabled=false, emitterName=FrontMuzzle002, rendererClasses [&NiagaraMeshRendererProperties]}
  {bEnabled=false, emitterName=FrontMuzzle003, rendererClasses [&NiagaraMeshRendererProperties]}
  {bEnabled=false, emitterName=Light, rendererClasses [&NiagaraLightRendererProperties]}
  {emitterName=Spark, simTarget=GPUComputeSim, rendererClasses [&NiagaraSpriteRendererProperties]}
  {emitterName=FN_Weapons_Smoke_BETA, rendererClasses [&NiagaraSpriteRendererProperties]}
  {
   emitterName=Lit_CINE
   rendererClasses [
    &NiagaraSpriteRendererProperties
    &NiagaraMeshRendererProperties
    &NiagaraComponentRendererProperties
   ]
  }
  {emitterName=SparksLong001, simTarget=GPUComputeSim, rendererClasses [&NiagaraSpriteRendererProperties]}
  {bEnabled=false, emitterName=NE_Sprites, rendererClasses [&NiagaraSpriteRendererProperties]}
  {
   bEnabled=false
   emitterName=Minimal
   rendererClasses [&NiagaraSpriteRendererProperties, &NiagaraRibbonRendererProperties]
  }
  {
   bEnabled=false
   emitterName=Minimal001
   rendererClasses [&NiagaraSpriteRendererProperties, &NiagaraRibbonRendererProperties]
  }
  {emitterName=Minimal003, rendererClasses [&NiagaraSpriteRendererProperties]}
  {emitterName=CoreFlash001, rendererClasses []}
  {emitterName=MuzzleFlashStreak001, rendererClasses []}
  {emitterName=ShellBurst, rendererClasses []}
  {emitterName=Shockwave, rendererClasses []}
  {emitterName=NE_LoopingZaps001, rendererClasses []}
  {emitterName=Lightning, rendererClasses []}
  {emitterName=Lit, rendererClasses []}
  {emitterName=Glow_HIGH, rendererClasses [&NiagaraSpriteRendererProperties, &NiagaraMeshRendererProperties]}
  {emitterName=Glow_LOW, rendererClasses []}
 ]
 userVariables [
  User.bIsFrontEnd {type=NiagaraBool}
  User.bIsFrontEndPreview {type=NiagaraBool}
  User.bisInVehicle {type=NiagaraBool}
  User.ExplosionDelay {type=NiagaraFloat, defaultValue=0, description=""}
  User.HueShift {type=NiagaraFloat, defaultValue=0, description=""}
  User.IsSkydiving {type=NiagaraBool}
  User.LeftRightHand {type=NiagaraFloat, defaultValue=1, description=""}
  User.Overheat {type=NiagaraFloat, defaultValue=0, description=""}
  User.OriginOffset {type=NiagaraPosition, description="", defaultValue {}}
 ]
}
