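# Niagara system for the PepperCat projectile, a spinning, electrified purple-blue mesh effect.
# Cube001 spins the core, while Lightning001 orbits four offset pieces with center-facing, length-stretched lightning and random lights, alongside flickering Bolt_Meshes.
# Electricity traces random world points and aims short arcs at hits, culling misses and distant hits.
# MedLow launches colliding particles whose mesh and glow appear on impact, with alpha scaled by time of day.
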NiagaraSystem /PepperCatShared/Assets/FX/NS_PepperCat_Projectile
dependencies=null
emitters {
 Airglow {rendererClasses [], renderers []}
 Bolt_Meshes {
  emitterUpdateScript [
   &EmitterState()
   &SpawnBurst_Instantaneous(Use Spawn Probability=false, UseLoopCountLimit=true)
   &SpawnRate(SpawnGroup=1, SpawnRate=30, Use Spawn Probability=false) {module=SpawnRate001}
   &SpawnBurst_Instantaneous(Spawn Group=1, Use Spawn Probability=false, UseLoopCountLimit=true) {module=SpawnBurst_Instantaneous001}
  ]
  particleSpawnScript [
   /SetVariables(Particles.VisibilityTag=&SpawnGroup()) {isSetParameters=true}
   &InitializeParticle(
    Color=&RandomRangeLinearColor(Minimum=<rgba 0.1979 0.0038 0.0691 1>, Maximum=<rgba 0.195 0 0.3802 1>)
    Lifetime=&MaskFloatBySpawnGroup(
     Masked Float=&RandomRangeFloat(Minimum=0.1, Maximum=0.2)
     Passthrough Float=&RandomRangeFloat(Minimum=0.15, Maximum=0.1)
    )
    Mesh Scale Mode=::NewEnumerator3
    Mesh Uniform Scale=&MaskFloatBySpawnGroup(Masked Float=5, Passthrough Float=&RandomRangeFloat(Minimum=1, Maximum=4))
    UsePositionOffset=false
    Write Material Random=true
   )
   &InitialMeshOrientation(
    Mesh Orientation Mode=::NewEnumerator7
    Rotation=&RandomRangeVector(Minimum=<0 -0.25 -0.02>, Maximum=<0 -0.25 0.02>)
    Use Rotation Vector=true
   )
   &ShapeLocation(MaskSpawnGroup=true, Override Seed=false, Spawn Group Mask=1, Sphere Radius=15, UseOwnerScale=true) {module=ShapeLocation001}
  ]
  particleUpdateScript [
   &ParticleState()
   &ScaleMeshSize(Scale Factor=&MaskVectorBySpawnGroup(
     Masked Vector=&RandomRangeVector(Minimum=<0.5 0.5 1.5>, Maximum=<0.5 0.5 2>)
     Passthrough Vector=&RandomRangeVector(Minimum=<1 1 2>, Maximum=<1.2 1.2 2>)
    )) {module=ScaleMeshSize001}
   &ScaleMeshSize(Scale Factor=&MakeVector(
     X=&FloatFromCurve(
      FloatCurve=DataInterface() {propertyValues {Curve {keys [<at 0 0>, {arriveTangent=0.0951, leaveTangent=-0.1745, time=0.15, value=1}]}}}
      CurveIndex=$Emitter.Age
      Scale Curve=0.5
     )
     Y=&FloatFromCurve(
      FloatCurve=DataInterface() {propertyValues {Curve {keys [<at 0 0>, {arriveTangent=0.0951, leaveTangent=-0.1745, time=0.15, value=1}]}}}
      CurveIndex=$Emitter.Age
      Scale Curve=0.5
     )
     Z=&FloatFromCurve(FloatCurve=DataInterface() {propertyValues {Curve {keys [
          {interpMode=RCIM_Cubic, leaveTangent=0.0966, tangentMode=RCTM_Break, time=0, value=0.9893}
          {arriveTangent=0.0951, interpMode=RCIM_Cubic, leaveTangent=-0.1745, tangentMode=RCTM_Break, time=1.0021, value=1.02}
         ]}}})
    ))
   &DynamicMaterialParameters(
    Index 0 Param 3=&RandomRangeFloat(Minimum=0, Maximum=1)
    Index 0 Param 4=&MaskFloatBySpawnGroup(
     Masked Float=1
     Passthrough Float=&SelectFloatFromArray(
      Direct Array Index=&ModuloInt(A=&ReturnParticleIDAttributes(), B=3)
      Float Selection Array=DataInterface() {propertyValues {FloatData [3,0,10,3]}}
     )
    )
    Index 1 Param 1=1
    Index 1 Param 2=2
    Index 1 Param 3=&MaskFloatBySpawnGroup(
     Masked Float=-1
     Passthrough Float=&FloatFromCurve(FloatCurve=DataInterface() {propertyValues {Curve {keys [{interpMode=RCIM_Cubic, tangentMode=RCTM_Break, time=0.3, value=-0.5}, <at 1 0.1>]}}})
    )
    Index 2 Param 1=&MaskFloatBySpawnGroup(
     Masked Float=&FloatFromCurve(FloatCurve=DataInterface() {propertyValues {Curve {keys [<at 0 1>,<at 1 0>]}}})
     Passthrough Float=&Multiply_Float(
      A=&FloatFromCurve(
       FloatCurve=DataInterface() {propertyValues {Curve {keys [{interpMode=RCIM_Cubic, leaveTangent=8.1085, tangentMode=RCTM_Break, time=0.8568, value=1}, {tangentMode=RCTM_User, time=1, value=0}]}}}
       Scale Curve=0.5
      )
      B=&Divide_Float(A=&MakeFloatFromVector(Vector=$Particles.Scale, Channel=::NewEnumerator0), B=18)
     )
    )
    Index 2 Param 2=&MaskFloatBySpawnGroup(
     Masked Float=1
     Passthrough Float=&FloatFromCurve(
      FloatCurve=DataInterface() {propertyValues {Curve {keys [
          {interpMode=RCIM_Cubic, tangentMode=RCTM_Break, time=0.7, value=0}
          {arriveTangent=-0.0859, interpMode=RCIM_Cubic, tangentMode=RCTM_Break, time=1, value=-1}
         ]}}}
      Scale Curve=50
     )
    )
    Index 2 Param 3=&MaskFloatBySpawnGroup(
     Masked Float=0
     Passthrough Float=&Waveform(
      Global Input Position=&RandomRangeFloat(Minimum=0, Maximum=50)
      Global Phase=$Particles.Age
      "Waveform [1]"=::NewEnumerator12
      "[1] Phase Scale"=2
      "[1] Frequency"=1
      "[1] Amplitude Min/Max"=<0 1>
     )
    )
    Index 2 Param 4=&MaskFloatBySpawnGroup(Masked Float=0, Passthrough Float=25)
    Index 3 Param 3=&FloatFromCurve(FloatCurve=DataInterface() {propertyValues {Curve {keys [<at 0 -0.65>,<at 0.5 0>]}}}, CurveIndex=$Emitter.Age)
    Index 3 Param 4=-1
    Module.Index0Param0Write=true
    Module.Index0Param1Write=true
    Module.Index0Param2Write=true
    Module.Index0Param3Write=true
    Module.Index1Param0Write=true
    Module.Index1Param1Write=true
    Module.Index1Param2Write=true
    Module.Index1Param3Write=true
    Module.Index2Param0Write=true
    Module.Index2Param1Write=true
    Module.Index2Param2Write=true
    Module.Index2Param3Write=true
    Module.Index3Param0Write=true
    Module.Index3Param1Write=true
    Module.Index3Param2Write=true
    Module.Index3Param3Write=true
    Write Parameter Index 1=true
    Write Parameter Index 2=true
    Write Parameter Index 3=true
   )
   &ScaleColor(
    Scale Alpha=&FloatFromCurve(FloatCurve=DataInterface() {propertyValues {Curve {keys [
         {tangentMode=RCTM_User, time=0, value=1}
         {arriveTangent=-0.0405, interpMode=RCIM_Cubic, leaveTangent=-0.0405, tangentMode=RCTM_User, time=0.7009, value=1}
         {arriveTangent=-8.0089, interpMode=RCIM_Cubic, leaveTangent=-8.0089, tangentMode=RCTM_User, time=1, value=0}
        ]}}})
    Scale RGB=&MaskVectorBySpawnGroup(
     Masked Vector=&VectorFromFloat(Value=&RandomRangeFloat(Minimum=1, Maximum=2, Evaluation Type=::NewEnumerator1))
     Passthrough Vector=&Lerp_Vector(
      A=&VectorFromFloat(Value=&FloatFromCurve(FloatCurve=DataInterface() {propertyValues {Curve {keys [<at 0.2771 1>,<at 0.3786 1>]}}}, Scale Curve=10))
      B=&VectorFromFloat(Value=&RandomRangeFloat(Minimum=0.5, Maximum=1, Evaluation Type=::NewEnumerator1))
      Alpha=&FloatFromCurve(FloatCurve=DataInterface() {propertyValues {Curve {keys [
           {interpMode=RCIM_Cubic, leaveTangent=2.5294, tangentMode=RCTM_Break, time=0, value=0}
           {arriveTangent=-0.2166, interpMode=RCIM_Cubic, tangentMode=RCTM_Break, time=1, value=1}
          ]}}})
     )
    )
    ScaleA=true
    ScaleRGB=true
   )
   /SetVariables(Particles.MeshOrientation2:Quat4f=&MultiplyQuaternion(
     Quaternion A=$Particles.MeshOrientation
     Quaternion B=&MakeQuatFromAxisAngle(Axis=<0 0 1>, AngleInDegrees=&RandomRangeFloat(Minimum=0, Maximum=360))
    )) {isSetParameters=true}
   &ScaleColor(
    Scale Alpha=&FloatFromCurve(FloatCurve=DataInterface() {propertyValues {Curve {
        defaultValue=1
        keys [{interpMode=RCIM_Cubic, tangentMode=RCTM_User, time=0, value=1}, {interpMode=RCIM_Cubic, tangentMode=RCTM_User, time=1, value=0}]
       }}})
    ScaleA=true
    ScaleRGB=false
   ) {module=ScaleColor001}
   &KillParticles(Kill Particles=&SetBoolByFloatComparison(A=$System.Timeline.NormalizedTime, B=1, Comparison Type=::NewEnumerator2))
  ]
  rendererClasses [&NiagaraMeshRendererProperties]
  renderers [
   {rendererClass=&NiagaraMeshRendererProperties, rendererIndex=0}
   {rendererClass=&NiagaraMeshRendererProperties, rendererIndex=1}
  ]
 }
 Cube001 {
  emitterUpdateScript [
   &EmitterState(
    Inactive Response=::NewEnumerator1
    Life Cycle Mode=::NewEnumerator1
    Loop Behavior=::NewEnumerator1
    Loop Duration Mode=::NewEnumerator1
    UseLoopDelay=false
   )
   &SpawnBurst_Instantaneous(Use Spawn Probability=false, UseLoopCountLimit=false)
  ]
  particleSpawnScript [
   &InitializeParticle(Lifetime=5, Mesh Scale Mode=::NewEnumerator3, Mesh Uniform Scale=3, UsePositionOffset=false, Write Material Random=true)
   &MeshRotationForce(Rotation=$System.RandomRotation) {enabled=false}
   &ApplyInitialForces()
  ]
  particleUpdateScript [
   &ParticleState(Kill Particles When Lifetime Has Elapsed=false, Let Infinitely Lived Particles Die When Emitter Deactivates=true)
   /SetVariables(Particles.MeshOrientation=&MakeQuaternion(
     Quaternion From=0
     Axis=$System.Spin
     Angle=&Multiply_Float(A=$Emitter.Age, B=$System.rotationrate)
     From=?
     To=$System.Spin
     Angle Type=::NewEnumerator1
     Normalize Vectors=false
    )) {isSetParameters=true}
   &ScaleMeshSize(Scale Factor=&VectorFromFloat(Value=&FloatFromCurve(
      FloatCurve=DataInterface() {propertyValues {Curve {
         defaultValue=1
         keys [
          {arriveTangent=10.1958, interpMode=RCIM_Cubic, leaveTangent=10.1958, tangentMode=RCTM_User, time=0, value=0}
          {arriveTangent=0.0224, interpMode=RCIM_Cubic, leaveTangent=0.0224, tangentMode=RCTM_User, time=0.5, value=1}
         ]
        }}}
      CurveIndex=$Particles.Age
     )))
   &SolveForcesAndVelocity(Clamp Velocity=false, Limit Acceleration=false)
  ]
  rendererClasses [&NiagaraSpriteRendererProperties, &NiagaraMeshRendererProperties]
  renderers [
   {rendererClass=&NiagaraSpriteRendererProperties}
   {rendererClass=&NiagaraMeshRendererProperties, rendererIndex=1}
  ]
 }
 Electricity {
  emitterUpdateScript [
   &EmitterState()
   &SpawnRate(SpawnRate=&Multiply_Float(A=10, B=1), Use Spawn Probability=false)
  ]
  particleSpawnScript [
   &PartitionParticles(Number Of Partitions=4)
   &InitializeParticle(
    Color=<rgba 4.3152 0 8 1>
    Lifetime Max=0.2
    Lifetime Min=0.15
    Lifetime Mode=::NewEnumerator1
    Mesh Scale=&Multiply_VectorByFloat(Vector=&RandomRangeVector(Minimum=<2 1 1>, Maximum=<2 1 1>), Float=1)
    Mesh Scale Mode=::NewEnumerator4
    Sprite Size Mode=::NewEnumerator3
    Uniform Sprite Size=50
    UsePositionOffset=true
    Write Material Random=true
   )
   &self:TraceLocation(TraceEndWorld:NiagaraPosition=&AddVectorToPosition(
     Position=/Niagara/DynamicInputs/Helpers/SimulationPosition()
     Vector=&RandomRangeVector(Minimum=<-2250 -2250 -2250>, Maximum=<2250 2250 2250>)
    )) {module=TraceLocation}
   /SetVariables(Particles.DynamicParamOverride:Vector4f=&MakeFloat4FromVectorAndFloat(
     "Vector (RGB)"=/Niagara/DynamicInputs/Transforms/TransformPosition(
      Position=Linked() {linkedVariable {name=Particles.CollisionPosWorld, type=NiagaraPosition}}
      Source Space=::Simulation
      Destination Space=::World
      Apply Scale=true
     )
     "Float (Alpha)"=0
    )) {isSetParameters=true}
   &KillParticles(Kill Particles=&Boolean_NotOperation(Bool=$Particles.CollisionValid)) {module=KillParticles001}
  ]
  particleUpdateScript [
   &ParticleState()
   &ScaleSpriteSize(Uniform Curve Sprite Scale=DataInterface() {propertyValues {Curve {keys [
        {interpMode=RCIM_Cubic, leaveTangent=5.8794, tangentMode=RCTM_Break, time=0, value=0}
        {interpMode=RCIM_Cubic, tangentMode=RCTM_Break, time=0.5, value=1.05}
        {arriveTangent=-6.7193, interpMode=RCIM_Cubic, tangentMode=RCTM_Break, time=1, value=0}
       ]}}})
   &ScaleMeshSize(Scale Factor=&MakeVector(
     X=&FloatFromCurve(FloatCurve=DataInterface() {propertyValues {Curve {
         defaultValue=1
         keys [{interpMode=RCIM_Cubic, leaveTangent=2.7575, tangentMode=RCTM_Break, time=0.25, value=0}, {tangentMode=RCTM_User, time=1, value=1.375}]
        }}})
     Y=1
     Z=&FloatFromCurve(FloatCurve=DataInterface() {propertyValues {Curve {
         defaultValue=1
         keys [{interpMode=RCIM_Cubic, leaveTangent=2.2978, tangentMode=RCTM_Break, time=0, value=0}, {tangentMode=RCTM_User, time=1, value=1}]
        }}})
    )) {enabled=false}
   &SkeletalMeshLocation(
    Bone / Socket Index=$Particles.PartitionParticles.Partition
    Bone Sampling Mode=::NewEnumerator4
    MaskSpawnGroup=false
    Orientation Sampling=::NewEnumerator1
    Skeletal Mesh=$System.Skeletal Mesh
   ) {enabled=false}
   &ShapeLocation(
    Additional Quaternion Rotation=true
    Additional Yaw / Pitch / Roll=<0 0.25 0>
    Cylinder Height=&SelectFloatFromArray(
     Direct Array Index=$Particles.Initial.PartitionParticles.Partition
     Float Selection Array=DataInterface() {propertyValues {FloatData [50,50,75,75]}}
    )
    Cylinder Radius=10
    MaskSpawnGroup=false
    Offset=&SelectVectorFromArray(
     Direct Array Index=$Particles.PartitionParticles.Partition
     Vector Selection Array=DataInterface() {propertyValues {FloatData [<0 0 0>, <0 0 0>, <0 0 0>, <0 0 -28>]}}
    )
    Offset Mode=::NewEnumerator0
    Override Seed=false
    Rotation Mode=::NewEnumerator3
    Rotation Quaternion=$Particles.SkeletalMeshLocation.SampledRotation
    Shape Origin=$Particles.SkeletalMeshLocation.SampledPosition
    Shape Primitive=::NewEnumerator1
    Surface Only=false
    UseOwnerScale=true
   ) {enabled=false, module=ShapeLocation001}
   &UpdateMeshOrientation(
    Facing Direction=&Subtract_Position(A=$Particles.CollisionPosWorld, B=$Particles.ShapeLocation001.ShapePosition)
    Facing Interpolation=false
    Orientation Method=::NewEnumerator1
    Override Quat=false
    TransformFacing=false
    TransformUp=false
   )
   &KillParticles(Kill Particles=&SetBoolByFloatComparison(
     A=&ClampFloat(Float=&DotProduct(
       Vector A=$Particles.ShapeLocation.ShapeNormal
       Vector B=&NormalizeVector(
        Vector To Normalize=Linked() {linkedVariable {name=Particles.ShapeLocation001.ShapePosition, type=NiagaraPosition}}
        Safe Normalize=true
        Safe Fallback Threshold=0
        Safe Fallback Vector=<0 0 0.001>
       )
      ))
     B=0.5
    ))
   &DynamicMaterialParameters(
    Index 0 Param 1=&Multiply_Float(
     A=&Lerp_Float(
      A=0.5
      B=1.5
      Alpha=&FN_LOD_Distance_Remap(Start Distance=0, End Distance=3500, Minimum Output=0, Maximum Output=1, Clamp Output=true)
     )
     B=&VectorLength(Vector=$Engine.Owner.Scale)
    )
    Index 0 Param 2=&FloatFromCurve(
     FloatCurve=DataInterface() {propertyValues {Curve {keys [{interpMode=RCIM_Cubic, leaveTangent=3.2702, tangentMode=RCTM_Break, time=0, value=0}, <at 1 1>]}}}
     Scale Curve=2.5
    )
    Index 0 Param 3=&RandomRangeFloat(Minimum=6, Maximum=8)
    Index 0 Param 4=&RandomRangeFloat(Minimum=0, Maximum=1)
    Module.Index0Float0Write=true
    Module.Index0Float1Write=true
    Module.Index0Float2Write=true
    Module.Index0Float3Write=true
   )
   &SolveForcesAndVelocity(Clamp Velocity=false, Limit Acceleration=false)
   &ScaleColor(
    Scale Alpha=&FloatFromCurve(FloatCurve=DataInterface() {propertyValues {Curve {keys [{interpMode=RCIM_Cubic, tangentMode=RCTM_Break, time=0, value=1}, <at 1 0>]}}})
    ScaleA=true
    ScaleRGB=true
   ) {module=ScaleColor001}
   &FN_Sprite_Attributes(
    Color=&MultiplyLinearColor(Linear Color=$Particles.Color, Scale Factor=<1 1 1 0.05>)
    Create Camera Offset Attribute=true
    Create Color Attribute=true
    Write to Module Namespace=true
   )
   &FN_Sprite_Attributes(
    Camera Offset=20
    Create Camera Offset Attribute=true
    Create Position Attribute=true
    Create SpriteSize Attribute=true
    Position:NiagaraPosition=$Particles.ShapeLocation.ShapePosition
    SpriteSize=&Multiply_Vector2D_ByFloat(
     Float=&FloatFromCurve(FloatCurve=DataInterface() {propertyValues {Curve {keys [
          {interpMode=RCIM_Cubic, leaveTangent=2.7622, tangentMode=RCTM_Break, time=0, value=-0.05}
          {interpMode=RCIM_Cubic, leaveTangent=2.7622, tangentMode=RCTM_Break, time=1, value=1.05}
         ]}}})
     Vector2D=<50 50>
    )
    Write to Module Namespace=true
   ) {module=FN_Sprite_Attributes001}
   /SetVariables(Particles.Body Position:NiagaraPosition=$Particles.CollisionPosWorld, Particles.Dyn:Vector4f=<8 1 5 0>) {isSetParameters=true}
   &KillParticlesInVolume(
    Invert Volume=true
    Position:NiagaraPosition=$Particles.CollisionPosWorld
    Sphere Radius=1250
    Volume Origin:NiagaraPosition=$Engine.Owner.Position
   ) {module=KillParticlesInVolume001}
  ]
  rendererClasses [&NiagaraMeshRendererProperties, &NiagaraSpriteRendererProperties]
  renderers [
   {rendererClass=&NiagaraMeshRendererProperties, rendererIndex=0}
   {rendererClass=&NiagaraSpriteRendererProperties, rendererIndex=1}
   {rendererClass=&NiagaraSpriteRendererProperties, rendererIndex=2}
  ]
 }
 Lightning001 {
  emitterUpdateScript [
   &EmitterState()
   &SpawnBurst_Instantaneous(Spawn Group=1, Use Spawn Probability=false, UseLoopCountLimit=true)
   &SpawnBurst_Instantaneous(Spawn Count=4, Use Spawn Probability=false, UseLoopCountLimit=true) {module=SpawnBurst_Instantaneous001}
   &SpawnRate(SpawnGroup=1, SpawnRate=30, Use Spawn Probability=false)
  ]
  particleSpawnScript [
   &PartitionParticles(Number Of Partitions=4, Particles Per Partition=1, Partition Mode:ENiagara_PartitionMode_v2=::NewEnumerator1)
   /SetVariables(Particles.VisibilityTag=$Emitter.SpawnGroup) {isSetParameters=true}
   &InitializeParticle(
    Color=&RandomRangeLinearColor(Minimum=<rgba 0.3558 0.01 2 1>, Maximum=<rgba 3.2524 1.6 4 1>)
    Lifetime=&MakeCustomFloatFromBool(
     Bool=&SetBoolByFloatComparison(A=0.5, B=&MakeFloatFromInt(Int=$Emitter.SpawnGroup), Comparison Type=::NewEnumerator6)
     True Float=&RandomRangeFloat(Minimum=0.175, Maximum=0.35)
     False Float=50
    )
    Mesh Scale=<0.5 0.5 1>
    Mesh Scale Mode=::NewEnumerator4
    Position Offset=&SelectVectorFromArray(
     Direct Array Index=$Particles.PartitionParticles.Partition
     Vector Selection Array=DataInterface() {propertyValues {FloatData [<27.2538 -30.1791 23.8599>, <-31.9508 -29.7967 0.0721>, <23.4953 35.6742 28.9636>, <38.0981 -16.7619 -24.8232>]}}
    )
    Sprite Size Mode=::NewEnumerator3
    UsePositionOffset=false
    Write Material Random=true
   )
   /SetVariables(Particles.MeshOffsetPositions:Vector3f=&SelectVectorFromArray(
     Direct Array Index=$Particles.PartitionParticles.Partition
     Vector Selection Array=DataInterface() {propertyValues {FloatData [<27.2538 -30.1791 23.8599>, <-31.9508 -29.7967 0.0721>, <23.4953 35.6742 28.9636>, <38.0981 -16.7619 -24.8232>]}}
    )) {isSetParameters=true}
   /SetVariables(Particles.Random=&FloatFromCurve(
     FloatCurve=DataInterface() {propertyValues {Curve {
        defaultValue=1
        keys [
         {interpMode=RCIM_Cubic, tangentMode=RCTM_User, time=0, value=0}
         {arriveTangent=2.4041, interpMode=RCIM_Cubic, leaveTangent=2.4041, tangentMode=RCTM_User, time=1, value=1}
        ]
       }}}
     CurveIndex=$Particles.MaterialRandom
    )) {isSetParameters=true}
   &InitialMeshOrientation(
    Mesh Axis To Orient=<0 0 1>
    Mesh Orientation Mode=::NewEnumerator2
    Rotation=&RandomRangeVector(Minimum=<0 0 -1>, Maximum=<0 0 1>)
    Use Rotation Vector=true
   )
   &CameraOffset(Camera Offset Amount=&RandomRangeFloat(Minimum=-5, Maximum=5))
   &AddVelocity(Velocity=<10 0 0>)
   /SetVariables(Particles.MeshIndex2:NiagaraInt32=$Particles.PartitionParticles.Partition) {isSetParameters=true}
  ]
  particleUpdateScript [
   &ParticleState()
   /SetVariables(Particles.LerpPositions:Vector3f=&Lerp_Vector(
     A=&SimulationPosition()
     B=$Particles.Initial.MeshOffsetPositions
     Alpha=&FloatFromCurve(
      FloatCurve=DataInterface() {propertyValues {Curve {
         defaultValue=0
         keys [{interpMode=RCIM_Cubic, tangentMode=RCTM_User, time=0, value=0}, {interpMode=RCIM_Cubic, tangentMode=RCTM_User, time=1, value=1}]
        }}}
      CurveIndex=$Emitter.Age
     )
    )) {isSetParameters=true}
   /SetVariables(Particles.MeshOrientation=&MultiplyQuaternion(
     Quaternion A=&OrientMeshToVector(
      Look At Direction=&Subtract_Vector(A=&SimulationPosition(), B=Linked() {linkedVariable {name=Particles.Position, type=NiagaraPosition}})
      Reference Axis Vector=<1 0 0>
     )
     Quaternion B=&MakeQuaternion(
      Quaternion From=0
      Axis=$System.Spin
      Angle=&Multiply_Float(A=$Emitter.Age, B=180)
      From=?
      To=$System.Spin
      Angle Type=::NewEnumerator0
      Normalize Vectors=false
     )
    )) {enabled=false, isSetParameters=true}
   &UpdateMeshOrientation(
    Facing Direction=&Subtract_Vector(A=&SimulationPosition(), B=Linked() {linkedVariable {name=Particles.Position, type=NiagaraPosition}})
    Facing Interpolation=false
    Orientation Method=::NewEnumerator1
    Override Quat=false
    TransformFacing=false
    TransformUp=false
   )
   &ScaleMeshSize(Scale Factor=&MakeVector(
     X=1
     Y=1
     Z=&MakeFloatFromVector(
      Vector=&VectorFromFloat(Value=&Divide_Float(
        A=&VectorLength(Vector=&Subtract_Vector(A=&SimulationPosition(), B=Linked() {linkedVariable {name=Particles.Position, type=NiagaraPosition}}))
        B=130
       ))
      Channel=::NewEnumerator0
     )
    ))
   &SolveForcesAndVelocity(Clamp Velocity=false, Limit Acceleration=false)
   &DynamicMaterialParameters(
    Index 0 Param 1=&MakeFloatFromVector(Vector=$Particles.Scale, Channel=::NewEnumerator0)
    Index 0 Param 2=&RandomRangeFloat(Minimum=0, Maximum=10)
    Index 0 Param 3=&FloatFromCurve(
     FloatCurve=DataInterface() {propertyValues {Curve {keys [{interpMode=RCIM_Cubic, leaveTangent=8.1085, tangentMode=RCTM_Break, time=0.8568, value=1}, {tangentMode=RCTM_User, time=1, value=0}]}}}
     Scale Curve=0.0175
    )
    Index 0 Param 4=0.4
    Index 1 Param 1=&FloatFromCurve(FloatCurve=DataInterface() {propertyValues {Curve {keys [
         {interpMode=RCIM_Cubic, leaveTangent=1.1034, tangentMode=RCTM_Break, time=0, value=0}
         {arriveTangent=2.2399, tangentMode=RCTM_User, time=0.5, value=1}
        ]}}})
    Index 1 Param 2=&FloatFromCurve(FloatCurve=DataInterface() {propertyValues {Curve {keys [<at 0.5 0>, {arriveTangent=3.2053, tangentMode=RCTM_User, time=0.9, value=1}]}}})
    Index 1 Param 3=&FloatFromCurve(FloatCurve=DataInterface() {propertyValues {Curve {keys [<at 0 0 RCIM_Cubic>, {arriveTangent=3.4223, interpMode=RCIM_Cubic, leaveTangent=3.4223, tangentMode=RCTM_User, time=1, value=1}]}}})
    Index 1 Param 4=&FloatFromCurve(
     FloatCurve=DataInterface() {propertyValues {Curve {keys [
         {interpMode=RCIM_Cubic, leaveTangent=0.1169, tangentMode=RCTM_Break, time=0, value=0}
         {arriveTangent=2.5708, interpMode=RCIM_Cubic, leaveTangent=2.5708, tangentMode=RCTM_User, time=1, value=1}
        ]}}}
     Scale Curve=&RandomRangeFloat(Minimum=1, Maximum=40)
    )
    Index 2 Param 1=&FloatFromCurve(FloatCurve=DataInterface() {propertyValues {Curve {keys [
         {interpMode=RCIM_Cubic, leaveTangent=-0.0025, tangentMode=RCTM_Break, time=0, value=0.5}
         {arriveTangent=-0.0316, interpMode=RCIM_Cubic, tangentMode=RCTM_Break, time=1, value=-0.2}
        ]}}})
    Index 2 Param 2=&FloatFromCurve(
     FloatCurve=DataInterface() {propertyValues {Curve {keys [
         {interpMode=RCIM_Cubic, tangentMode=RCTM_User, time=0, value=0}
         {arriveTangent=2.0522, interpMode=RCIM_Cubic, leaveTangent=2.0522, tangentMode=RCTM_User, time=1, value=1}
        ]}}}
     Scale Curve=2
    )
    Module.Index0Float0Write=true
    Module.Index0Float1Write=true
    Module.Index0Float2Write=true
    Module.Index0Float3Write=true
    Module.Index1Float0Write=true
    Module.Index1Float1Write=true
    Module.Index1Float2Write=true
    Module.Index1Float3Write=true
    Module.Index2Float0Write=true
    Module.Index2Float1Write=true
    Module.Index2Float2Write=false
    Module.Index2Float3Write=false
    Module.Index3Float0Write=false
    Module.Index3Float1Write=false
    Module.Index3Float2Write=false
    Module.Index3Float3Write=true
    Write Parameter Index 1=::NewEnumerator1
    Write Parameter Index 2=::NewEnumerator1
    Write Parameter Index 3=::NewEnumerator1
   )
   &ScaleColor(
    Scale Alpha=&FloatFromCurve(FloatCurve=DataInterface() {propertyValues {Curve {keys [
         {tangentMode=RCTM_User, time=0, value=1}
         {arriveTangent=-0.0405, interpMode=RCIM_Cubic, leaveTangent=-0.0405, tangentMode=RCTM_User, time=0.7009, value=1}
         {arriveTangent=-8.0089, interpMode=RCIM_Cubic, leaveTangent=-8.0089, tangentMode=RCTM_User, time=1, value=0}
        ]}}})
    Scale RGB=&Lerp_Vector(
     A=&VectorFromFloat(Value=&FloatFromCurve(FloatCurve=DataInterface() {propertyValues {Curve {keys [<at 0 1>,<at 0.1 1>,<at 0.2 2>,<at 0.3 1>,<at 0.6 1>,<at 0.7 2>,<at 0.8 1>,<at 1 1>]}}}))
     B=&VectorFromFloat(Value=&RandomRangeFloat(Minimum=0.3, Maximum=8, Evaluation Type=::NewEnumerator1))
     Alpha=&FloatFromCurve(FloatCurve=DataInterface() {propertyValues {Curve {keys [
          {interpMode=RCIM_Cubic, leaveTangent=2.5294, tangentMode=RCTM_Break, time=0, value=0}
          {arriveTangent=-0.2166, interpMode=RCIM_Cubic, tangentMode=RCTM_Break, time=0.4, value=1}
         ]}}})
    )
    ScaleA=true
    ScaleRGB=true
   )
   /SetVariables(Particles.AirGlowColor:LinearColor=&MultiplyLinearColor(Linear Color=$Particles.Color, Scale Factor=<1 1 1 0.04>)) {isSetParameters=true}
   &FN_Light_Attributes(
    Color=<rgba 0 0.1231 1 1>
    Create Color Attribute=true
    Create Exponent Attribute=true
    Create Radius Attribute=true
    Enable Particle Light Enable=true
    Exponent=1.5
    Light Enabled:NiagaraBool=&RandomBool()
    Radius=&RandomRangeFloat(Minimum=80, Maximum=100)
    Scale Color=&Multiply_Float(A=20, B=&RandomRangeFloat(Minimum=0.5, Maximum=2, Evaluation Type=::NewEnumerator1))
   )
   /Niagara/Noise/Waveform(
    EnableExponent1=false
    GlobalClampMax=false
    GlobalClampMin=false
    UseInput1=true
    UsePhase1=false
    "[1] Frequency"=0.5
    "[1] Input Position":NiagaraFloat=&SelectFloatFromArray(Direct Array Index=$Particles.Initial.PartitionParticles.Partition, Float Selection Array=$System.ChunkArray)
   )
   /SetVariables(Particles.Position=&RotateVectorByQuaternion(
     Vector To Rotate=&Add_Vector(
      A=$Particles.LerpPositions
      B=&Multiply_VectorByFloat(
       Vector=&NormalizeVector(
        Vector To Normalize=&SubtractVectorFromPosition(Position=Linked() {linkedVariable {name=Particles.LerpPositions, type=Vector3f}}, Vector=&SimulationPosition())
        Safe Normalize=true
        Safe Fallback Threshold=0
        Safe Fallback Vector=<0 0 0.001>
       )
       Float=&Lerp_Float(A=-5, B=5, Alpha=$Particles.Waveform.Waveform)
      )
     )
     Quaternion=&MakeQuaternion(
      Quaternion From=0
      Axis=$System.Spin
      Angle=&Multiply_Float(A=$Emitter.Age, B=$System.rotationrate)
      From=?
      To=?
      Angle Type=::NewEnumerator1
      Coordinate Space=::Local
      Normalize Vectors=false
     )
    )) {isSetParameters=true}
   &ScaleMeshSize(Scale Factor=&MakeVector(
     X=&FloatFromCurve(
      FloatCurve=DataInterface() {propertyValues {Curve {
         defaultValue=1
         keys [
          {arriveTangent=4.1785, interpMode=RCIM_Cubic, leaveTangent=4.1785, tangentMode=RCTM_User, time=0, value=0}
          {arriveTangent=1.1179, interpMode=RCIM_Cubic, leaveTangent=1.1179, tangentMode=RCTM_User, time=0.2042, value=0.6228}
          {arriveTangent=0.0224, interpMode=RCIM_Cubic, leaveTangent=0.0224, tangentMode=RCTM_User, time=1, value=1}
         ]
        }}}
      CurveIndex=$Particles.Age
     )
     Y=&FloatFromCurve(
      FloatCurve=DataInterface() {propertyValues {Curve {
         defaultValue=1
         keys [
          {arriveTangent=4.1785, interpMode=RCIM_Cubic, leaveTangent=4.1785, tangentMode=RCTM_User, time=0, value=0}
          {arriveTangent=1.1179, interpMode=RCIM_Cubic, leaveTangent=1.1179, tangentMode=RCTM_User, time=0.2042, value=0.6228}
          {arriveTangent=0.0224, interpMode=RCIM_Cubic, leaveTangent=0.0224, tangentMode=RCTM_User, time=1, value=1}
         ]
        }}}
      CurveIndex=$Particles.Age
     )
     Z=1
    )) {module=ScaleMeshSize001}
   /SetVariables(
    Particles.Scale2:Vector3f=&Multiply_VectorByFloat(
     Vector=<3 3 3>
     Float=&FloatFromCurve(
      FloatCurve=DataInterface() {propertyValues {Curve {
         defaultValue=1
         keys [
          {arriveTangent=4.1785, interpMode=RCIM_Cubic, leaveTangent=4.1785, tangentMode=RCTM_User, time=0, value=0}
          {arriveTangent=1.1179, interpMode=RCIM_Cubic, leaveTangent=1.1179, tangentMode=RCTM_User, time=0.2042, value=0.6228}
          {arriveTangent=0.0224, interpMode=RCIM_Cubic, leaveTangent=0.0224, tangentMode=RCTM_User, time=1, value=1}
         ]
        }}}
      CurveIndex=$Emitter.Age
     )
    )
    Particles.Sprite Size2:Vector2f=&Multiply_Vector2D_ByFloat(
     Float=&FloatFromCurve(
      FloatCurve=DataInterface() {propertyValues {Curve {
         defaultValue=1
         keys [
          {arriveTangent=4.1785, interpMode=RCIM_Cubic, leaveTangent=4.1785, tangentMode=RCTM_User, time=0, value=0}
          {arriveTangent=1.1179, interpMode=RCIM_Cubic, leaveTangent=1.1179, tangentMode=RCTM_User, time=0.2042, value=0.6228}
          {arriveTangent=0.0224, interpMode=RCIM_Cubic, leaveTangent=0.0224, tangentMode=RCTM_User, time=1, value=1}
         ]
        }}}
      CurveIndex=$Emitter.Age
     )
     Vector2D=<100 100>
    )
   ) {isSetParameters=true}
   /SetVariables(Particles.SpriteColor=&MultiplyLinearColor(
     Linear Color=<rgba 0.1779 0.0052 1 1>
     Scale Factor=&MakeVector4_FromLinearColor(LinearColor=<rgba 1 1 1 1>)
    )) {isSetParameters=true}
  ]
  rendererClasses [&NiagaraMeshRendererProperties, &NiagaraSpriteRendererProperties]
  renderers [
   {rendererClass=&NiagaraMeshRendererProperties, rendererIndex=0}
   {rendererClass=&NiagaraMeshRendererProperties, rendererIndex=1}
   {rendererClass=&NiagaraMeshRendererProperties, rendererIndex=2}
   {rendererClass=&NiagaraSpriteRendererProperties, rendererIndex=3}
   {rendererClass=&NiagaraMeshRendererProperties, rendererIndex=4}
  ]
 }
 MedLow {
  emitterUpdateScript [
   &EmitterState(
    Delay First Loop Only=true
    Inactive Response=::NewEnumerator1
    Life Cycle Mode=::NewEnumerator1
    Loop Delay=0.5
    Loop Duration=1
    UseLoopDelay=true
   )
   &SpawnRate(
    SpawnRate=&RandomRangeFloat(Minimum=-3, Maximum=10, Evaluation Type=::NewEnumerator1)
    Use Spawn Probability=false
   )
  ]
  particleSpawnScript [
   &InitializeParticle(
    Color=<rgba 0.3184 0.0808 1 2>
    Lifetime Max=0.9
    Lifetime Min=0.6
    Lifetime Mode=::NewEnumerator1
    Mesh Scale Mode=::NewEnumerator3
    Mesh Uniform Scale=&Multiply_Float(A=&RandomRangeFloat(Minimum=0.18, Maximum=0.21), B=0.25)
    Position=$Engine.Owner.Position
    Position Mode=::NewEnumerator1
    Sprite Rotation Mode=::NewEnumerator2
    Sprite Size Mode=::NewEnumerator1
    Sprite UV Mode=::NewEnumerator3
    Uniform Sprite Size Max=70
    Uniform Sprite Size Min=50
    UsePositionOffset=false
    Write Material Random=true
   ) {module=InitializeParticle001}
   &ShapeLocation(MaskSpawnGroup=false, Override Seed=false, Scale Mode:ENiagara_ScaleMode=::NewEnumerator1, Sphere Radius=10)
   /SetVariables(Particles.SpriteMesh_Quat:Quat4f=/Niagara/DynamicInputs/Rotation/RandomQuaternion(), Particles.Sprite_CamOfs=20) {isSetParameters=true}
   &InitialMeshOrientation(
    Mesh Orientation Mode=::NewEnumerator1
    Rotation=&RandomRangeVector(Minimum=<-0.1 -0.1 -0.1>, Maximum=<0.1 0.1 0.1>)
    Use Rotation Vector=true
   )
   /SetVariables(
    Particles.InitPos:NiagaraPosition=$Particles.Initial.Position
    Particles.StoredVector:Vector3f=&TransformVector(
     Vector=Linked() {linkedVariable {name=Particles.Initial.Position, type=NiagaraPosition}}
     Source Space=::Simulation
     Destination Space=::World
     Apply Scale=true
     Transform Vector as Position=true
    )
   ) {enabled=false, isSetParameters=true}
   &AddVelocity(
    Velocity=&Multiply_Vector(
     A=&RandomRangeVector(Minimum=<0.6 0.6 0.6>, Maximum=<0.9 0.9 1>)
     B=&SelectVectorFromArray(
      Array Sampling Mode=::NewEnumerator0
      Direct Array Index=?
      Vector Selection Array=DataInterface() {propertyValues {FloatData [<-100 -100 -50>, <100 100 -60>]}}
      Random Seed=0
     )
    )
    Velocity Speed Scale=20
   ) {enabled=false, module=AddVelocity001}
   &AddVelocity(Module.Use Falloff Distance=false, Override Seed=false, Velocity Mode=::NewEnumerator1, Velocity Speed=5000)
  ]
  particleUpdateScript [
   &ParticleState()
   &UpdateMeshOrientation(Rotation Rate=&RandomRangeFloat(Minimum=-1, Maximum=1), Rotation Vector=<0.15 0.15 0.1>)
   &GravityForce(
    Coordinate Space=::Local
    Gravity=&Multiply_VectorByFloat(
     Vector=&SelectVectorFromArray(
      Array Sampling Mode=::NewEnumerator0
      Direct Array Index=?
      Vector Selection Array=DataInterface() {propertyValues {FloatData [<0 0 -35000>, <0 0 35000>]}}
      Random Seed=0
     )
     Float=&MakeCustomFloatFromBool(
      Bool=&SetBoolByFloatComparison(A=1, B=HlslExpression() {hlslExpression="length(Particles.Velocity)"})
      True Float=0
      False Float=1
     )
    )
   )
   &Collision(
    Age Colliding Particles=false
    CPU Collision Trace Channel=ECC_WorldStatic
    CPU Friction Blending:ENiagaraFrictionMergeType=::NewEnumerator3
    Collision Enabled=&SetBoolByFloatComparison(
     A=1
     B=&FloatFromCurve(FloatCurve=DataInterface() {propertyValues {Curve {keys [<at 0.05 0.95>,<at 0.2 2>]}}}, Scale Curve=0.5)
    )
    Enable Rest State=false
    Friction=10000
    Friction During a Bounce=10000
    KillOnCollision=false
    Method for Calculating Particle Radius:ENiagaraCalculateRadiusOptions=::NewEnumerator2
    Module.EnableMaxCollisionCount=false
    Module.Randomize Collision Normal Vector=false
    Randomize Collision Normal=0
    Restitution=0
   )
   &InterpolateOverTime(
    Initial Float Value=0.01
    Rate Of Change:NiagaraFloat=&MakeCustomFloatFromBool(Bool=$Particles.HasCollided, True Float=&Divide_Float(A=0.9, B=$Particles.Initial.Lifetime), False Float=0)
    Target Float Value=1
   )
   &SolveForcesAndVelocity(Clamp Velocity=false, Limit Acceleration=false)
   &ScaleSpriteSize(
    Scale Sprite Size Mode=::NewEnumerator0
    Uniform Scale Factor=&Multiply_Float(A=10, B=$Particles.InterpolateOverTime.InterpolateOverTime_Float.Moving Average)
   )
   &ScaleMeshSize(Scale Factor=&Multiply_VectorByFloat(
     Vector=&MakeVector(
      X=1
      Y=1
      Z=&ClampFloat(Float=&Multiply_Float(A=10, B=HlslExpression() {hlslExpression="length(Particles.Velocity)"}), Min=5, Max=8)
     )
     Float=&MakeCustomFloatFromBool(
      Bool=$Particles.HasCollided
      True Float=&FloatFromCurve(
       FloatCurve=DataInterface() {propertyValues {Curve {keys [
           {tangentMode=RCTM_User, time=0, value=0.4}
           <at 0.1 0.5>
           <at 0.1651 0.3>
           <at 0.2226 0.45>
           {tangentMode=RCTM_User, time=0.3, value=0.25}
           <at 0.3288 0.45>
           <at 0.3636 0.4>
           <at 0.4 0.7>
           <at 0.4246 0.4>
           <at 0.4803 0.6>
           <at 0.5186 0.25>
           <at 0.5761 0.4>
           <at 0.6108 0.15>
           <at 0.6533 0.25>
           <at 0.7 0.4>
           <at 0.7391 0.183>
           <at 0.8 0.45>
           <at 0.8424 0.1684>
           <at 0.8845 0.3723>
           <at 0.9185 0.0999>
           <at 0.9483 0.4763>
           <at 0.9912 0.5>
          ]}}}
       Scale Curve=1.5
      )
      False Float=0
     )
    ))
   &DynamicMaterialParameters(
    Index 0 Param 1=&MakeFloatFromVector(
     Vector=&AddVectorToPosition(Position=$Engine.Owner.Position, Vector=<0 0 0>)
     Channel=::NewEnumerator0
    )
    Index 0 Param 2=&MakeFloatFromVector(
     Vector=&AddVectorToPosition(Position=$Engine.Owner.Position, Vector=<0 0 0>)
     Channel=::NewEnumerator1
    )
    Index 0 Param 3=&MakeFloatFromVector(
     Vector=&AddVectorToPosition(Position=$Engine.Owner.Position, Vector=<0 0 0>)
     Channel=::NewEnumerator2
    )
    Index 0 Param 4=11
    Index 1 Param 1=&FloatFromCurve(FloatCurve=DataInterface() {propertyValues {Curve {keys [<at 0 1 RCIM_Cubic>, {arriveTangent=-1.1493, tangentMode=RCTM_User, time=0.3, value=0}]}}})
    Module.Param0WriteEnabled=true
    Module.Param0WriteEnabled1=true
    Module.Param1WriteEnabled=true
    Module.Param1WriteEnabled1=false
    Module.Param2WriteEnabled=true
    Module.Param2WriteEnabled1=false
    Module.Param3WriteEnabled=true
    Module.Param3WriteEnabled1=false
    Write Parameter Index 1=true
   ) {module=DynamicMaterialParameters001}
   &ScaleColor(
    Scale Alpha=&MakeCustomFloatFromBool(
     Bool=$Particles.HasCollided
     True Float=&FloatFromCurve(
      FloatCurve=DataInterface() {propertyValues {Curve {keys [
          {arriveTangent=0.0978, interpMode=RCIM_Cubic, leaveTangent=0.0978, tangentMode=RCTM_User, time=0.0417, value=0.4929}
          {arriveTangent=0.0598, interpMode=RCIM_Cubic, leaveTangent=0.0598, tangentMode=RCTM_User, time=0.1694, value=0.5078}
          {arriveTangent=-2.4139, interpMode=RCIM_Cubic, leaveTangent=-2.4139, tangentMode=RCTM_User, time=0.6, value=0}
         ]}}}
      Scale Curve=&FNTimeOfDayScalar(Night Scale=0.6, Morning Scale=0.7, Evening Scale=0.7)
     )
     False Float=0
    )
    Scale RGB=&Multiply_VectorByFloat(
     Vector=<2 2 2>
     Float=&FloatFromCurve(FloatCurve=DataInterface() {propertyValues {Curve {defaultValue=1, keys [{tangentMode=RCTM_User, time=0.2, value=0}, {tangentMode=RCTM_User, time=0.25, value=2}]}}})
    )
    ScaleA=true
    ScaleRGB=true
   )
   /SetVariables(
    Particles.Color2=&Multiply_LinearColorByFloat(
     Linear Color=&ColorFromCurve(ColorCurve=DataInterface() {propertyValues {
        AlphaCurve {keys [
          <at 0 0>
          <at 0.1666 1>
          {interpMode=RCIM_Cubic, leaveTangent=-2.2986, time=0.4537, value=1}
          {interpMode=RCIM_Cubic, leaveTangent=2.0391, time=0.5137, value=0}
         ]}
        BlueCurve {keys [<at 0 0>, {arriveTangent=6.4704, interpMode=RCIM_Cubic, leaveTangent=6.4704, time=0.5312, value=30}, <at 0.7728 5>]}
        GreenCurve {keys [<at 0 0>, {arriveTangent=0.0647, interpMode=RCIM_Cubic, leaveTangent=0.0647, time=0.5312, value=0.3}, <at 0.7728 0.05>]}
        RedCurve {keys [<at 0 0>, {arriveTangent=5.0847, interpMode=RCIM_Cubic, leaveTangent=5.0847, time=0.5312, value=9.7185}, <at 0.7728 3.9292>]}
       }})
     Float=&MakeCustomFloatFromBool(Bool=$Particles.HasCollided, True Float=1, False Float=0)
    )
    Particles.Color3:LinearColor=&ColorFromCurve(ColorCurve=DataInterface() {propertyValues {
       AlphaCurve {keys [<at 0.5008 0.8 RCIM_Cubic>,<at 1 0 RCIM_Cubic>]}
       BlueCurve {keys [<at 0.2831 0 RCIM_Cubic>,<at 0.3847 20>]}
       GreenCurve {keys [<at 0.2831 0 RCIM_Cubic>,<at 0.3847 0>]}
       RedCurve {keys [<at 0.2831 0 RCIM_Cubic>,<at 0.3847 6.898>]}
      }})
    Particles.DM2_0:Vector4f=&MakeVector4(
     X=&Multiply_Float(
      A=&FloatFromCurve(
       FloatCurve=DataInterface() {propertyValues {Curve {keys [{tangentMode=RCTM_User, time=0, value=1}, {tangentMode=RCTM_User, time=0.7618, value=1.5189}]}}}
       Scale Curve=6
      )
      B=&MakeCustomFloatFromBool(Bool=$Particles.HasCollided, True Float=0.2, False Float=0)
     )
     Y=&FloatFromCurve(
      FloatCurve=DataInterface() {propertyValues {Curve {keys [
          {arriveTangent=-0.0737, interpMode=RCIM_Cubic, leaveTangent=-0.0737, tangentMode=RCTM_User, time=0, value=0}
          {tangentMode=RCTM_User, time=1.0259, value=1.9743}
         ]}}}
      Scale Curve=4
     )
     Z=&FloatFromCurve(
      FloatCurve=DataInterface() {propertyValues {Curve {keys [
          {arriveTangent=3.2315, interpMode=RCIM_Cubic, leaveTangent=3.2315, tangentMode=RCTM_User, time=0, value=0}
          {arriveTangent=1.1179, interpMode=RCIM_Cubic, leaveTangent=1.1179, tangentMode=RCTM_User, time=0.4, value=1}
         ]}}}
      Scale Curve=55
     )
     W=&FloatFromCurve(
      FloatCurve=DataInterface() {propertyValues {Curve {keys [
          {arriveTangent=0.6176, interpMode=RCIM_Cubic, leaveTangent=0.6176, tangentMode=RCTM_User, time=0.2, value=0}
          {arriveTangent=6.707, interpMode=RCIM_Cubic, leaveTangent=6.707, tangentMode=RCTM_User, time=1, value=2.5}
         ]}}}
      Scale Curve=-250
     )
    )
    Particles.Mesh Scale2:Vector3f=&Multiply_VectorByFloat(
     Vector=<20 20 20>
     Float=&FloatFromCurve(FloatCurve=DataInterface() {propertyValues {Curve {
         defaultValue=1
         keys [
          {arriveTangent=2.0344, interpMode=RCIM_Cubic, leaveTangent=2.0344, tangentMode=RCTM_User, time=0, value=0.1}
          {tangentMode=RCTM_User, time=1, value=1}
         ]
        }}})
    )
   ) {isSetParameters=true}
   &KillParticlesInVolume(Invert Volume=true, Sphere Radius=1250)
  ]
  rendererClasses [&NiagaraMeshRendererProperties, &NiagaraSpriteRendererProperties]
  renderers [
   {rendererClass=&NiagaraMeshRendererProperties, rendererIndex=0}
   {rendererClass=&NiagaraMeshRendererProperties, rendererIndex=1}
   {rendererClass=&NiagaraSpriteRendererProperties, rendererIndex=2}
  ]
 }
 Runes {
  emitterUpdateScript [
   &EmitterState()
   &SpawnBurst_Instantaneous(Spawn Count=10, Spawn Time=2.25, Use Spawn Probability=false, UseLoopCountLimit=false) {enabled=false}
   &SpawnRate(SpawnRate=25, Use Spawn Probability=true)
  ]
  particleSpawnScript [
   &InitializeParticle(
    Lifetime=&RandomRangeFloat(Minimum=0.2, Maximum=0.5)
    Sprite Size Mode=::NewEnumerator1
    UsePositionOffset=false
    Write Material Random=true
   )
   &SkeletalMeshLocation(
    Bone Sampling Mode=::NewEnumerator0
    MaskSpawnGroup=false
    OverrideRandomSeed=false
    Sampled Scale Coordinate Space:ENiagaraCoordinateSpace=::World
    Skeletal Mesh=$System.Cubes
   )
   &ShapeLocation(MaskSpawnGroup=false, Offset Mode=::NewEnumerator0, Override Seed=false, Sphere Radius=45, UseOwnerScale=true)
   &AddVelocity(
    Module.Use Falloff Distance=false
    Override Seed=false
    Velocity Mode=::NewEnumerator1
    Velocity Origin=$Particles.SkeletalMeshLocation.SampledPosition
    Velocity Speed=&RandomRangeFloat(Minimum=10, Maximum=40)
   )
  ]
  particleUpdateScript [
   &ParticleState()
   &ScaleSpriteSize(
    Uniform Curve Scale=2
    Uniform Curve Sprite Scale=DataInterface() {propertyValues {Curve {keys [{tangentMode=RCTM_User, time=0, value=0.5}, {tangentMode=RCTM_User, time=1, value=1}]}}}
   )
   &DynamicMaterialParameters(
    Index 0 Param 1=/Niagara/DynamicInputs/Helpers/Waveform(
     "Waveform [1]"=::NewEnumerator9
     "[1] Phase Scale"=&RandomRangeFloat(Minimum=0, Maximum=1)
     "[1] Frequency"=&RandomRangeFloat(Minimum=6, Maximum=10)
     "[1] Amplitude Min/Max"=<0 4>
    )
    Index 0 Param 2=1
    Index 0 Param 3=1
    Module.Index0Float0Write=true
    Module.Index0Float1Write=true
    Module.Index0Float2Write=true
    Module.Index0Float3Write=true
   )
   &SubUVAnimation(
    Random Start Frame=true
    "Sprite Renderer "=DataInterface() {propertyValues {SpriteRenderer=&self:Runes_0.NiagaraSpriteRendererProperties_0}}
    SubUV Lookup Index:NiagaraFloat=&Add_Float(
     A=&Multiply_Float(A=$Engine.System.Age, B=&RandomRangeFloat(Minimum=0.05, Maximum=0.2))
     B=&RandomRangeFloat(Minimum=0, Maximum=1)
    )
    SubUV Lookup Index Scale:NiagaraFloat=&FloatFromCurve(
     FloatCurve=DataInterface() {propertyValues {Curve {keys [<at 0 4>,<at 0.1 1>,<at 1 1>]}}}
     Scale Curve=&RandomRangeFloat(Minimum=1, Maximum=2)
    )
    UseEndFrame=false
    UseStartFrame=false
   ) {module=SubUVAnimation001}
   &ScaleColor(
    Scale Alpha=&FloatFromCurve(FloatCurve=DataInterface() {propertyValues {Curve {keys [
         {interpMode=RCIM_Cubic, tangentMode=RCTM_User, time=0, value=0.1}
         {arriveTangent=-0.1, interpMode=RCIM_Cubic, leaveTangent=-0.1, time=0.1562, value=0.9653}
         <at 1 0>
        ]}}})
    Scale RGB=&Multiply_VectorByFloat(
     Vector=<1 1 1>
     Float=&FloatFromCurve(
      FloatCurve=DataInterface() {propertyValues {Curve {keys [<at 0 1>, {interpMode=RCIM_Cubic, tangentMode=RCTM_User, time=0.2, value=4}, <at 0.4 1>, <at 1 1>]}}}
      Scale Curve=2
     )
    )
    ScaleA=true
    ScaleRGB=true
   )
   &Drag(Drag=3, Use Linear Drag=true, Use Rotational Drag=false)
   &CurlNoiseForce(
    Bias Noise=false
    Noise Frequency=&RandomRangeFloat(Minimum=10, Maximum=50)
    Noise Strength=&Multiply_Float(
     A=&RandomRangeFloat(Minimum=100, Maximum=300)
     B=&FloatFromCurve(
      FloatCurve=DataInterface() {propertyValues {Curve {
         defaultValue=1
         keys [
          {interpMode=RCIM_Cubic, tangentMode=RCTM_User, time=0.4435, value=1}
          {interpMode=RCIM_Cubic, tangentMode=RCTM_User, time=0.4645, value=0.2434}
          {arriveTangent=2.6614, interpMode=RCIM_Cubic, leaveTangent=2.6614, tangentMode=RCTM_User, time=0.9853, value=0.6612}
         ]
        }}}
      CurveIndex=$Emitter.NormalizedLoopAge
     )
    )
    Pan Noise=false
   ) {enabled=false}
   &SolveForcesAndVelocity(Clamp Velocity=false, Limit Acceleration=false)
   /SetVariables(Particles.Position=&Lerp_Vector(
     A=Linked() {linkedVariable {name=StackContext.Position, type=NiagaraPosition}}
     B=Linked() {linkedVariable {name=Particles.ShapeLocation.ShapePosition, type=NiagaraPosition}}
     Alpha=0
    )) {enabled=false, isSetParameters=true, module=SetVariables_#1}
  ]
  rendererClasses [&NiagaraSpriteRendererProperties]
  renderers [{rendererClass=&NiagaraSpriteRendererProperties}]
 }
 Sparks {
  bEnabled=false
  emitterUpdateScript [
   &EmitterState()
   &SpawnBurst_Instantaneous(Spawn Count=10, Spawn Time=2.25, Use Spawn Probability=false, UseLoopCountLimit=false) {enabled=false}
   &SpawnRate(SpawnRate=25, Use Spawn Probability=true)
  ]
  particleSpawnScript [
   &InitializeParticle(
    Color=<rgba 0.1779 0.005 1 1>
    Lifetime=&RandomRangeFloat(Minimum=0.2, Maximum=0.5)
    Sprite Size Mode=::NewEnumerator1
    Uniform Sprite Size Max=20
    Uniform Sprite Size Min=10
    UsePositionOffset=false
    Write Material Random=true
   )
   &SkeletalMeshLocation(
    Bone Sampling Mode=::NewEnumerator0
    MaskSpawnGroup=false
    OverrideRandomSeed=false
    Sampled Scale Coordinate Space:ENiagaraCoordinateSpace=::World
    Skeletal Mesh=$System.Cubes
   )
   &ShapeLocation(MaskSpawnGroup=false, Offset Mode=::NewEnumerator0, Override Seed=false, Sphere Radius=22, UseOwnerScale=true)
   &AddVelocity(
    Module.Use Falloff Distance=false
    Override Seed=false
    Velocity Mode=::NewEnumerator1
    Velocity Origin=$Particles.SkeletalMeshLocation.SampledPosition
    Velocity Speed=&RandomRangeFloat(Minimum=10, Maximum=30)
   )
   &InheritVelocity(Inherited Velocity Amount Scale=<-0.25 -0.25 -0.25>, Speed Limit=false, UseSpeedThreshold=false)
  ]
  particleUpdateScript [
   &ParticleState()
   &DynamicMaterialParameters(
    Index 0 Param 1=&Waveform(
     "Waveform [1]"=::NewEnumerator9
     "[1] Phase Scale"=&RandomRangeFloat(Minimum=0, Maximum=1)
     "[1] Frequency"=&RandomRangeFloat(Minimum=6, Maximum=10)
     "[1] Amplitude Min/Max"=<0 4>
    )
    Index 0 Param 2=1
    Index 0 Param 3=1
    Module.Index0Float0Write=true
    Module.Index0Float1Write=true
    Module.Index0Float2Write=true
    Module.Index0Float3Write=true
   )
   &ScaleColor(
    Scale Alpha=&FloatFromCurve(FloatCurve=DataInterface() {propertyValues {Curve {keys [{interpMode=RCIM_Cubic, tangentMode=RCTM_User, time=0, value=1}, <at 1 0>]}}})
    Scale RGB=&Multiply_VectorByFloat(
     Vector=<1 1 1>
     Float=&FloatFromCurve(
      FloatCurve=DataInterface() {propertyValues {Curve {keys [<at 0 1>, {interpMode=RCIM_Cubic, tangentMode=RCTM_User, time=0.2, value=3}, <at 0.4 1>, <at 1 1>]}}}
      Scale Curve=2
     )
    )
    ScaleA=true
    ScaleRGB=true
   )
   &CurlNoiseForce(
    Bias Noise=false
    Noise Frequency=&RandomRangeFloat(Minimum=10, Maximum=50)
    Noise Strength=&Multiply_Float(
     A=&RandomRangeFloat(Minimum=100, Maximum=300)
     B=&FloatFromCurve(
      FloatCurve=DataInterface() {propertyValues {Curve {
         defaultValue=1
         keys [
          {interpMode=RCIM_Cubic, tangentMode=RCTM_User, time=0.4435, value=1}
          {interpMode=RCIM_Cubic, tangentMode=RCTM_User, time=0.4645, value=0.2434}
          {arriveTangent=2.6614, interpMode=RCIM_Cubic, leaveTangent=2.6614, tangentMode=RCTM_User, time=0.9853, value=0.6612}
         ]
        }}}
      CurveIndex=$Emitter.NormalizedLoopAge
     )
    )
    Pan Noise=false
   ) {enabled=false}
   &SolveForcesAndVelocity(Clamp Velocity=false, Limit Acceleration=false)
   &ScaleSpriteSizeBySpeed(Max Scale Factor=<0.5 5>, Min Scale Factor=<0.05 0.05>)
   /SetVariables(Particles.Position=&Lerp_Vector(
     A=Linked() {linkedVariable {name=StackContext.Position, type=NiagaraPosition}}
     B=Linked() {linkedVariable {name=Particles.ShapeLocation.ShapePosition, type=NiagaraPosition}}
     Alpha=0
    )) {enabled=false, isSetParameters=true, module=SetVariables_#1}
  ]
  rendererClasses [&NiagaraSpriteRendererProperties]
  renderers [{rendererClass=&NiagaraSpriteRendererProperties}]
 }
 Spawn_Sampler {
  simTarget=GPUComputeSim
  emitterUpdateScript [
   &EmitterState(Life Cycle Mode=::NewEnumerator1, Loop Duration=0.1, UseLoopDelay=false)
   &self:Sample_Position(Particle Attribute Reader=DataInterface() {propertyValues {EmitterBinding {bindingMode=Other, emitterName=Electricity}}}) {module=Sample_Position}
   &SpawnBurst_Instantaneous(
    Spawn Count=50
    Spawn Probability=&MakeCustomFloatFromBool(
     Bool=&SetBoolByFloatComparison(
      A=10
      B=&DistanceBetweenPositions(A=$Emitter.Ground Position, B=$Emitter.Reader.Body Position)
      Comparison Type=::NewEnumerator6
     )
     True Float=1
     False Float=0
    )
    Use Spawn Probability=true
    UseLoopCountLimit=false
   )
  ]
  particleSpawnScript [
   &InitializeParticle(
    Color=<rgba 5 5 5 1>
    Lifetime Max=1.6
    Lifetime Min=0.5
    Lifetime Mode=::NewEnumerator1
    Position Offset Coordinate Space=::World
    Sprite Size Mode=::NewEnumerator1
    Uniform Sprite Size Max=8
    Uniform Sprite Size Min=3
    UsePositionOffset=true
    Write Material Random=true
   )
   &ShapeLocation(MaskSpawnGroup=false, Override Seed=false, Shape Origin=$Emitter.Reader.Body Position, Sphere Radius=5, UseOwnerScale=true)
   &KillParticlesInVolume(
    Invert Volume=true
    Position:NiagaraPosition=$Emitter.Ground Position
    Sphere Radius=1250
    Volume Origin:NiagaraPosition=$Engine.Owner.Position
   ) {enabled=false}
   &AddVelocity(
    Cone Axis=<0 0 1>
    Inner Cone Angle=180
    Module.UseConeFalloff=false
    Module.UseVelDistribution=false
    Override Seed=false
    Velocity Mode=::NewEnumerator2
    Velocity Speed=&RandomRangeFloat(Minimum=100, Maximum=500)
   )
  ]
  particleUpdateScript [
   &ParticleState()
   &GravityForce()
   &Drag(Use Linear Drag=true, Use Rotational Drag=true)
   &Collision(
    Age Colliding Particles=false
    Friction=1
    Kill Offscreen Particles=true
    KillOnCollision=false
    Module.Control Rest State Via Surface Normal=false
    Module.EnableMaxCollisionCount=false
    Module.Randomize Collision Normal Vector=false
   )
   &SolveForcesAndVelocity(Clamp Velocity=false, Limit Acceleration=false)
   &ScaleSpriteSizeBySpeed(Max Scale Factor=<0.5 3>, Min Scale Factor=<0.5 0.5>)
   &ScaleColor(
    Linear Color Curve=DataInterface() {propertyValues {
      AlphaCurve {keys [<at 0.2302 1.5>,<at 1 0>]}
      BlueCurve {keys [<at 0 1>,<at 1 0>]}
      GreenCurve {keys [<at 0 0.0002>,<at 1 0>]}
      RedCurve {keys [<at 0 0.4465>,<at 1 0>]}
     }}
    Scale Mode=::NewEnumerator2
   )
   &ScaleSpriteSize(
    Scale Sprite Size Mode=::NewEnumerator0
    Uniform Scale Factor=&Lerp_Float(
     A=1
     B=3
     Alpha=&RemapRange(Input Value To Remap=$Engine.Owner.LODDistance, Input Min=0, Input Max=3500, Output Min=0, Output Max=1)
    )
   )
  ]
  rendererClasses [&NiagaraSpriteRendererProperties]
  renderers [{rendererClass=&NiagaraSpriteRendererProperties}]
 }
 Spawn_Sampler001 {
  simTarget=GPUComputeSim
  emitterUpdateScript [
   &EmitterState(Life Cycle Mode=::NewEnumerator1, Loop Duration=0.1, UseLoopDelay=false)
   &self:Sample_Position_01(Particle Attribute Reader=DataInterface() {propertyValues {EmitterBinding {bindingMode=Other, emitterName=Electricity}}}) {module=Sample_Position_01}
   &SpawnBurst_Instantaneous(
    Spawn Count=10
    Spawn Probability=&MakeCustomFloatFromBool(
     Bool=&SetBoolByFloatComparison(
      A=10
      B=&DistanceBetweenPositions(A=$Emitter.Ground Position, B=$Emitter.Reader.Body Position)
      Comparison Type=::NewEnumerator6
     )
     True Float=1
     False Float=0
    )
    Use Spawn Probability=true
    UseLoopCountLimit=false
   )
  ]
  particleSpawnScript [
   &InitializeParticle(
    Color=<rgba 5 5 5 1>
    Lifetime Max=1
    Lifetime Min=0.5
    Lifetime Mode=::NewEnumerator1
    Position Offset Coordinate Space=::World
    Sprite Size Mode=::NewEnumerator1
    Uniform Sprite Size Max=70
    Uniform Sprite Size Min=50
    UsePositionOffset=true
    Write Material Random=true
   )
   &ShapeLocation(MaskSpawnGroup=false, Override Seed=false, Shape Origin=$Emitter.Reader.Body Position, Sphere Radius=5, UseOwnerScale=true)
   &AddVelocity(
    Cone Axis=<0 0 1>
    Inner Cone Angle=180
    Module.UseConeFalloff=false
    Module.UseVelDistribution=false
    Override Seed=false
    Velocity Mode=::NewEnumerator2
    Velocity Speed=&RandomRangeFloat(Minimum=100, Maximum=500)
   )
   &KillParticlesInVolume(
    Invert Volume=true
    Position:NiagaraPosition=$Emitter.Ground Position
    Sphere Radius=1250
    Volume Origin:NiagaraPosition=$Engine.Owner.Position
   ) {enabled=false}
   &CameraOffset(Camera Offset Amount=10)
  ]
  particleUpdateScript [
   &ParticleState()
   &GravityForce()
   &Drag(Use Linear Drag=true, Use Rotational Drag=true)
   &SolveForcesAndVelocity(Clamp Velocity=false, Limit Acceleration=false)
   &ScaleColor(
    Linear Color Curve=DataInterface() {propertyValues {
      AlphaCurve {keys [<at 0 1>,<at 1 0>]}
      BlueCurve {keys [<at 0 1>,<at 0.0705 0>]}
      GreenCurve {keys [<at 0 0>,<at 0.0705 0>]}
      RedCurve {keys [<at 0 0.4759>,<at 0.0705 0>]}
     }}
    Scale Mode=::NewEnumerator2
   )
  ]
  rendererClasses [&NiagaraSpriteRendererProperties]
  renderers [{rendererClass=&NiagaraSpriteRendererProperties}]
 }
 SplitLightning {rendererClasses [], renderers []}
 wave {rendererClasses [], renderers []}
}
summary {
 systemName=NS_PepperCat_Projectile
 emitters [
  {emitterName=Airglow, rendererClasses []}
  {emitterName=Runes, rendererClasses [&NiagaraSpriteRendererProperties]}
  {bEnabled=false, emitterName=Sparks, rendererClasses [&NiagaraSpriteRendererProperties]}
  {emitterName=Cube001, rendererClasses [&NiagaraSpriteRendererProperties, &NiagaraMeshRendererProperties]}
  {
   bEnabled=false
   emitterName=Lightning
   rendererClasses [&NiagaraMeshRendererProperties, &NiagaraLightRendererProperties]
  }
  {
   bEnabled=false
   emitterName=Cube002
   rendererClasses [&NiagaraSpriteRendererProperties, &NiagaraMeshRendererProperties]
  }
  {emitterName=Bolt_Meshes, rendererClasses [&NiagaraMeshRendererProperties]}
  {emitterName=SplitLightning, rendererClasses []}
  {emitterName=Electricity, rendererClasses [&NiagaraMeshRendererProperties, &NiagaraSpriteRendererProperties]}
  {emitterName=Spawn_Sampler, simTarget=GPUComputeSim, rendererClasses [&NiagaraSpriteRendererProperties]}
  {emitterName=Spawn_Sampler001, simTarget=GPUComputeSim, rendererClasses [&NiagaraSpriteRendererProperties]}
  {emitterName=MedLow, rendererClasses [&NiagaraMeshRendererProperties, &NiagaraSpriteRendererProperties]}
  {emitterName=Lightning001, rendererClasses [&NiagaraMeshRendererProperties, &NiagaraSpriteRendererProperties]}
  {emitterName=wave, rendererClasses []}
 ]
 userVariables [User.bIsFrontEndPreview {type=NiagaraBool}]
}
