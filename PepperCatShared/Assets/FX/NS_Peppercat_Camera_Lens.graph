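# Niagara camera lens effect for Peppercat, with three emitters named for screen distortion (two) and a screen wave.
# User parameters expose lifetime scale, overload, opacity, size and a set-color toggle.
# System scripts use SystemState and a variable assignment, with float and color multiply inputs plus a platform-dependent float pick.
# No renderers are listed for any emitter, and the wiring from parameters to emitters is not shown.
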NiagaraSystem /PepperCatShared/Assets/FX/NS_Peppercat_Camera_Lens
dependencies {
 usedDataInterfaces [&NiagaraDataInterfacePlatformSet]
 usedDynamicInputs [
  &Multiply_Float
  &Multiply_LinearColorByFloat
  &FN_PickFloatDependOnPlatform
 ]
 usedModules [
  &self:SystemSpawnScript.SystemScriptSource.SystemScriptGraph.NiagaraNodeAssignment_2.SetVariables_#1
  &SystemState
 ]
}
emitters {
 Screen_Distortion {rendererClasses [], renderers []}
 Screen_Distortion001 {rendererClasses [], renderers []}
 Screen_Wave {rendererClasses [], renderers []}
}
summary {
 systemName=NS_Peppercat_Camera_Lens
 emitters [
  {emitterName=Screen_Distortion, rendererClasses []}
  {emitterName=Screen_Wave, rendererClasses []}
  {emitterName=Screen_Distortion001, rendererClasses []}
 ]
 userVariables [
  User.LifetimeScale {type=NiagaraFloat, defaultValue=1.25, description=""}
  User.Overload {type=NiagaraFloat, defaultValue=1, description=""}
  User.ScaleOpacity {type=NiagaraFloat, defaultValue=1, description=""}
  User.ScaleSize {type=NiagaraFloat, defaultValue=1, description=""}
  User.SetColor {type=NiagaraBool}
 ]
}
