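# Niagara system for a Peppercat camera lens trigger effect, judging by its name, with a screen wave emitter and two sphere distortion emitters.
# A system-spawn SetVariables module runs alongside SystemState, using float and color multiplies and a platform-dependent float pick.
# User parameters expose lifetime, opacity and size scales, an Overload float and a SetColor bool.
# The emitters list no renderers here.
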NiagaraSystem /PepperCatShared/Assets/FX/NS_Peppercat_Camera_Lens_Triggered
dependencies {
 usedDataInterfaces [&NiagaraDataInterfacePlatformSet]
 usedDynamicInputs [
  &Multiply_Float
  &Multiply_LinearColorByFloat
  &FN_PickFloatDependOnPlatform
 ]
 usedModules [
  &self:SystemSpawnScript.SystemScriptSource.SystemScriptGraph.NiagaraNodeAssignment_3.SetVariables_#1
  &SystemState
 ]
}
emitters {
 Screen_Wave {rendererClasses [], renderers []}
 Sphere_Distortion {rendererClasses [], renderers []}
 Sphere_Distortion001 {rendererClasses [], renderers []}
}
summary {
 systemName=NS_Peppercat_Camera_Lens_Triggered
 emitters [
  {emitterName=Sphere_Distortion, rendererClasses []}
  {emitterName=Screen_Wave, rendererClasses []}
  {emitterName=Sphere_Distortion001, rendererClasses []}
 ]
 userVariables [
  User.LifetimeScale {type=NiagaraFloat, defaultValue=1.5, description=""}
  User.Overload {type=NiagaraFloat, defaultValue=1, description=""}
  User.ScaleOpacity {type=NiagaraFloat, defaultValue=1, description=""}
  User.ScaleSize {type=NiagaraFloat, defaultValue=1, description=""}
  User.SetColor {type=NiagaraBool}
 ]
}
