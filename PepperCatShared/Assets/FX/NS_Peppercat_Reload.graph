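# Niagara system for the Peppercat reload effect, judging by its name, built from purple glows, flickering sprites and electrical zaps.
# Delayed emitters stage the timing, some through a shared LightningDelay variable.
# Sprite and lightning-mesh particles follow skeletal-mesh sockets with per-particle offsets, while curve-driven dynamic material parameters animate the bolts.
# Lightning brightness scales with a system time-of-day value; several emitters, including VentSmoke, are disabled.
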NiagaraSystem /PepperCatShared/Assets/FX/NS_Peppercat_Reload
dependencies {
 usedDataInterfaces [
  &NiagaraDataInterfaceCurve
  &NiagaraDataInterfaceSpriteRendererInfo
  &NiagaraDataInterfaceColorCurve
  &NiagaraDataInterfaceArrayFloat
  &NiagaraDataInterfaceArrayFloat3
  &NiagaraDataInterfaceSkeletalMesh
 ]
 usedDynamicInputs [&FNTimeOfDayScalar]
 usedModules [
  &self:SystemSpawnScript.SystemScriptSource.SystemScriptGraph.NiagaraNodeAssignment_0.SetVariables_#1
  &self:SystemSpawnScript.SystemScriptSource.SystemScriptGraph.NiagaraNodeAssignment_1.SetVariables_#2
  &SystemState
  &self:SystemSpawnScript.SystemScriptSource.SystemScriptGraph.NiagaraNodeAssignment_2.SetVariables_#3
  &self:SystemSpawnScript.SystemScriptSource.SystemScriptGraph.NiagaraNodeAssignment_3.SetVariables_#4
 ]
}
emitters {
 Linger_Airglow {
  emitterUpdateScript [
   &EmitterState(Life Cycle Mode=::NewEnumerator1, Loop Behavior=::NewEnumerator1, Loop Delay=1.35, UseLoopDelay=true)
   &SpawnBurst_Instantaneous(Spawn Count=1, Use Spawn Probability=false)
  ]
  particleSpawnScript [
   &InitializeParticle(
    Color=<rgba 0.7166 0 1 0.5>
    Position Offset=<50 0 10>
    Sprite Rotation Mode=::NewEnumerator2
    Sprite Size Mode=::NewEnumerator3
    Uniform Sprite Size=200
    UsePositionOffset=true
    Write Material Random=true
   ) {module=InitializeParticle001}
   &CameraOffset(Camera Offset Amount=10)
  ]
  particleUpdateScript [
   &ParticleState()
   &ScaleSpriteSize(Scale Factor=&Vector2DFromFloat(Value=&FloatFromCurve(FloatCurve=DataInterface() {propertyValues {Curve {keys [{arriveTangent=5.3461, interpMode=RCIM_Cubic, leaveTangent=5.3461, tangentMode=RCTM_User, time=0, value=0}, <at 1 2 RCIM_Cubic>]}}})))
   &DynamicMaterialParameters(
    Index 0 Param 1=0.5
    Index 0 Param 2=1
    Module.Param0WriteEnabled=true
    Module.Param1WriteEnabled=true
    Module.Param2WriteEnabled=true
    Module.Param3WriteEnabled=true
   )
   &ScaleColor(
    Scale Alpha=&FloatFromCurve(
     FloatCurve=DataInterface() {propertyValues {Curve {keys [{interpMode=RCIM_Cubic, tangentMode=RCTM_User, time=0, value=2}, {interpMode=RCIM_Cubic, tangentMode=RCTM_User, time=1, value=0}]}}}
     Scale Curve=&FN_FrontEndFloatModifier(Gameplay=1, Front End=0.5)
    )
    Scale RGB=&VectorFromFloat(Value=1)
    ScaleA=true
    ScaleRGB=true
   )
  ]
  rendererClasses [&NiagaraSpriteRendererProperties]
  renderers [{rendererClass=&NiagaraSpriteRendererProperties}]
 }
 Minimal {
  emitterUpdateScript [
   &EmitterState()
   &SpawnBurst_Instantaneous(Spawn Time=1.2, Use Spawn Probability=false, UseLoopCountLimit=false)
  ]
  particleSpawnScript [
   &InitializeParticle(Color=<rgba 0.5711 0 0.7991 1>, UsePositionOffset=false, Write Material Random=true)
   &InitialMeshOrientation(Mesh Orientation Mode=::NewEnumerator2, Rotation=<0 0 -0.25>, Use Rotation Vector=true)
  ]
  particleUpdateScript [
   &ParticleState()
   &DynamicMaterialParameters(
    Index 0 Param 1=&FloatFromCurve(FloatCurve=DataInterface() {propertyValues {Curve {
        defaultValue=1
        keys [
         {interpMode=RCIM_Cubic, tangentMode=RCTM_User, time=0, value=-1}
         {interpMode=RCIM_Cubic, tangentMode=RCTM_User, time=0.5, value=1}
         {tangentMode=RCTM_User, time=1, value=-1}
        ]
       }}})
    Module.Index0Float0Write=true
    Module.Index0Float1Write=true
    Module.Index0Float2Write=true
    Module.Index0Float3Write=true
   )
   &ScaleColor(
    Scale RGB=&VectorFromFloat(Value=&FloatFromCurve(
      FloatCurve=DataInterface() {propertyValues {Curve {
         defaultValue=1
         keys [
          {arriveTangent=9.0725, interpMode=RCIM_Cubic, leaveTangent=9.0725, tangentMode=RCTM_User, time=0, value=0}
          {arriveTangent=-2.5249, interpMode=RCIM_Cubic, leaveTangent=-2.5249, tangentMode=RCTM_User, time=0.1408, value=0.9847}
          {arriveTangent=-0.0405, interpMode=RCIM_Cubic, leaveTangent=-0.0405, tangentMode=RCTM_User, time=1, value=0}
         ]
        }}}
      Scale Curve=10
     ))
    ScaleA=true
    ScaleRGB=true
   )
  ]
  rendererClasses [&NiagaraSpriteRendererProperties, &NiagaraMeshRendererProperties]
  renderers [
   {rendererClass=&NiagaraSpriteRendererProperties}
   {rendererClass=&NiagaraMeshRendererProperties, rendererIndex=1}
  ]
 }
 Minimal001 {
  bEnabled=false
  emitterUpdateScript [
   &EmitterState(Life Cycle Mode=::NewEnumerator1, Loop Behavior=::NewEnumerator1, Loop Duration=0.3, UseLoopDelay=false)
   &SpawnRate(SpawnRate=50, Use Spawn Probability=false)
   &SpawnBurst_Instantaneous(Spawn Count=10, Use Spawn Probability=false, UseLoopCountLimit=false) {enabled=false}
  ]
  particleSpawnScript [
   &InitializeParticle(Sprite Size Mode=::NewEnumerator3, Uniform Sprite Size=15, UsePositionOffset=false, Write Material Random=true)
   &PartitionParticles(Number Of Partitions=&MakeIntFromFloat(Float=&FloatFromCurve(
      FloatCurve=DataInterface() {propertyValues {Curve {keys [<at 0 2 RCIM_Constant>,<at 0.05 4 RCIM_Constant>,<at 0.1 6 RCIM_Constant>,<at 0.15 8 RCIM_Constant>,<at 0.2 10 RCIM_Constant>]}}}
      CurveIndex=$Emitter.Age
     )))
   &CameraOffset(Camera Offset Amount=-1) {enabled=false}
   &SpriteFacingAndAlignment(
    Sprite Alignment=&RotateVectorByQuaternion(Vector To Rotate=<0 0 1>, Quaternion=$Particles.SkeletalMeshLocation.SampledRotation)
    WriteAlignment=true
    WriteFacing=true
   ) {enabled=false}
   &SkeletalMeshLocation(
    Bone / Socket Index=$Particles.PartitionParticles.Partition
    Bone Sampling Mode=::NewEnumerator4
    MaskSpawnGroup=false
    Sampled Position Offset=&MakeVector(
     X=-1.6
     Y=&MergeHelper_FloatSwitch(0=-2, 1=2, 2=?, 3=?, Index=&ModuloInt(A=$Particles.PartitionParticles.Partition, B=2), Num=::NewEnumerator0)
     Z=-3
    )
    Skeletal Mesh=$System.Skeletal Mesh
   ) {enabled=false, module=SkeletalMeshLocation001}
   &AddVelocity(
    Velocity=&MakeVector(
     X=-1
     Y=&MergeHelper_FloatSwitch(0=-2, 1=2, 2=?, 3=?, Index=&ModuloInt(A=$Particles.PartitionParticles.Partition, B=2), Num=::NewEnumerator0)
     Z=1
    )
    Velocity Speed Scale=50
   )
   &AddVelocity(Velocity=<0 0 1>) {enabled=false, module=AddVelocity001}
  ]
  particleUpdateScript [
   &ParticleState()
   &SkeletalMeshLocation(
    Bone / Socket Index=$Particles.PartitionParticles.Partition
    Bone Sampling Mode=::NewEnumerator4
    MaskSpawnGroup=false
    Position Sampling=::NewEnumerator1
    Sampled Position Offset Coordinate Space:ENiagaraSkelMeshTransforms=::NewEnumerator2
    Skeletal Mesh=$System.Skeletal Mesh
    Source Movement Scale:Vector3f=&VectorFromFloat(Value=&FloatFromCurve(FloatCurve=DataInterface() {propertyValues {Curve {keys [
          {interpMode=RCIM_Cubic, tangentMode=RCTM_User, time=0, value=1}
          {arriveTangent=0.0173, interpMode=RCIM_Cubic, leaveTangent=0.0173, tangentMode=RCTM_User, time=1, value=0.5}
         ]}}}))
   )
   &ScaleSpriteSize(
    Scale Factor=&MakeVector2D(
     X=1
     Y=&FloatFromCurve(
      FloatCurve=DataInterface() {propertyValues {Curve {keys [<at 0 0.3>,<at 1 1>]}}}
      Scale Curve=&Lerp_Float(
       A=17
       B=5
       Alpha=&FloatFromCurve(FloatCurve=DataInterface() {propertyValues {Curve {keys [<at 0 0>,<at 0.5 1>]}}}, CurveIndex=$Emitter.Age)
      )
     )
    )
    Scale Sprite Size Mode=::NewEnumerator2
   ) {enabled=false}
   &Drag(Drag=10, Use Linear Drag=true, Use Rotational Drag=true)
   &InheritSourceMovement(
    Applied Position Scale=&VectorFromFloat(Value=1)
    ApplyForce=false
    ApplyPosition=true
    ApplyPositionFalloff=false
   ) {enabled=false}
   &SolveForcesAndVelocity(Clamp Velocity=false, Limit Acceleration=false)
   &SubUVAnimation(
    "Sprite Renderer "=DataInterface() {propertyValues {SpriteRenderer=&self:Minimal001_0.NiagaraSpriteRendererProperties_0}}
    UseEndFrame=false
    UseStartFrame=false
   )
   &ScaleColor(
    Scale Alpha=&FloatFromCurve(FloatCurve=DataInterface() {propertyValues {Curve {keys [<at 0 1>,<at 1 0>]}}}, Scale Curve=0.2)
    Scale RGB=&MakeVectorFromLinearColorRGB(Color=&ColorFromCurve(ColorCurve=DataInterface() {propertyValues {
        AlphaCurve {keys [<at 0 1 RCIM_Cubic>]}
        BlueCurve {keys [<at 0.2782 0.705 RCIM_Cubic>,<at 1 1 RCIM_Cubic>]}
        GreenCurve {keys [<at 0.2782 0.45 RCIM_Cubic>,<at 1 1 RCIM_Cubic>]}
        RedCurve {keys [<at 0.2782 3 RCIM_Cubic>,<at 1 1 RCIM_Cubic>]}
       }}))
    ScaleA=true
    ScaleRGB=true
   )
   &ScaleColor(
    Scale Alpha=&MergeHelper_FloatSwitch(
     0=&FloatFromCurve(FloatCurve=DataInterface() {propertyValues {Curve {keys [<at 0 0>,<at 0.05 1>]}}}, CurveIndex=$Emitter.Age)
     1=&FloatFromCurve(FloatCurve=DataInterface() {propertyValues {Curve {keys [<at 0.05 0>,<at 0.1 1>]}}}, CurveIndex=$Emitter.Age)
     2=&FloatFromCurve(FloatCurve=DataInterface() {propertyValues {Curve {keys [<at 0.1 0>,<at 0.15 1>]}}}, CurveIndex=$Emitter.Age)
     3=&FloatFromCurve(FloatCurve=DataInterface() {propertyValues {Curve {keys [<at 0.15 0>,<at 0.2 1>]}}}, CurveIndex=$Emitter.Age)
     4=&FloatFromCurve(FloatCurve=DataInterface() {propertyValues {Curve {keys [<at 0.2 0>,<at 0.25 1>]}}}, CurveIndex=$Emitter.Age)
     Index=&ModuloInt(A=$Particles.PartitionParticles.Partition, B=5)
     Num=::NewEnumerator3
    )
    ScaleA=true
    ScaleRGB=true
   ) {enabled=false, module=ScaleColor001}
  ]
  rendererClasses [&NiagaraSpriteRendererProperties]
  renderers [
   {rendererClass=&NiagaraSpriteRendererProperties}
   {rendererClass=&NiagaraSpriteRendererProperties, rendererIndex=1}
  ]
 }
 Minimal002 {
  bEnabled=false
  emitterUpdateScript [
   &EmitterState(
    Delay First Loop Only=true
    Inactive Response=::NewEnumerator1
    Life Cycle Mode=::NewEnumerator1
    Loop Behavior=::NewEnumerator2
    Loop Count=6
    Loop Delay=0.5
    Loop Duration=0.1
    UseLoopDelay=true
   )
   &SpawnRate(SpawnRate=20, Use Spawn Probability=false) {enabled=false}
   &SpawnBurst_Instantaneous(Spawn Count=2, Use Spawn Probability=false, UseLoopCountLimit=false)
  ]
  particleSpawnScript [
   &InitializeParticle(
    Color=<rgba 4.8786 2.4 6 1>
    Mesh Scale Mode=::NewEnumerator3
    Mesh Uniform Scale=0.1
    Ribbon Width=2
    Ribbon Width Mode=::NewEnumerator1
    Sprite Size Mode=::NewEnumerator3
    Uniform Sprite Size=5
    UsePositionOffset=false
    Write Material Random=true
   )
   &PartitionParticles(
    Particles Per Partition=2
    Partition Count Mode:ENiagara_InfiniteLoopDuration=::NewEnumerator1
    Partition Mode:ENiagara_PartitionMode_v2=::NewEnumerator1
    Write Partition Index to Ribbon Link Order:ENiagara_AttributeSamplingApplyOutput=::NewEnumerator0
    Write Partition to Ribbon ID=::NewEnumerator0
   )
   &PartitionParticles(
    Number Of Partitions=4
    Particles Per Partition=1
    Partition Mode:ENiagara_PartitionMode_v2=::NewEnumerator1
    Write Partition Index to Ribbon Link Order:ENiagara_AttributeSamplingApplyOutput=::NewEnumerator0
    Write Partition to Ribbon ID=::NewEnumerator0
   ) {enabled=false, module=PartitionParticles001}
  ]
  particleUpdateScript [
   &ParticleState()
   &UpdateMeshOrientation()
   &SkeletalMeshLocation(
    Bone / Socket Index=&ReturnExecIndex()
    MaskSpawnGroup=false
    Mesh Sampling Type=::NewEnumerator5
    Sampled Position Offset=&Add_Vector(
     A=&MakeVector(
      X=&Lerp_Float(
       A=0
       B=0
       Alpha=&FloatFromCurve(FloatCurve=DataInterface() {propertyValues {Curve {defaultValue=1, keys [{tangentMode=RCTM_User, time=0, value=0}, {tangentMode=RCTM_User, time=1, value=1}]}}})
      )
      Y=0
      Z=0
     )
     B=&RandomRangeVector(Minimum=<0 0 0>, Maximum=<30 0 0>)
    )
    Skeletal Mesh=$System.Skeletal Mesh2
    Socket Sampling Mode=::NewEnumerator5
    Use Sampled Mesh Attribute as Ribbon ID=false
   )
   &CameraOffset(Camera Offset Amount=5)
   &ScaleRibbonWidth()
   /SetVariables(Particles.RibbonLinkOrder:NiagaraFloat=$Particles.PartitionParticles.Normalized Index In Partition) {enabled=false, isSetParameters=true}
   /SetVariables(Particles.NewRibbonOrder:NiagaraFloat=$Particles.PartitionParticles.Normalized Index In Partition) {isSetParameters=true}
   &DynamicMaterialParameters(Module.Index0Float0Write=true, Module.Index0Float1Write=true, Module.Index0Float2Write=true, Module.Index0Float3Write=true)
   &ScaleColor(
    Scale RGB=&VectorFromFloat(Value=&FloatFromCurve(
      FloatCurve=DataInterface() {propertyValues {Curve {
         defaultValue=1
         keys [
          {arriveTangent=9.0725, interpMode=RCIM_Cubic, leaveTangent=9.0725, tangentMode=RCTM_User, time=0, value=0}
          {arriveTangent=-2.5249, interpMode=RCIM_Cubic, leaveTangent=-2.5249, tangentMode=RCTM_User, time=0.1408, value=0.9847}
          {arriveTangent=-0.0405, interpMode=RCIM_Cubic, leaveTangent=-0.0405, tangentMode=RCTM_User, time=1, value=0}
         ]
        }}}
      Scale Curve=10
     ))
    ScaleA=true
    ScaleRGB=true
   )
  ]
  rendererClasses [
   &NiagaraSpriteRendererProperties
   &NiagaraRibbonRendererProperties
   &NiagaraMeshRendererProperties
  ]
  renderers [
   {rendererClass=&NiagaraSpriteRendererProperties}
   {rendererClass=&NiagaraRibbonRendererProperties, rendererIndex=1}
   {rendererClass=&NiagaraRibbonRendererProperties, rendererIndex=2}
   {rendererClass=&NiagaraMeshRendererProperties, rendererIndex=3}
   {rendererClass=&NiagaraMeshRendererProperties, rendererIndex=4}
  ]
 }
 Minimal003 {
  emitterUpdateScript [
   &EmitterState(Life Cycle Mode=::NewEnumerator1, Loop Behavior=::NewEnumerator1, Loop Delay=$System.LightningDelay, Loop Duration=1.2, UseLoopDelay=true)
   &SpawnBurst_Instantaneous(Spawn Count=4, Use Spawn Probability=false, UseLoopCountLimit=false)
  ]
  particleSpawnScript [
   &InitializeParticle(
    Color=<rgba 1.4294 0 2 1>
    Lifetime=1.2
    Sprite Size=<30 10>
    Sprite Size Mode=::NewEnumerator4
    UsePositionOffset=false
    Write Material Random=true
   )
   &PartitionParticles(Number Of Partitions=4)
   /SetVariables(Particles.PartitionBool:NiagaraBool=&SetBoolByIntComparison(A=$Particles.PartitionParticles.Partition, B=1, Comparison Type=::NewEnumerator4)) {isSetParameters=true, module=SetVariables_#5}
   /SetVariables(Particles.RandomBone:NiagaraInt32=&MakeCustomIntFromBool(Bool=$Particles.PartitionBool, True INT=0, False INT=2)) {isSetParameters=true, module=SetVariables_#6}
   &CameraOffset(Camera Offset Amount=10)
  ]
  particleUpdateScript [
   &ParticleState()
   &SkeletalMeshLocation(
    Bone / Socket Index=$Particles.PartitionParticles.Partition
    MaskSpawnGroup=false
    Mesh Sampling Type=::NewEnumerator5
    Sampled Position Offset=&SelectVectorFromArray(
     Direct Array Index=$Particles.PartitionParticles.Partition
     Vector Selection Array=DataInterface() {propertyValues {FloatData [<15 1.5 -2>, <15 1.5 2>, <15 -1.5 -2>, <15 -1.5 2>]}}
    )
    Skeletal Mesh=$System.Skeletal Mesh2
    Socket Sampling Mode=::NewEnumerator5
   )
   &SpriteFacingAndAlignment(
    Alignment Coordinate Space=::World
    Facing Coordinate Space=::World
    Sprite Alignment=&RotateVectorByQuaternion(Vector To Rotate=<0 0 1>, Quaternion=$Particles.SkeletalMeshLocation.SampledRotation)
    Sprite Facing=&RotateVectorByQuaternion(Vector To Rotate=<0 1 0>, Quaternion=$Particles.SkeletalMeshLocation.SampledRotation)
    WriteAlignment=true
    WriteFacing=true
   )
   &ScaleColor(
    Scale Alpha=&FloatFromCurve(
     FloatCurve=DataInterface() {propertyValues {Curve {
        defaultValue=0
        keys [
         {interpMode=RCIM_Cubic, tangentMode=RCTM_User, time=0, value=0}
         {arriveTangent=0.1167, interpMode=RCIM_Cubic, leaveTangent=0.1167, tangentMode=RCTM_User, time=0.2, value=1}
         {interpMode=RCIM_Cubic, tangentMode=RCTM_User, time=0.8, value=1}
         {tangentMode=RCTM_User, time=1, value=0}
        ]
       }}}
     Scale Curve=&Waveform("Waveform [1]"=::NewEnumerator12, "[1] Frequency"=10, "[1] Amplitude Min/Max"=<0 1>, "[1] Bias"=0.5)
    )
    ScaleA=true
    ScaleRGB=true
   )
  ]
  rendererClasses [&NiagaraSpriteRendererProperties]
  renderers [{rendererClass=&NiagaraSpriteRendererProperties}]
 }
 Minimal004 {
  emitterUpdateScript [
   &EmitterState(Life Cycle Mode=::NewEnumerator1, Loop Behavior=::NewEnumerator1, Loop Delay=1.3, UseLoopDelay=true)
   &SpawnBurst_Instantaneous(Spawn Count=4, Use Spawn Probability=false, UseLoopCountLimit=false)
  ]
  particleSpawnScript [
   &InitializeParticle(
    Color=<rgba 2.8588 0 4 1>
    Lifetime=0.5
    Sprite Size=<30 10>
    Sprite Size Mode=::NewEnumerator4
    UsePositionOffset=false
    Write Material Random=true
   )
   &PartitionParticles(Number Of Partitions=4)
   /SetVariables(Particles.PartitionBool:NiagaraBool=&SetBoolByIntComparison(A=$Particles.PartitionParticles.Partition, B=1, Comparison Type=::NewEnumerator4)) {isSetParameters=true, module=SetVariables_#5}
   /SetVariables(Particles.RandomBone:NiagaraInt32=&MakeCustomIntFromBool(Bool=$Particles.PartitionBool, True INT=0, False INT=2)) {isSetParameters=true, module=SetVariables_#6}
   &CameraOffset(Camera Offset Amount=10)
  ]
  particleUpdateScript [
   &ParticleState()
   &SkeletalMeshLocation(
    Bone / Socket Index=$Particles.PartitionParticles.Partition
    MaskSpawnGroup=false
    Mesh Sampling Type=::NewEnumerator5
    Sampled Position Offset=&SelectVectorFromArray(
     Direct Array Index=$Particles.PartitionParticles.Partition
     Vector Selection Array=DataInterface() {propertyValues {FloatData [<15 1.5 -2>, <15 1.5 2>, <15 -1.5 -2>, <15 -1.5 2>]}}
    )
    Skeletal Mesh=$System.Skeletal Mesh2
    Socket Sampling Mode=::NewEnumerator5
   )
   &SpriteFacingAndAlignment(
    Alignment Coordinate Space=::World
    Facing Coordinate Space=::World
    Sprite Alignment=&RotateVectorByQuaternion(Vector To Rotate=<0 0 1>, Quaternion=$Particles.SkeletalMeshLocation.SampledRotation)
    Sprite Facing=&RotateVectorByQuaternion(Vector To Rotate=<0 1 0>, Quaternion=$Particles.SkeletalMeshLocation.SampledRotation)
    WriteAlignment=true
    WriteFacing=true
   )
   &ScaleColor(
    Scale Alpha=&FloatFromCurve(FloatCurve=DataInterface() {propertyValues {Curve {
        defaultValue=0
        keys [
         {interpMode=RCIM_Cubic, tangentMode=RCTM_User, time=0, value=0}
         {arriveTangent=0.1167, interpMode=RCIM_Cubic, leaveTangent=0.1167, tangentMode=RCTM_User, time=0.2, value=1}
         {interpMode=RCIM_Cubic, tangentMode=RCTM_User, time=0.8, value=1}
         {tangentMode=RCTM_User, time=1, value=0}
        ]
       }}})
    ScaleA=true
    ScaleRGB=true
   )
  ]
  rendererClasses [&NiagaraSpriteRendererProperties]
  renderers [{rendererClass=&NiagaraSpriteRendererProperties}]
 }
 NE_LoopingZaps {
  emitterUpdateScript [
   &EmitterState(
    Inactive Response=::NewEnumerator1
    Life Cycle Mode=::NewEnumerator1
    Loop Behavior=::NewEnumerator1
    Loop Delay=1.4
    Loop Duration=0.5
    UseLoopDelay=true
   )
   &SpawnRate(SpawnRate=10, Use Spawn Probability=false)
   &SpawnBurst_Instantaneous(
    Spawn Count=&RandomRangeInt(Minimum=1, Maximum=2)
    Use Spawn Probability=false
    UseLoopCountLimit=false
   )
  ]
  particleSpawnScript [
   &InitializeParticle(
    Color=&RandomRangeLinearColor(Minimum=<rgba 1.6262 0.8 2 0.4>, Maximum=<rgba 0.6951 0 4 0.7>)
    Lifetime=&RandomRangeFloat(Minimum=0.15, Maximum=0.25)
    Mesh Scale Mode=::NewEnumerator1
    Mesh Uniform Scale Max=0.6
    Mesh Uniform Scale Min=0.4
    Position Offset=<50 0 10>
    Sprite Size Mode=::NewEnumerator3
    Uniform Sprite Size=20
    UsePositionOffset=true
    Write Material Random=true
   )
   &InitialMeshOrientation(Use Rotation Vector=true)
   &CameraOffset(Camera Offset Amount=-10)
  ]
  particleUpdateScript [
   &ParticleState()
   &ScaleMeshSize(Scale Factor=&Multiply_VectorByFloat(
     Vector=&RandomRangeVector(Minimum=<1 1 2.2>, Maximum=<1.3 1.3 2.2>)
     Float=&FloatFromCurve(FloatCurve=DataInterface() {propertyValues {Curve {keys [<at 0 1>,<at 1 1.5>]}}})
    )) {module=ScaleMeshSize001}
   &DynamicMaterialParameters(
    Index 0 Param 2=&RandomRangeFloat(Minimum=0, Maximum=0)
    Index 0 Param 4=13
    Index 1 Param 1=&FloatFromCurve(FloatCurve=DataInterface() {propertyValues {Curve {keys [<at 0 0.5>,<at 1 1>]}}})
    Index 1 Param 2=2
    Index 1 Param 3=&FloatFromCurve(FloatCurve=DataInterface() {propertyValues {Curve {keys [{interpMode=RCIM_Cubic, tangentMode=RCTM_Break, time=0.5, value=-0.5}, <at 1 0.1>]}}})
    Index 2 Param 1=&Multiply_Float(
     A=&FloatFromCurve(FloatCurve=DataInterface() {propertyValues {Curve {keys [{interpMode=RCIM_Cubic, leaveTangent=8.1085, tangentMode=RCTM_Break, time=0.8568, value=1}, {tangentMode=RCTM_User, time=1, value=0}]}}})
     B=&Divide_Float(A=&MakeFloatFromVector(Vector=$Particles.Scale, Channel=::NewEnumerator0), B=40)
    )
    Index 2 Param 2=&FloatFromCurve(
     FloatCurve=DataInterface() {propertyValues {Curve {keys [
         {interpMode=RCIM_Cubic, tangentMode=RCTM_Break, time=0.7, value=0}
         {arriveTangent=-0.0859, interpMode=RCIM_Cubic, tangentMode=RCTM_Break, time=1, value=-1}
        ]}}}
     Scale Curve=50
    )
    Index 2 Param 3=&Waveform(
     Global Input Position=&RandomRangeFloat(Minimum=0, Maximum=15)
     Global Phase=$Particles.Age
     "Waveform [1]"=::NewEnumerator12
     "[1] Phase Scale"=2
     "[1] Frequency"=1
     "[1] Amplitude Min/Max"=<0 1>
    )
    Index 2 Param 4=25
    Index 3 Param 1=&FloatFromCurve(
     FloatCurve=DataInterface() {propertyValues {Curve {keys [
         {interpMode=RCIM_Cubic, leaveTangent=-0.0025, tangentMode=RCTM_Break, time=0, value=0}
         {arriveTangent=1.8528, interpMode=RCIM_Cubic, tangentMode=RCTM_Break, time=1, value=1}
        ]}}}
     Scale Curve=&RandomRangeFloat(Minimum=0, Maximum=0)
    )
    Index 3 Param 2=&FloatFromCurve(
     FloatCurve=DataInterface() {propertyValues {Curve {keys [
         {interpMode=RCIM_Cubic, leaveTangent=-0.0025, tangentMode=RCTM_Break, time=0, value=0.5}
         {arriveTangent=-0.0316, interpMode=RCIM_Cubic, tangentMode=RCTM_Break, time=1, value=-0.2}
        ]}}}
     Scale Curve=0.1
    )
    Index 3 Param 3=&FloatFromCurve(FloatCurve=DataInterface() {propertyValues {Curve {keys [
         {interpMode=RCIM_Cubic, leaveTangent=1.1034, tangentMode=RCTM_Break, time=0, value=-1}
         {arriveTangent=2.2399, tangentMode=RCTM_User, time=0.7, value=0.125}
        ]}}})
    Index 3 Param 4=&FloatFromCurve(FloatCurve=DataInterface() {propertyValues {Curve {keys [
         {interpMode=RCIM_Cubic, leaveTangent=3.0601, tangentMode=RCTM_Break, time=0.5, value=0}
         {arriveTangent=3.2053, tangentMode=RCTM_User, time=1, value=1}
        ]}}})
    Module.Index0Float0Write=true
    Module.Index0Float1Write=true
    Module.Index0Float2Write=true
    Module.Index0Float3Write=true
    Module.Index1Float0Write=true
    Module.Index1Float1Write=true
    Module.Index1Float2Write=true
    Module.Index1Float3Write=true
    Module.Index2Float0Write=true
    Module.Index2Float1Write=true
    Module.Index2Float2Write=true
    Module.Index2Float3Write=true
    Module.Index3Float0Write=true
    Module.Index3Float1Write=true
    Module.Index3Float2Write=true
    Module.Index3Float3Write=true
    Write Parameter Index 1=::NewEnumerator1
    Write Parameter Index 2=::NewEnumerator1
    Write Parameter Index 3=::NewEnumerator1
   )
   &ScaleColor(
    Color Scale:LinearColor=&ColorFromCurve(ColorCurve=DataInterface() {propertyValues {
       AlphaCurve {keys [<at 0 1 RCIM_Cubic>,<at 0.7884 1>,<at 1 0 RCIM_Cubic>]}
       BlueCurve {keys [<at 0.2225 5>, {interpMode=RCIM_Cubic, leaveTangent=-19.7604, time=0.4906, value=1.5}]}
       GreenCurve {keys [<at 0.2225 5>, {interpMode=RCIM_Cubic, leaveTangent=-19.7604, time=0.4906, value=1.5}]}
       RedCurve {keys [<at 0.2225 5>, {interpMode=RCIM_Cubic, leaveTangent=-19.7604, time=0.4906, value=1.5}]}
      }})
    Scale Mode=3
   )
   /SetVariables(Particles.DecalColor:LinearColor=&MultiplyLinearColor(Linear Color=<rgba 0.4 0.4 0.4 1>, Scale Factor=&MakeVector4_FromLinearColor(LinearColor=$Particles.Color))) {isSetParameters=true}
   /SetVariables(Particles.GlowDecalSize:Vector2f=&Multiply_Vector2D_ByFloat(Float=40, Vector2D=<1 1>)) {isSetParameters=true}
   /SetVariables(Particles.DecalDynMat:Vector4f=<4 1 0 0>) {isSetParameters=true}
  ]
  rendererClasses [&NiagaraMeshRendererProperties, &NiagaraSpriteRendererProperties]
  renderers [
   {rendererClass=&NiagaraMeshRendererProperties, rendererIndex=0}
   {rendererClass=&NiagaraSpriteRendererProperties, rendererIndex=1}
   {rendererClass=&NiagaraMeshRendererProperties, rendererIndex=2}
  ]
 }
 SplitLightning {
  emitterUpdateScript [
   &EmitterState(Life Cycle Mode=::NewEnumerator1, Loop Behavior=::NewEnumerator1, Loop Delay=$System.LightningDelay, Loop Duration=1.2, UseLoopDelay=true)
   &SpawnBurst_Instantaneous(Spawn Count=4, Use Spawn Probability=false, UseLoopCountLimit=false) {enabled=false}
   &SpawnRate(SpawnRate=15, Use Spawn Probability=false)
  ]
  particleSpawnScript [
   &InitializeParticle(
    Color=&RandomRangeLinearColor(Minimum=<rgba 0.7 0.7 0.7 0.4>, Maximum=<rgba 1 1 1 0.6>)
    Lifetime=&RandomRangeFloat(Minimum=0.15, Maximum=0.3)
    Mesh Scale=&RandomRangeVector(Minimum=<0.025 0.025 0.025>, Maximum=<0.05 0.05 0.025>)
    Mesh Scale Mode=::NewEnumerator4
    Position=&RandomRangeVector(Minimum=<20 0 0>, Maximum=<20 0 0>)
    Position Mode=::NewEnumerator1
    UsePositionOffset=false
    Write Material Random=true
   )
   /SetVariables(Particles.Random:NiagaraBool=&RandomBool()) {isSetParameters=true}
   /SetVariables(Particles.RandomBone:NiagaraInt32=&MakeCustomIntFromBool(Bool=$Particles.Random, True INT=0, False INT=2)) {isSetParameters=true}
   &AddVelocity(Velocity=&RandomRangeVector(Minimum=<-0.2 0 -1>, Maximum=<0.2 0 -1>))
   &CameraOffset()
  ]
  particleUpdateScript [
   &ParticleState()
   &SkeletalMeshLocation(
    Bone / Socket Index=$Particles.RandomBone
    MaskSpawnGroup=false
    Mesh Sampling Type=::NewEnumerator5
    Sampled Position Offset=&MakeCustomVectorFromBool(
     Bool=$Particles.Random
     True Vector=&RandomRangeVector(Minimum=<10 1 0>, Maximum=<25 1 0>)
     False Vector=&RandomRangeVector(Minimum=<10 -1 0>, Maximum=<25 -1 0>)
    )
    Skeletal Mesh=$System.Skeletal Mesh2
    Socket Sampling Mode=::NewEnumerator5
   )
   &ScaleMeshSize(Scale Factor=&RandomRangeVector(Minimum=<2.5 2.5 4.5>, Maximum=<3 3 5.5>)) {module=ScaleMeshSize001}
   &ScaleMeshSize(Scale Factor=&MakeVector(
     X=1
     Y=1
     Z=&FloatFromCurve(FloatCurve=DataInterface() {propertyValues {Curve {keys [
          {interpMode=RCIM_Cubic, leaveTangent=0.0966, tangentMode=RCTM_Break, time=0, value=1}
          {arriveTangent=0.0951, interpMode=RCIM_Cubic, leaveTangent=-0.1745, tangentMode=RCTM_Break, time=1, value=1.07}
         ]}}})
    ))
   &DynamicMaterialParameters(
    Index 0 Param 4=&MakeCustomFloatFromBool(
     Bool=false
     True Float=3
     False Float=&SelectFloatFromArray(
      Direct Array Index=&ReturnParticleIDAttributes()
      Float Selection Array=DataInterface() {propertyValues {FloatData [1,0,9,3,10,9]}}
     )
    )
    Index 1 Param 1=1
    Index 1 Param 2=2
    Index 1 Param 3=&FloatFromCurve(FloatCurve=DataInterface() {propertyValues {Curve {keys [{interpMode=RCIM_Cubic, tangentMode=RCTM_Break, time=0.5, value=-0.5}, <at 1 0.1>]}}})
    Index 2 Param 1=&Multiply_Float(
     A=&FloatFromCurve(
      FloatCurve=DataInterface() {propertyValues {Curve {keys [{interpMode=RCIM_Cubic, leaveTangent=8.1085, tangentMode=RCTM_Break, time=0.8568, value=1}, {tangentMode=RCTM_User, time=1, value=0}]}}}
      Scale Curve=0.7
     )
     B=&Divide_Float(
      A=&MakeFloatFromVector(Vector=$Particles.Scale, Channel=::NewEnumerator0)
      B=&RandomRangeFloat(Minimum=10, Maximum=15)
     )
    )
    Index 2 Param 2=&FloatFromCurve(
     FloatCurve=DataInterface() {propertyValues {Curve {keys [
         {interpMode=RCIM_Cubic, tangentMode=RCTM_Break, time=0.7, value=0}
         {arriveTangent=-0.0859, interpMode=RCIM_Cubic, tangentMode=RCTM_Break, time=1, value=-1}
        ]}}}
     Scale Curve=50
    )
    Index 2 Param 3=&Waveform(
     Global Input Position=&RandomRangeFloat(Minimum=0, Maximum=15)
     Global Phase=$Particles.Age
     "Waveform [1]"=::NewEnumerator12
     "[1] Phase Scale"=2
     "[1] Frequency"=1
     "[1] Amplitude Min/Max"=<0 1>
    )
    Index 2 Param 4=25
    Index 3 Param 1=&FloatFromCurve(
     FloatCurve=DataInterface() {propertyValues {Curve {keys [
         {interpMode=RCIM_Cubic, leaveTangent=-0.0025, tangentMode=RCTM_Break, time=0, value=0}
         {arriveTangent=1.8528, interpMode=RCIM_Cubic, tangentMode=RCTM_Break, time=1, value=1}
        ]}}}
     Scale Curve=&RandomRangeFloat(Minimum=25, Maximum=35)
    )
    Index 3 Param 2=&FloatFromCurve(FloatCurve=DataInterface() {propertyValues {Curve {keys [
         {interpMode=RCIM_Cubic, leaveTangent=-0.0025, tangentMode=RCTM_Break, time=0, value=0.5}
         {arriveTangent=-0.0316, interpMode=RCIM_Cubic, tangentMode=RCTM_Break, time=1, value=-0.2}
        ]}}})
    Index 3 Param 3=&FloatFromCurve(FloatCurve=DataInterface() {propertyValues {Curve {keys [
         {interpMode=RCIM_Cubic, leaveTangent=1.1034, tangentMode=RCTM_Break, time=0, value=-1}
         {arriveTangent=2.2399, tangentMode=RCTM_User, time=0.5, value=0.125}
        ]}}})
    Index 3 Param 4=&FloatFromCurve(FloatCurve=DataInterface() {propertyValues {Curve {keys [
         {interpMode=RCIM_Cubic, leaveTangent=3.0601, tangentMode=RCTM_Break, time=0.5, value=-0.5}
         {arriveTangent=3.2053, tangentMode=RCTM_User, time=1, value=1}
        ]}}})
    Module.Index0Param0Write=true
    Module.Index0Param1Write=true
    Module.Index0Param2Write=true
    Module.Index0Param3Write=true
    Module.Index1Param0Write=true
    Module.Index1Param1Write=true
    Module.Index1Param2Write=true
    Module.Index1Param3Write=true
    Module.Index2Param0Write=true
    Module.Index2Param1Write=true
    Module.Index2Param2Write=true
    Module.Index2Param3Write=true
    Module.Index3Param0Write=true
    Module.Index3Param1Write=true
    Module.Index3Param2Write=true
    Module.Index3Param3Write=true
    Write Parameter Index 1=true
    Write Parameter Index 2=true
    Write Parameter Index 3=true
   )
   &SolveForcesAndVelocity(Clamp Velocity=false, Limit Acceleration=false)
   &ScaleColor(
    Scale Alpha=&FloatFromCurve(FloatCurve=DataInterface() {propertyValues {Curve {keys [
         {tangentMode=RCTM_User, time=0, value=1}
         {arriveTangent=-0.0405, interpMode=RCIM_Cubic, leaveTangent=-0.0405, tangentMode=RCTM_User, time=0.7009, value=1}
         {arriveTangent=-8.0089, interpMode=RCIM_Cubic, leaveTangent=-8.0089, tangentMode=RCTM_User, time=1, value=0}
        ]}}})
    Scale RGB=&Lerp_Vector(
     A=&VectorFromFloat(Value=&FloatFromCurve(FloatCurve=DataInterface() {propertyValues {Curve {keys [<at 0.2771 1>,<at 0.3786 1>]}}}, Scale Curve=20))
     B=&VectorFromFloat(Value=&RandomRangeFloat(Minimum=0.3, Maximum=2, Evaluation Type=::NewEnumerator1))
     Alpha=&FloatFromCurve(FloatCurve=DataInterface() {propertyValues {Curve {keys [
          {interpMode=RCIM_Cubic, leaveTangent=2.5294, tangentMode=RCTM_Break, time=0, value=0}
          {arriveTangent=-0.2166, interpMode=RCIM_Cubic, tangentMode=RCTM_Break, time=0.4, value=1}
         ]}}})
    )
    ScaleA=true
    ScaleRGB=true
   )
   /SetVariables(Particles.MeshOrientation=&MultiplyQuaternion(
     Quaternion A=&MultiplyQuaternion(
      Quaternion A=&MakeQuatFromAxisAngle(Axis=<0 0 1>, AngleInDegrees=0)
      Quaternion B=&FindQuatBetweenTwoVectors(From Vector=<0 0 1>, To Vector=$Particles.Velocity)
     )
     Quaternion B=<0 0 0 1>
    )) {isSetParameters=true}
   &ScaleColor(
    Scale RGB=&VectorFromFloat(Value=&MakeCustomFloatFromBool(Bool=$User.bIsFrontEndPreview, True Float=0.6, False Float=1))
    ScaleA=true
    ScaleRGB=true
   ) {enabled=false, module=ScaleColor001}
   &ScaleColor(Scale RGB=&VectorFromFloat(Value=$System.SMTOD), ScaleA=true, ScaleRGB=true) {module=ScaleColor002}
  ]
  rendererClasses [&NiagaraMeshRendererProperties]
  renderers [
   {rendererClass=&NiagaraMeshRendererProperties, rendererIndex=0}
   {rendererClass=&NiagaraMeshRendererProperties, rendererIndex=1}
  ]
 }
 VentSmoke {
  bEnabled=false
  emitterUpdateScript [
   &EmitterState(Life Cycle Mode=::NewEnumerator1, Loop Behavior=::NewEnumerator1, Loop Delay=0.05, Loop Duration=0.2, UseLoopDelay=true)
   &SpawnRate(SpawnRate=50, Use Spawn Probability=false)
   &SpawnBurst_Instantaneous(Spawn Count=10, Use Spawn Probability=false, UseLoopCountLimit=false) {enabled=false}
   /SetVariables(Emitter.ExhaustLifetime:NiagaraFloat=&FloatFromCurve(FloatCurve=DataInterface() {propertyValues {Curve {keys [<at 0 0>,<at 0.5 1>]}}}, CurveIndex=$Emitter.Age)) {isSetParameters=true, module=SetVariables_#7}
  ]
  particleSpawnScript [
   &InitializeParticle(
    Color=&FNTimeOfDayColor(Night Color=<rgba 0.7 0.7958 1 2>, Morning Color=<rgba 1 1 1 1>, Day Color=<rgba 1 1 1 1>, Evening Color=<rgba 1 1 1 1>)
    Lifetime=&RandomRangeFloat(Minimum=0.4, Maximum=0.5)
    Sprite Rotation Mode=::NewEnumerator2
    Sprite Size=&RandomRangeVector2D(Minimum=<5 5>, Maximum=<7 6>)
    Sprite Size Mode=::NewEnumerator4
    UsePositionOffset=false
    Write Material Random=true
   )
   &PartitionParticles(Number Of Partitions=&MakeIntFromFloat(Float=&FloatFromCurve(
      FloatCurve=DataInterface() {propertyValues {Curve {keys [<at 0 2 RCIM_Constant>,<at 0.05 4 RCIM_Constant>,<at 0.1 6 RCIM_Constant>,<at 0.15 8 RCIM_Constant>,<at 0.2 10 RCIM_Constant>]}}}
      CurveIndex=$Emitter.Age
     ))) {enabled=false, module=PartitionParticles001}
   &PartitionParticles(Number Of Partitions=10)
   &AddVelocity(Velocity=&MakeVector(
     X=-1
     Y=&MergeHelper_FloatSwitch(0=-0.7, 1=0.7, 2=?, 3=?, Index=&ModuloInt(A=$Particles.PartitionParticles.Partition, B=2), Num=::NewEnumerator0)
     Z=-0.1
    ))
   /SetVariables(Particles.PartitionParticles.Partition:NiagaraInt32=&MakeIntFromFloat(Float=&ClampFloat(
      Float=&MakeFloatFromInt(Int=&Add_Integer(
        A=$Particles.PartitionParticles.Partition
        B=&MakeIntFromFloat(Float=&FloatFromCurve(
          FloatCurve=DataInterface() {propertyValues {Curve {keys [<at 0.5 0 RCIM_Constant>,<at 0.6 2 RCIM_Constant>,<at 0.7 4 RCIM_Constant>,<at 0.8 6 RCIM_Constant>,<at 0.9 8 RCIM_Constant>,<at 1 10 RCIM_Constant>]}}}
          CurveIndex=$Emitter.Age
         ))
       ))
      Max=10
     ))) {enabled=false, isSetParameters=true, module=SetVariables_#8}
  ]
  particleUpdateScript [
   &SkeletalMeshLocation(
    Bone / Socket Index=$Particles.PartitionParticles.Partition
    MaskSpawnGroup=false
    Mesh Sampling Type=::NewEnumerator5
    Sampled Position Offset Coordinate Space:ENiagaraSkelMeshTransforms=::NewEnumerator2
    Skeletal Mesh=$System.Skeletal Mesh
    Socket Sampling Mode=::NewEnumerator5
   )
   /SetVariables(Particles.SpriteSizePuff:Vector2f=&Multiply_Vector2D_ByFloat(
     Float=&Multiply_Float(
      A=25
      B=&FloatFromCurve(FloatCurve=DataInterface() {propertyValues {Curve {keys [{interpMode=RCIM_Cubic, tangentMode=RCTM_User, time=0, value=0.1}, {interpMode=RCIM_Cubic, tangentMode=RCTM_User, time=1, value=1}]}}})
     )
     Vector2D=&Vector2DFromFloat(Value=&Lerp_Float(
       A=1
       B=0.5
       Alpha=&FloatFromCurve(FloatCurve=DataInterface() {propertyValues {Curve {keys [<at 0 0>,<at 1 1>]}}}, CurveIndex=$Emitter.Age)
      ))
    )) {isSetParameters=true}
   /SetVariables(Particles.PositionStream:NiagaraPosition=&AddVectorToPosition(
     Position=$Particles.Position
     Vector=&Multiply_VectorByFloat(
      Vector=&NormalizeVector(Vector To Normalize=$Particles.Initial.Velocity, Safe Normalize=true, Safe Fallback Threshold=0, Safe Fallback Vector=<0 0 0.001>)
      Float=&Lerp_Float(
       A=0
       B=10
       Alpha=&FloatFromCurve(
        FloatCurve=DataInterface() {propertyValues {Curve {keys [<at 0 0>,<at 1 1>]}}}
        Scale Curve=&RandomRangeFloat(Minimum=0.9, Maximum=1.1)
       )
      )
     )
    )) {isSetParameters=true}
   /SetVariables(Particles.PositionPuff:NiagaraPosition=&AddVectorToPosition(
     Position=$Particles.Position
     Vector=&Multiply_VectorByFloat(
      Vector=&NormalizeVector(Vector To Normalize=$Particles.Initial.Velocity, Safe Normalize=true, Safe Fallback Threshold=0, Safe Fallback Vector=<0 0 0.001>)
      Float=&Multiply_Float(
       A=&Multiply_Float(
        A=&Lerp_Float(
         A=13
         B=30
         Alpha=&FloatFromCurve(
          FloatCurve=DataInterface() {propertyValues {Curve {keys [<at 0 0>,<at 1 1>]}}}
          Scale Curve=&RandomRangeFloat(Minimum=0.9, Maximum=1.1)
         )
        )
        B=&Lerp_Float(A=1.5, B=1, Alpha=$Emitter.ExhaustLifetime)
       )
       B=&Lerp_Float(
        A=1.1
        B=0.5
        Alpha=&Divide_Float(
         A=&MakeFloatFromInt(Int=/Niagara/DynamicInputs/Divide/Divide_Int(A:NiagaraInt32=$Particles.PartitionParticles.Partition, B:NiagaraInt32=2))
         B=5
        )
       )
      )
     )
    )) {isSetParameters=true}
   &ParticleState()
   &ScaleSpriteSize(
    Scale Factor=&MakeVector2D(
     X=1
     Y=&Lerp_Float(
      A=1.1
      B=0.5
      Alpha=&Divide_Float(
       A=&MakeFloatFromInt(Int=&Divide_Int(A:NiagaraInt32=$Particles.PartitionParticles.Partition, B:NiagaraInt32=2))
       B=5
      )
     )
    )
    Scale Sprite Size Mode=::NewEnumerator2
   ) {module=ScaleSpriteSize001}
   &ScaleSpriteSize(
    Scale Factor=&MakeVector2D(
     X=&FloatFromCurve(FloatCurve=DataInterface() {propertyValues {Curve {keys [
          {interpMode=RCIM_Cubic, tangentMode=RCTM_User, time=0, value=0.5}
          {arriveTangent=0.7401, interpMode=RCIM_Cubic, leaveTangent=0.7401, tangentMode=RCTM_User, time=1, value=1}
         ]}}})
     Y=&FloatFromCurve(
      FloatCurve=DataInterface() {propertyValues {Curve {keys [{interpMode=RCIM_Cubic, tangentMode=RCTM_User, time=0, value=0.2}, {interpMode=RCIM_Cubic, tangentMode=RCTM_User, time=1, value=1}]}}}
      Scale Curve=&Lerp_Float(A=17, B=5, Alpha=$Emitter.ExhaustLifetime)
     )
    )
    Scale Sprite Size Mode=::NewEnumerator2
   )
   &Drag(Use Linear Drag=true, Use Rotational Drag=true) {enabled=false}
   &SolveForcesAndVelocity(Clamp Velocity=false, Limit Acceleration=false)
   &SubUVAnimation(
    "Sprite Renderer "=DataInterface() {propertyValues {SpriteRenderer=&self:NiagaraEmitter_0.NiagaraSpriteRendererProperties_0}}
    UseEndFrame=false
    UseStartFrame=false
   )
   &ScaleColor(
    Scale Alpha=&FloatFromCurve(
     FloatCurve=DataInterface() {propertyValues {Curve {keys [
         {interpMode=RCIM_Cubic, tangentMode=RCTM_User, time=0, value=0}
         {interpMode=RCIM_Cubic, tangentMode=RCTM_User, time=0.15, value=1}
         {interpMode=RCIM_Cubic, tangentMode=RCTM_User, time=1, value=0}
        ]}}}
     Scale Curve=&RandomRangeFloat(Minimum=0.7, Maximum=1)
    )
    Scale RGB=&Lerp_Vector(
     A=&MakeVectorFromLinearColorRGB(Color=&ColorFromCurve(ColorCurve=DataInterface() {propertyValues {
         AlphaCurve {keys [<at 0 1 RCIM_Cubic>]}
         BlueCurve {keys [<at 0 4 RCIM_Cubic>,<at 0.8503 0.5 RCIM_Cubic>]}
         GreenCurve {keys [<at 0 0.4 RCIM_Cubic>,<at 0.8503 0.5 RCIM_Cubic>]}
         RedCurve {keys [<at 0 2.3992 RCIM_Cubic>,<at 0.8503 0.6 RCIM_Cubic>]}
        }}))
     B=<0.6 0.5 0.5>
     Alpha=&FloatFromCurve(
      FloatCurve=DataInterface() {propertyValues {Curve {keys [
          {interpMode=RCIM_Cubic, tangentMode=RCTM_User, time=0.1, value=0}
          {arriveTangent=2.3957, interpMode=RCIM_Cubic, leaveTangent=2.3957, tangentMode=RCTM_User, time=0.8, value=1}
         ]}}}
      CurveIndex=$Emitter.Age
     )
    )
    ScaleA=true
    ScaleRGB=true
   )
   &ScaleColor(
    Scale Alpha=&MakeCustomFloatFromBool(
     Bool=&SetBoolByFloatComparison(
      A=&MakeFloatFromInt(Int=$Particles.PartitionParticles.Partition)
      B=&FloatFromCurve(
       FloatCurve=DataInterface() {propertyValues {Curve {keys [<at 0 2 RCIM_Constant>,<at 0.08 4 RCIM_Constant>,<at 0.16 6 RCIM_Constant>,<at 0.24 8 RCIM_Constant>,<at 0.32 10 RCIM_Constant>]}}}
       CurveIndex=$Emitter.Age
      )
      Comparison Type=::NewEnumerator7
     )
     True Float=1
     False Float=0
    )
    ScaleA=true
    ScaleRGB=false
   ) {module=ScaleColor001}
   &ScaleColor(
    Scale Alpha=&ClampFloat(Float=&Subtract_Float(
      A=&Add_Float(
       A=&MakeFloatFromInt(Int=&Divide_Int(A:NiagaraInt32=$Particles.PartitionParticles.Partition, B:NiagaraInt32=2))
       B=1
      )
      B=&Lerp_Float(
       A=0
       B=5
       Alpha=&FloatFromCurve(FloatCurve=DataInterface() {propertyValues {Curve {keys [<at 0.5 0>,<at 1 1>]}}}, CurveIndex=$Emitter.Age)
      )
     ))
    ScaleA=true
    ScaleRGB=false
   ) {enabled=false, module=ScaleColor002}
   &ScaleColor(
    Scale Alpha=&FloatFromCurve(
     FloatCurve=DataInterface() {propertyValues {Curve {keys [<at 0 1>,<at 1 0>]}}}
     CurveIndex=&Divide_Float(
      A=&MakeFloatFromInt(Int=&Divide_Int(A:NiagaraInt32=$Particles.PartitionParticles.Partition, B:NiagaraInt32=2))
      B=5
     )
    )
    ScaleA=true
    ScaleRGB=true
   ) {enabled=false, module=ScaleColor003}
   /SetVariables(Particles.ColorPuff:LinearColor=&MakeLinearColorFromVectorAndFloat(
     "Vector (RGB)"=&MakeVectorFromLinearColorRGB(Color=&LerpLinearColors(
       StartColor=$Particles.Color
       EndColor=<rgba 0.6 0.5 0.5 1>
       LerpFactor=&FloatFromCurve(
        FloatCurve=DataInterface() {propertyValues {Curve {keys [{interpMode=RCIM_Cubic, tangentMode=RCTM_User, time=0, value=0}, {interpMode=RCIM_Cubic, tangentMode=RCTM_User, time=1, value=1}]}}}
        CurveIndex=$Emitter.Age
       )
      ))
     "Float (Alpha)"=&Multiply_Float(
      A=&MakeFloatFromLinearColor(LinearColor=$Particles.Color, Channel:ENiagaraFloatFromLinearColorOptions=::NewEnumerator3)
      B=&FloatFromCurve(FloatCurve=DataInterface() {propertyValues {Curve {keys [{interpMode=RCIM_Cubic, tangentMode=RCTM_User, time=0.1, value=0}, {interpMode=RCIM_Cubic, tangentMode=RCTM_User, time=0.3, value=1}]}}})
     )
    )) {isSetParameters=true}
  ]
  rendererClasses [&NiagaraSpriteRendererProperties]
  renderers [
   {rendererClass=&NiagaraSpriteRendererProperties}
   {rendererClass=&NiagaraSpriteRendererProperties, rendererIndex=1}
  ]
 }
 VentSmoke001 {
  bEnabled=false
  emitterUpdateScript [
   &EmitterState(Life Cycle Mode=::NewEnumerator1, Loop Behavior=::NewEnumerator1, Loop Delay=0.05, Loop Duration=2, UseLoopDelay=false)
   &SpawnRate(
    SpawnRate=&FloatFromCurve(FloatCurve=DataInterface() {propertyValues {Curve {keys [<at 0.3 1>,<at 1 0.4>]}}}, CurveIndex=$Emitter.Age, Scale Curve=100)
    Use Spawn Probability=false
   ) {enabled=false}
   &SpawnBurst_Instantaneous(Spawn Count=10, Use Spawn Probability=false, UseLoopCountLimit=false)
   /SetVariables(Emitter.ExhaustLifetime:NiagaraFloat=&FloatFromCurve(FloatCurve=DataInterface() {propertyValues {Curve {keys [<at 0 0>,<at 0.5 1>]}}}, CurveIndex=$Emitter.Age)) {enabled=false, isSetParameters=true, module=SetVariables_#7}
  ]
  particleSpawnScript [
   &InitializeParticle(
    Lifetime=3.3
    Sprite Rotation Mode=::NewEnumerator2
    Sprite Size Mode=::NewEnumerator3
    Uniform Sprite Size=30
    UsePositionOffset=false
    Write Material Random=true
   )
   &PartitionParticles(Number Of Partitions=&MakeIntFromFloat(Float=&FloatFromCurve(
      FloatCurve=DataInterface() {propertyValues {Curve {keys [<at 0 2 RCIM_Constant>,<at 0.05 4 RCIM_Constant>,<at 0.1 6 RCIM_Constant>,<at 0.15 8 RCIM_Constant>,<at 0.2 10 RCIM_Constant>]}}}
      CurveIndex=$Emitter.Age
     ))) {enabled=false, module=PartitionParticles001}
   &PartitionParticles(Number Of Partitions=10)
   /SetVariables(Particles.PartitionParticles.Partition:NiagaraInt32=&MakeIntFromFloat(Float=&ClampFloat(
      Float=&MakeFloatFromInt(Int=&Add_Integer(
        A=$Particles.PartitionParticles.Partition
        B=&MakeIntFromFloat(Float=&FloatFromCurve(
          FloatCurve=DataInterface() {propertyValues {Curve {keys [<at 0.5 0 RCIM_Constant>,<at 0.6 2 RCIM_Constant>,<at 0.7 4 RCIM_Constant>,<at 0.8 6 RCIM_Constant>,<at 0.9 8 RCIM_Constant>,<at 1 10 RCIM_Constant>]}}}
          CurveIndex=$Emitter.Age
         ))
       ))
      Max=10
     ))) {enabled=false, isSetParameters=true, module=SetVariables_#8}
   &CameraOffset(Camera Offset Amount=-3)
  ]
  particleUpdateScript [
   &SkeletalMeshLocation(
    Bone / Socket Index=$Particles.PartitionParticles.Partition
    Bone Sampling Mode=::NewEnumerator4
    MaskSpawnGroup=false
    Sampled Position Offset=&MakeVector(
     X=-4
     Y=&Multiply_Float(
      A=&MergeHelper_FloatSwitch(0=-7, 1=7, 2=?, 3=?, Index=&ModuloInt(A=$Particles.PartitionParticles.Partition, B=2), Num=::NewEnumerator0)
      B=&FloatFromCurve(
       FloatCurve=DataInterface() {propertyValues {Curve {keys [{interpMode=RCIM_Cubic, tangentMode=RCTM_User, time=0, value=1}, {interpMode=RCIM_Cubic, tangentMode=RCTM_User, time=0.55, value=0.5}]}}}
       Scale Curve=1.5
      )
     )
     Z=-2
    )
    Sampled Position Offset Coordinate Space:ENiagaraSkelMeshTransforms=::NewEnumerator2
    Skeletal Mesh=$System.Skeletal Mesh
   )
   &ScaleColor(
    Scale Alpha=&FloatFromCurve(
     FloatCurve=DataInterface() {propertyValues {Curve {keys [
         {interpMode=RCIM_Cubic, tangentMode=RCTM_User, time=0, value=0}
         {interpMode=RCIM_Cubic, tangentMode=RCTM_User, time=0.1, value=1}
         {interpMode=RCIM_Cubic, tangentMode=RCTM_User, time=1, value=0}
        ]}}}
     Scale Curve=2
    )
    Scale RGB=&MakeVectorFromLinearColorRGB(Color=&ColorFromCurve(ColorCurve=DataInterface() {propertyValues {
        AlphaCurve {keys [<at 0 1 RCIM_Cubic>]}
        BlueCurve {keys [<at 0.0779 3.16 RCIM_Cubic>,<at 0.3496 0.6>,<at 1 0.0333 RCIM_Cubic>]}
        GreenCurve {keys [<at 0.0779 2.4 RCIM_Cubic>,<at 0.3496 0.84>,<at 1 0 RCIM_Cubic>]}
        RedCurve {keys [<at 0.0779 10 RCIM_Cubic>,<at 0.3496 3>,<at 1 1 RCIM_Cubic>]}
       }}))
    ScaleA=true
    ScaleRGB=true
   )
   &ParticleState()
  ]
  rendererClasses [&NiagaraSpriteRendererProperties]
  renderers [{rendererClass=&NiagaraSpriteRendererProperties}]
 }
}
summary {
 systemName=NS_Peppercat_Reload
 emitters [
  {bEnabled=false, emitterName=Minimal001, rendererClasses [&NiagaraSpriteRendererProperties]}
  {bEnabled=false, emitterName=VentSmoke001, rendererClasses [&NiagaraSpriteRendererProperties]}
  {bEnabled=false, emitterName=VentSmoke, rendererClasses [&NiagaraSpriteRendererProperties]}
  {emitterName=Minimal, rendererClasses [&NiagaraSpriteRendererProperties, &NiagaraMeshRendererProperties]}
  {
   bEnabled=false
   emitterName=Minimal002
   rendererClasses [
    &NiagaraSpriteRendererProperties
    &NiagaraRibbonRendererProperties
    &NiagaraMeshRendererProperties
   ]
  }
  {emitterName=SplitLightning, rendererClasses [&NiagaraMeshRendererProperties]}
  {emitterName=Minimal003, rendererClasses [&NiagaraSpriteRendererProperties]}
  {emitterName=Minimal004, rendererClasses [&NiagaraSpriteRendererProperties]}
  {emitterName=NE_LoopingZaps, rendererClasses [&NiagaraMeshRendererProperties, &NiagaraSpriteRendererProperties]}
  {emitterName=Linger_Airglow, rendererClasses [&NiagaraSpriteRendererProperties]}
 ]
 userVariables [
  User.bIsFrontEnd {type=NiagaraBool}
  User.bIsFrontEndPreview {type=NiagaraBool}
 ]
}
